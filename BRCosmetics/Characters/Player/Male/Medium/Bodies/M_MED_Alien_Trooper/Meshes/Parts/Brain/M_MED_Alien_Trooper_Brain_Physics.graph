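# Physics asset for the brain part of the male medium Alien Trooper cosmetic character body.
# It defines capsule bodies on two bones: a large, offset, rotated capsule on dyn_brains and a tiny one on dyn_brains_MSTR.
# A single constraint joins the two with tight swing and twist limits, restricting the brain to slight movement relative to its master bone.
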PhysicsAsset /BRCosmetics/Characters/Player/Male/Medium/Bodies/M_MED_Alien_Trooper/Meshes/Parts/Brain/M_MED_Alien_Trooper_Brain_Physics
bodies {
 dyn_brains [{shapeType=Capsule, length=10.8858, radius=9.784, center=<6.5352 -2.7733 -0.1297>, rotation=<rot 0 -32.659 -83.5415>}]
 dyn_brains_MSTR [{shapeType=Capsule, length=0.505, radius=0.505, rotation=<rot 0 -90 0>}]
}
constraints [{bone1Name=dyn_brains, bone2Name=dyn_brains_MSTR, swing1LimitDegrees=10, swing2LimitDegrees=5, twistLimitDegrees=1}]
