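# Control Rig for the Alien Trooper Cyclops part that shrinks the eye bone dyn_Cyclops_eye_MSTR as the right upper eyelid rotates.
# The lid's rotation relative to R_eye becomes an Euler X angle, which is offset, made absolute and clamp-remapped into the weight of a non-uniform scale.
# Lower-lid math branches and a similar modifier for dyn_R_brain_eye_MSTR, a bone absent from this hierarchy, are not connected to execution.
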ControlRigBlueprint /BRCosmetics/Characters/Player/Male/Medium/Bodies/M_MED_Alien_Trooper/Meshes/Parts/Cyclops/M_MED_Alien_Trooper_Cyclops_CtrlRig
graphs [
 RigVMFunctionLibrary {links [], nodes []}
 RigVMModel {nodes [
   n1 = &RigUnit_BeginExecution()
   n2 = /Script/RigVM.RigVMFunction_MathVectorAdd(
    A=(X=0,Y=0,Z=0)
    B=(X=70,Y=0,Z=0)
    A.X=&RigVMFunction_MathQuaternionToEuler(
     RotationOrder=ZYX
     Value=&RigUnit_GetRelativeTransformForItem(Child=(Type=Bone,Name="R_eye_lid_upper_mid"), Parent=(Type=Bone,Name="R_eye")).RelativeTransform.Rotation
    ).Result.X
   ) {is=struct RigVMUnitNode, template="Add::Execute(in A,in B,out Result)"}
   &RigUnit_ModifyTransforms(
    ItemToModify=((Item=(Type=Bone,Name="dyn_Cyclops_eye_MSTR"),Transform=(Rotation=(X=0,Y=0,Z=0,W=1),Translation=(X=0,Y=0,Z=0),Scale3D=(X=0.9,Y=0.85,Z=0.8))))
    Weight=&RigVMFunction_MathFloatRemap(
     SourceMaximum=141
     SourceMinimum=65
     TargetMinimum=0
     bClamp=true
     Value=/Script/RigVM.RigVMFunction_MathFloatAbs(Value=@n2.Result.X) {is=struct RigVMUnitNode, template="Absolute::Execute(in Value,out Result)"}.Result
    ).Result
    ExecutePin=@n1.ExecutePin
   )
   n3 = &RigVMFunction_MathQuaternionToEuler(
    RotationOrder=ZYX
    Value=&RigUnit_GetRelativeTransformForItem(Child=(Type=Bone,Name="R_eye_lid_lower_mid"), Parent=(Type=Bone,Name="R_eye")).RelativeTransform.Rotation
   )
   &RigVMFunction_MathFloatAbs(Value=&RigVMFunction_MathVectorAdd(A=(X=0,Y=0,Z=0), B=(X=70,Y=0,Z=0), A.X=@n3.Result.X) {template="Add::Execute(in A,in B,out Result)"}.Result.X) {template="Absolute::Execute(in Value,out Result)"}
   n4 = &RigVMFunction_MathQuaternionToEuler(
    RotationOrder=ZYX
    Value=&RigUnit_GetRelativeTransformForItem(Child=(Type=Bone,Name="R_eye_lid_lower_mid"), Parent=(Type=Bone,Name="R_eye")).RelativeTransform.Rotation
   )
   &RigVMFunction_MathFloatAbs(Value=&RigVMFunction_MathVectorAdd(A=(X=0,Y=0,Z=0), B=(X=70,Y=0,Z=0), A.X=@n4.Result.X) {template="Add::Execute(in A,in B,out Result)"}.Result.X) {template="Absolute::Execute(in Value,out Result)"}
   &RigVMFunction_MathFloatAdd(B=&RigVMFunction_MathFloatRemap(SourceMaximum=-56, SourceMinimum=30, TargetMaximum=0.5, TargetMinimum=0, bClamp=true, Value=@n3.Result.X).Result) {template="Add::Execute(in A,in B,out Result)"}
   &RigVMFunction_MathFloatAdd(B=&RigVMFunction_MathFloatRemap(SourceMaximum=-56, SourceMinimum=30, TargetMaximum=0.5, TargetMinimum=0, bClamp=true, Value=@n4.Result.X).Result) {template="Add::Execute(in A,in B,out Result)"}
   n5 = &RigVMFunction_MathVectorAdd(
    A=(X=0,Y=0,Z=0)
    B=(X=70,Y=0,Z=0)
    A.X=&RigVMFunction_MathQuaternionToEuler(
     RotationOrder=ZYX
     Value=&RigUnit_GetRelativeTransformForItem(Child=(Type=Bone,Name="R_eye_lid_upper_mid"), Parent=(Type=Bone,Name="R_eye")).RelativeTransform.Rotation
    ).Result.X
   ) {template="Add::Execute(in A,in B,out Result)"}
   &RigUnit_ModifyTransforms(
    ExecutePin=()
    ItemToModify=((Item=(Type=Bone,Name="dyn_R_brain_eye_MSTR"),Transform=(Rotation=(X=0,Y=0,Z=0,W=1),Translation=(X=0,Y=0,Z=0),Scale3D=(X=1,Y=0.85,Z=1))))
    Weight=&RigVMFunction_MathFloatRemap(
     SourceMaximum=141
     SourceMinimum=65
     TargetMinimum=0
     bClamp=true
     Value=&RigVMFunction_MathFloatAbs(Value=@n5.Result.X) {template="Absolute::Execute(in Value,out Result)"}.Result
    ).Result
   )
  ]}
]
hierarchy [
 root {loc=<0 0 0>, rot=<0 0 0 1>}
 pelvis {parent=root, loc=<0 -0.7071 93.1191>, rot=<0.0127 0.707 -0.0127 -0.707>}
 spine_01 {parent=pelvis, loc=<6.7494 -0.6388 0>, rot=<0 0 0.034 -0.9994>}
 spine_02 {parent=spine_01, loc=<7.2518 -0.035 0>, rot=<0 0 0.0238 -0.9997>}
 spine_03 {parent=spine_02, loc=<6.0104 0 0>, rot=<0 0 0.0149 -0.9999>}
 spine_04 {parent=spine_03, loc=<8.1024 0 0>, rot=<0 0 -0.0132 -0.9999>}
 spine_05 {parent=spine_04, loc=<10.0146 0 0>, rot=<0 0 -0.0015 -1>}
 neck_01 {parent=spine_05, loc=<11.3138 0.1529 0>, rot=<0 0 -0.2003 -0.9797>}
 neck_02 {parent=neck_01, loc=<4.5627 0 0>, rot=<0 0 0.0504 -0.9987>}
 head {parent=neck_02, loc=<5.3914 0 0>, rot=<0 0 0.115 -0.9934>}
 faceAttach {parent=head, loc=<-2.737 0.5477 -0.066>, rot=<0.0082 -0.7071 -0.0082 -0.7071>, scale=<0.89 0.89 0.89>}
 C_brow_mid {parent=faceAttach, loc=<-0.095 7.005 16.7645>, rot=<0 0.0159 -0.0031 -0.9999>}
 L_brow_mid {parent=faceAttach, loc=<4.5305 4.5046 16.14>, rot=<-0.0002 -0.0046 0.2553 -0.9669>}
 L_brow_outer {parent=faceAttach, loc=<7.4264 1.3836 13.3271>, rot=<-0.0775 -0.1014 0.4575 -0.88>}
 L_cheek_inner {parent=faceAttach, loc=<4.6628 7.8637 4.2323>, rot=<-0.0699 -0.2679 0.3186 -0.9065>}
 C_nose_bridge {parent=faceAttach, loc=<-0.094 10.0931 4.9854>, rot=<-0.2218 0.0001 0 -0.9751>}
 R_brow_mid {parent=faceAttach, loc=<-4.6788 4.5046 16.14>, rot=<-0.0002 0.0046 -0.2553 -0.9669>}
 R_brow_outer {parent=faceAttach, loc=<-7.5748 1.3836 13.3271>, rot=<-0.0775 0.1014 -0.4575 -0.88>}
 R_cheek_inner {parent=faceAttach, loc=<-4.7989 7.8699 4.2217>, rot=<-0.07 0.268 -0.3186 -0.9065>}
 R_eye {parent=faceAttach, loc=<-0.0738 4.874 10.3385>, rot=<0 0 0.9998 -0.0187>}
 R_eye_lid_upper_mid {parent=R_eye, loc=<-0.0423 -1.2436 0.2292>, rot=<0.0439 -0.0007 0.0167 -0.9989>}
 R_eye_lid_lower_mid {parent=R_eye, loc=<-0.0087 -0.9503 -0.7083>, rot=<0.0381 -0.0006 0.0167 -0.9991>}
 dyn_Cyclops_eye_MSTR {parent=R_eye, loc=<0 0 0>, rot=<0 0 0 1>}
 L_eye {parent=faceAttach, loc=<-0.0738 4.874 10.3385>, rot=<0 0 0.9998 0.0187>}
 L_eye_lid_upper_mid {parent=L_eye, loc=<0.0251 -1.1623 0.2388>, rot=<0.0523 0 0 -0.9986>}
 L_eye_lid_lower_mid {parent=L_eye, loc=<0.0223 -0.923 -0.7488>, rot=<0.0329 0 0 -0.9995>}
 dyn_spikes_MSTR {parent=head, loc=<9.0221 -2.0566 -0.2936>, rot=<0.0029 -0.7071 -0.0029 -0.7071>}
 dyn_spikes_19 {parent=dyn_spikes_MSTR, loc=<-1.7766 -2.2967 0.5331>, rot=<0 0 0 1>}
 dyn_spikes_18 {parent=dyn_spikes_MSTR, loc=<-1.391 -2.4432 -4.432>, rot=<0 0 0 1>}
 dyn_spikes_17 {parent=dyn_spikes_MSTR, loc=<-0.7684 -2.6661 -2.9917>, rot=<0 0 0 1>}
 dyn_spikes_16 {parent=dyn_spikes_MSTR, loc=<-1.0744 -2.9638 -1.0352>, rot=<0 0 0 1>}
 dyn_spikes_15 {parent=dyn_spikes_MSTR, loc=<1.3773 -2.6536 -1.194>, rot=<0 0 0 1>}
 dyn_spikes_14 {parent=dyn_spikes_MSTR, loc=<1.5175 -2.2973 0.3707>, rot=<0 0 0 1>}
 dyn_spikes_13 {parent=dyn_spikes_MSTR, loc=<3.4453 -0.4053 0.9949>, rot=<0 0 0 1>}
 dyn_spikes_12 {parent=dyn_spikes_MSTR, loc=<0.1187 0.1838 2.5381>, rot=<0 0 0 1>}
 dyn_spikes_11 {parent=dyn_spikes_MSTR, loc=<-1.5235 1.2848 2.7414>, rot=<0 0 0 1>}
 dyn_spikes_10 {parent=dyn_spikes_MSTR, loc=<1.7069 2.8631 2.8359>, rot=<0 0 0 1>}
 dyn_spikes_09 {parent=dyn_spikes_MSTR, loc=<-2.2926 3.947 2.9123>, rot=<0 0 0 1>}
 dyn_spikes_08 {parent=dyn_spikes_MSTR, loc=<-0.5333 3.1613 3.0987>, rot=<0 0 0 1>}
 dyn_spikes_07 {parent=dyn_spikes_MSTR, loc=<1.6866 5.1899 3.0696>, rot=<0 0 0 1>}
 dyn_spikes_06 {parent=dyn_spikes_MSTR, loc=<-0.0751 6.1932 3.2126>, rot=<0 0 0 1>}
 dyn_spikes_05 {parent=dyn_spikes_MSTR, loc=<-2.5 6.1876 3.0547>, rot=<0 0 0 1>}
 dyn_spikes_04 {parent=dyn_spikes_MSTR, loc=<0.6765 7.4006 3.1451>, rot=<0 0 0 1>}
 dyn_spikes_03 {parent=dyn_spikes_MSTR, loc=<-1.5863 7.9867 3.1565>, rot=<0 0 0 1>}
 dyn_spikes_02 {parent=dyn_spikes_MSTR, loc=<0.6176 8.9846 2.8687>, rot=<0 0 0 1>}
 dyn_spikes_01 {parent=dyn_spikes_MSTR, loc=<-0.2955 9.354 3.0114>, rot=<0 0 0 1>}
 dyn_spikes_00 {parent=dyn_spikes_MSTR, loc=<-1.2604 9.3232 2.929>, rot=<0 0 0 1>}
 dyn_spikes_20 {parent=dyn_spikes_MSTR, loc=<3.3572 -1.2186 -0.5929>, rot=<0 0 0 1>}
 C_glabella_down_pose {type=Curve, loc=<0 0 0>, rot=<0 0 0 1>}
 tongue_up_pose {type=Curve, loc=<0 0 0>, rot=<0 0 0 1>}
 phoneme_mbp_pose {type=Curve, loc=<0 0 0>, rot=<0 0 0 1>}
 L_upper_lip_lower_pose {type=Curve, loc=<0 0 0>, rot=<0 0 0 1>}
 R_blink_pose {type=Curve, loc=<0 0 0>, rot=<0 0 0 1>}
 R_wide_pose {type=Curve, loc=<0 0 0>, rot=<0 0 0 1>}
 look_right_pose {type=Curve, loc=<0 0 0>, rot=<0 0 0 1>}
 jaw_up_pose {type=Curve, loc=<0 0 0>, rot=<0 0 0 1>}
 L_squeeze_pose {type=Curve, loc=<0 0 0>, rot=<0 0 0 1>}
 R_brow_up_pose {type=Curve, loc=<0 0 0>, rot=<0 0 0 1>}
 look_left_pose {type=Curve, loc=<0 0 0>, rot=<0 0 0 1>}
 tongue_back_pose {type=Curve, loc=<0 0 0>, rot=<0 0 0 1>}
 R_squint_inner_pose {type=Curve, loc=<0 0 0>, rot=<0 0 0 1>}
 look_up_pose {type=Curve, loc=<0 0 0>, rot=<0 0 0 1>}
 phoneme_fv_pose {type=Curve, loc=<0 0 0>, rot=<0 0 0 1>}
 jaw_back_pose {type=Curve, loc=<0 0 0>, rot=<0 0 0 1>}
 R_smile_pose {type=Curve, loc=<0 0 0>, rot=<0 0 0 1>}
 R_lip_corner_narrow_pose {type=Curve, loc=<0 0 0>, rot=<0 0 0 1>}
 eye_convergence_pose {type=Curve, loc=<0 0 0>, rot=<0 0 0 1>}
 R_brow_down_pose {type=Curve, loc=<0 0 0>, rot=<0 0 0 1>}
 R_frown_pose {type=Curve, loc=<0 0 0>, rot=<0 0 0 1>}
 L_lower_lip_down_pose {type=Curve, loc=<0 0 0>, rot=<0 0 0 1>}
 L_blink_pose {type=Curve, loc=<0 0 0>, rot=<0 0 0 1>}
 L_lip_corner_narrow_pose {type=Curve, loc=<0 0 0>, rot=<0 0 0 1>}
 look_down_pose {type=Curve, loc=<0 0 0>, rot=<0 0 0 1>}
 R_upper_lip_raiser_pose {type=Curve, loc=<0 0 0>, rot=<0 0 0 1>}
 phoneme_oo_pose {type=Curve, loc=<0 0 0>, rot=<0 0 0 1>}
 R_brow_squeeze_in_pose {type=Curve, loc=<0 0 0>, rot=<0 0 0 1>}
 jaw_left_pose {type=Curve, loc=<0 0 0>, rot=<0 0 0 1>}
 L_lip_corner_wide_pose {type=Curve, loc=<0 0 0>, rot=<0 0 0 1>}
 C_nose_wrinkler_pose {type=Curve, loc=<0 0 0>, rot=<0 0 0 1>}
 R_lower_lip_up_pose {type=Curve, loc=<0 0 0>, rot=<0 0 0 1>}
 jaw_open_pose {type=Curve, loc=<0 0 0>, rot=<0 0 0 1>}
 L_brow_up_pose {type=Curve, loc=<0 0 0>, rot=<0 0 0 1>}
 R_lip_corner_wide_pose {type=Curve, loc=<0 0 0>, rot=<0 0 0 1>}
 L_lower_lip_up_pose {type=Curve, loc=<0 0 0>, rot=<0 0 0 1>}
 R_lower_lip_down_pose {type=Curve, loc=<0 0 0>, rot=<0 0 0 1>}
 L_upper_lip_raiser_pose {type=Curve, loc=<0 0 0>, rot=<0 0 0 1>}
 R_cheek_up_pose {type=Curve, loc=<0 0 0>, rot=<0 0 0 1>}
 L_brow_down_pose {type=Curve, loc=<0 0 0>, rot=<0 0 0 1>}
 jaw_forward_pose {type=Curve, loc=<0 0 0>, rot=<0 0 0 1>}
 L_squint_inner_pose {type=Curve, loc=<0 0 0>, rot=<0 0 0 1>}
 tongue_down_pose {type=Curve, loc=<0 0 0>, rot=<0 0 0 1>}
 R_upper_lip_lower_pose {type=Curve, loc=<0 0 0>, rot=<0 0 0 1>}
 L_brow_squeeze_in_pose {type=Curve, loc=<0 0 0>, rot=<0 0 0 1>}
 jaw_right_pose {type=Curve, loc=<0 0 0>, rot=<0 0 0 1>}
 mouth_down_pose {type=Curve, loc=<0 0 0>, rot=<0 0 0 1>}
 C_glabella_up_pose {type=Curve, loc=<0 0 0>, rot=<0 0 0 1>}
 phoneme_ch_pose {type=Curve, loc=<0 0 0>, rot=<0 0 0 1>}
 L_frown_pose {type=Curve, loc=<0 0 0>, rot=<0 0 0 1>}
 R_squeeze_pose {type=Curve, loc=<0 0 0>, rot=<0 0 0 1>}
 tongue_forward_pose {type=Curve, loc=<0 0 0>, rot=<0 0 0 1>}
 L_upperLidDown_pose {type=Curve, loc=<0 0 0>, rot=<0 0 0 1>}
 R_upperLidDown_pose {type=Curve, loc=<0 0 0>, rot=<0 0 0 1>}
 L_wide_pose {type=Curve, loc=<0 0 0>, rot=<0 0 0 1>}
 L_smile_pose {type=Curve, loc=<0 0 0>, rot=<0 0 0 1>}
 mouth_up_pose {type=Curve, loc=<0 0 0>, rot=<0 0 0 1>}
 L_cheek_up_pose {type=Curve, loc=<0 0 0>, rot=<0 0 0 1>}
 base_pose {type=Curve, loc=<0 0 0>, rot=<0 0 0 1>}
 blendParent1 {type=Curve, loc=<0 0 0>, rot=<0 0 0 1>}
]
