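# Physics asset for the Owl part of the male medium Alien Trooper body cosmetic, defining capsule bodies on its dyn_ bones.
# Brow, ear, mohawk and ponytail bones form separate chains, each linked bone to bone back to its own master with limited swing and twist.
# Larger capsules cover the upper spine, neck and head, with no constraints joining them to each other or to the chain masters.
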PhysicsAsset /BRCosmetics/Characters/Player/Male/Medium/Bodies/M_MED_Alien_Trooper/Meshes/Parts/Owl/M_MED_Alien_Trooper_Owl_Physics
bodies {
 dyn_brow_l_01 [{shapeType=Capsule, length=2, radius=1.3793, center=<1.5831 0.0001 0>, rotation=<rot 90 0 0>}]
 dyn_brow_l_02 [{shapeType=Capsule, length=2, radius=1.3793, center=<1.5831 0.0001 0>, rotation=<rot 90 0 0>}]
 dyn_brow_l_master [{shapeType=Capsule, length=2, radius=2, rotation=<rot 90 0 0>}]
 dyn_brow_r_01 [{shapeType=Capsule, length=2, radius=1.3793, center=<1.5831 0.0001 0>, rotation=<rot 90 0 0>}]
 dyn_brow_r_02 [{shapeType=Capsule, length=2, radius=1.3793, center=<1.5831 0.0001 0>, rotation=<rot 90 0 0>}]
 dyn_brow_r_master [{shapeType=Capsule, length=2, radius=2, rotation=<rot 90 0 0>}]
 dyn_ear_l_01 [{shapeType=Capsule, length=2, radius=2, center=<2.5999 -0.0002 0>, rotation=<rot 90 0 0>}]
 dyn_ear_l_02 [{shapeType=Capsule, length=2, radius=2, center=<2.5999 -0.0002 0>, rotation=<rot 90 0 0>}]
 dyn_ear_l_03 [{shapeType=Capsule, length=2, radius=2, center=<2.5999 -0.0002 0>, rotation=<rot 90 0 0>}]
 dyn_ear_l_04 [{shapeType=Capsule, length=2, radius=2, center=<2.5999 -0.0002 0>, rotation=<rot 90 0 0>}]
 dyn_ear_l_master [{shapeType=Capsule, length=2, radius=2, rotation=<rot 90 0 0>}]
 dyn_ear_r_01 [{shapeType=Capsule, length=2, radius=2, center=<2.5541 -0.0001 0>, rotation=<rot 90 0 0>}]
 dyn_ear_r_02 [{shapeType=Capsule, length=2, radius=2, center=<2.5541 -0.0001 0>, rotation=<rot 90 0 0>}]
 dyn_ear_r_03 [{shapeType=Capsule, length=2, radius=2, center=<2.5541 -0.0001 0>, rotation=<rot 90 0 0>}]
 dyn_ear_r_04 [{shapeType=Capsule, length=2, radius=2, center=<2.5541 -0.0001 0>, rotation=<rot 90 0 0>}]
 dyn_ear_r_master [{shapeType=Capsule, length=2, radius=2, rotation=<rot 90 0 0>}]
 dyn_mohawk_front_01 [{shapeType=Capsule, length=2, radius=2, center=<2.8379 0 0>, rotation=<rot 90 0 0>}]
 dyn_mohawk_front_02 [{shapeType=Capsule, length=2, radius=2, center=<2.8379 0 0>, rotation=<rot 90 0 0>}]
 dyn_mohawk_front_master [{shapeType=Capsule, length=2, radius=2, rotation=<rot 90 0 0>}]
 dyn_mohawk_mid_01 [{shapeType=Capsule, length=2, radius=2, center=<2.8379 0 0>, rotation=<rot 90 0 0>}]
 dyn_mohawk_mid_master [{shapeType=Capsule, length=2, radius=2, rotation=<rot 90 0 0>}]
 dyn_neck_01_owl [{shapeType=Capsule, length=4.4535, radius=4.4106, center=<2.9541 0 -0.0001>}]
 dyn_neck_02_owl [{shapeType=Capsule, length=15.1194, radius=4.7953, center=<2.9541 0 -0.0001>}]
 dyn_owl_head [{shapeType=Capsule, length=7.2407, radius=4.3899, rotation=<rot 90 0 0>}]
 dyn_ponytail_01 [{shapeType=Capsule, length=2, radius=2, center=<2.7178 0 0>, rotation=<rot 0 0 90>}]
 dyn_ponytail_02 [{shapeType=Capsule, length=4.3477, radius=1.4292, center=<2.7182 0.0001 0>, rotation=<rot 0 0 90>}]
 dyn_ponytail_03 [{shapeType=Capsule, length=0.9655, radius=2.1559, center=<2.5983 0.0001 0>, rotation=<rot 64.8928 0 0>}]
 dyn_ponytail_04 [{shapeType=Capsule, length=0.0001, radius=2.2619, center=<3.1452 0.0001 0>, rotation=<rot 90 0 0>}]
 dyn_ponytail_master [{shapeType=Capsule, length=2, radius=2, rotation=<rot 0 0 90>}]
 dyn_spine_05_owl [{shapeType=Capsule, length=7.609, radius=9.6104, center=<10.8969 -3.8183 -0.0003>, rotation=<rot 0 -90 0>}]
}
constraints [
 {bone1Name=dyn_ponytail_01, bone2Name=dyn_ponytail_master, swing1LimitDegrees=10, swing2LimitDegrees=30, twistLimitDegrees=8}
 {bone1Name=dyn_ponytail_02, bone2Name=dyn_ponytail_01, swing1LimitDegrees=15, swing2LimitDegrees=45, twistLimitDegrees=10}
 {bone1Name=dyn_ponytail_03, bone2Name=dyn_ponytail_02, swing1LimitDegrees=15, swing2LimitDegrees=30, twistLimitDegrees=10}
 {bone1Name=dyn_ponytail_04, bone2Name=dyn_ponytail_03, swing1LimitDegrees=15, swing2LimitDegrees=30, twistLimitDegrees=10}
 {bone1Name=dyn_ear_l_01, bone2Name=dyn_ear_l_master, swing1LimitDegrees=8, swing2LimitDegrees=8, twistLimitDegrees=5}
 {bone1Name=dyn_ear_l_02, bone2Name=dyn_ear_l_01, swing1LimitDegrees=8, swing2LimitDegrees=8, twistLimitDegrees=5}
 {bone1Name=dyn_ear_l_03, bone2Name=dyn_ear_l_02, swing1LimitDegrees=8, swing2LimitDegrees=8, twistLimitDegrees=5}
 {bone1Name=dyn_ear_l_04, bone2Name=dyn_ear_l_03, swing1LimitDegrees=8, swing2LimitDegrees=8, twistLimitDegrees=5}
 {bone1Name=dyn_ear_r_01, bone2Name=dyn_ear_r_master, swing1LimitDegrees=8, swing2LimitDegrees=8, twistLimitDegrees=5}
 {bone1Name=dyn_ear_r_02, bone2Name=dyn_ear_r_01, swing1LimitDegrees=8, swing2LimitDegrees=8, twistLimitDegrees=5}
 {bone1Name=dyn_ear_r_03, bone2Name=dyn_ear_r_02, swing1LimitDegrees=8, swing2LimitDegrees=8, twistLimitDegrees=5}
 {bone1Name=dyn_ear_r_04, bone2Name=dyn_ear_r_03, swing1LimitDegrees=8, swing2LimitDegrees=8, twistLimitDegrees=5}
 {bone1Name=dyn_mohawk_front_02, bone2Name=dyn_mohawk_front_01, swing1LimitDegrees=12, swing2LimitDegrees=5, twistLimitDegrees=5}
 {bone1Name=dyn_mohawk_front_01, bone2Name=dyn_mohawk_front_master, swing1LimitDegrees=12, swing2LimitDegrees=5, twistLimitDegrees=5}
 {bone1Name=dyn_mohawk_mid_01, bone2Name=dyn_mohawk_mid_master, swing1LimitDegrees=12, swing2LimitDegrees=5, twistLimitDegrees=5}
 {bone1Name=dyn_brow_r_02, bone2Name=dyn_brow_r_01, swing1LimitDegrees=8, swing2LimitDegrees=20, twistLimitDegrees=5}
 {bone1Name=dyn_brow_r_01, bone2Name=dyn_brow_r_master, swing1LimitDegrees=8, swing2LimitDegrees=20, twistLimitDegrees=5}
 {bone1Name=dyn_brow_l_02, bone2Name=dyn_brow_l_01, swing1LimitDegrees=8, swing2LimitDegrees=20, twistLimitDegrees=5}
 {bone1Name=dyn_brow_l_01, bone2Name=dyn_brow_l_master, swing1LimitDegrees=8, swing2LimitDegrees=20, twistLimitDegrees=5}
]
