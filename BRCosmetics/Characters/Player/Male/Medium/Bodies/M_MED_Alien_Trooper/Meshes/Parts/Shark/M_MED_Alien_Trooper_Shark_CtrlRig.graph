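# Control Rig for the Alien Trooper Shark part that scales the dyn_L_eye and dyn_R_eye bones from eyelid motion.
# Upper and lower lid rotations, read relative to each eye as Euler X, are remapped with clamping and summed into Modify Transforms weights that squash the bones.
# The right upper-lid remap keeps its default target maximum, unlike the left's 0.5.
# Two lower-lid offset/absolute chains are unused.
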ControlRigBlueprint /BRCosmetics/Characters/Player/Male/Medium/Bodies/M_MED_Alien_Trooper/Meshes/Parts/Shark/M_MED_Alien_Trooper_Shark_CtrlRig
graphs [
 RigVMFunctionLibrary {links [], nodes []}
 RigVMModel {nodes [
   RigUnit_BeginExecution = &RigUnit_BeginExecution()
   n1 = &RigUnit_GetRelativeTransformForItem(Child=(Type=Bone,Name="L_eye_lid_upper_mid"), Parent=(Type=Bone,Name="L_eye"))
   n2 = /Script/RigVM.RigVMFunction_MathFloatAbs(Value=/Script/RigVM.RigVMFunction_MathVectorAdd(
     A=(X=0,Y=0,Z=0)
     B=(X=70,Y=0,Z=0)
     A.X=&RigVMFunction_MathQuaternionToEuler(RotationOrder=ZYX, Value=@n1.RelativeTransform.Rotation).Result.X
    ) {is=struct RigVMUnitNode, template="Add::Execute(in A,in B,out Result)"}.Result.X) {is=struct RigVMUnitNode, template="Absolute::Execute(in Value,out Result)"}
   n3 = &RigVMFunction_MathQuaternionToEuler(
    RotationOrder=ZYX
    Value=&RigUnit_GetRelativeTransformForItem(Child=(Type=Bone,Name="L_eye_lid_lower_mid"), Parent=(Type=Bone,Name="L_eye")).RelativeTransform.Rotation
   )
   &RigUnit_ModifyTransforms(
    ItemToModify=((Item=(Type=Bone,Name="dyn_L_eye"),Transform=(Rotation=(X=0,Y=0,Z=0,W=1),Translation=(X=0,Y=0,Z=0),Scale3D=(X=0.95,Y=0.55,Z=0.55))))
    Weight=&RigVMFunction_MathFloatAdd(
     A=&RigVMFunction_MathFloatRemap(SourceMaximum=111, SourceMinimum=0, TargetMaximum=0.5, TargetMinimum=0, bClamp=true, Value=@n2.Result).Result
     B=&RigVMFunction_MathFloatRemap(SourceMaximum=-56, SourceMinimum=30, TargetMaximum=0.5, TargetMinimum=0, bClamp=true, Value=@n3.Result.X).Result
    ) {template="Add::Execute(in A,in B,out Result)"}.Result
    ExecutePin=@RigUnit_BeginExecution.ExecutePin
   )
   n4 = &RigUnit_GetRelativeTransformForItem(Child=(Type=Bone,Name="R_eye_lid_upper_mid"), Parent=(Type=Bone,Name="R_eye"))
   n5 = &RigVMFunction_MathFloatAbs(Value=&RigVMFunction_MathVectorAdd(
     A=(X=0,Y=0,Z=0)
     B=(X=70,Y=0,Z=0)
     A.X=&RigVMFunction_MathQuaternionToEuler(RotationOrder=ZYX, Value=@n4.RelativeTransform.Rotation).Result.X
    ) {template="Add::Execute(in A,in B,out Result)"}.Result.X) {template="Absolute::Execute(in Value,out Result)"}
   n6 = &RigVMFunction_MathQuaternionToEuler(
    RotationOrder=ZYX
    Value=&RigUnit_GetRelativeTransformForItem(Child=(Type=Bone,Name="R_eye_lid_lower_mid"), Parent=(Type=Bone,Name="R_eye")).RelativeTransform.Rotation
   )
   > &RigUnit_ModifyTransforms(
    ItemToModify=((Item=(Type=Bone,Name="dyn_R_eye"),Transform=(Rotation=(X=0,Y=0,Z=0,W=1),Translation=(X=0,Y=0,Z=0),Scale3D=(X=0.95,Y=0.55,Z=0.55))))
    Weight=&RigVMFunction_MathFloatAdd(
     A=&RigVMFunction_MathFloatRemap(SourceMaximum=111, SourceMinimum=0, TargetMinimum=0, bClamp=true, Value=@n5.Result).Result
     B=&RigVMFunction_MathFloatRemap(SourceMaximum=-56, SourceMinimum=30, TargetMaximum=0.5, TargetMinimum=0, bClamp=true, Value=@n6.Result.X).Result
    ) {template="Add::Execute(in A,in B,out Result)"}.Result
   )
   &RigVMFunction_MathFloatAbs(Value=&RigVMFunction_MathVectorAdd(A=(X=0,Y=0,Z=0), B=(X=70,Y=0,Z=0), A.X=@n3.Result.X) {template="Add::Execute(in A,in B,out Result)"}.Result.X) {template="Absolute::Execute(in Value,out Result)"}
   &RigVMFunction_MathFloatAbs(Value=&RigVMFunction_MathVectorAdd(A=(X=0,Y=0,Z=0), B=(X=70,Y=0,Z=0), A.X=@n6.Result.X) {template="Add::Execute(in A,in B,out Result)"}.Result.X) {template="Absolute::Execute(in Value,out Result)"}
  ]}
]
hierarchy [
 root {loc=<0 0 0>, rot=<0 0 0 1>}
 pelvis {parent=root, loc=<0 -0.7071 93.1191>, rot=<0.0127 0.707 -0.0127 -0.707>}
 spine_01 {parent=pelvis, loc=<6.7494 -0.6388 0>, rot=<0 0 0.034 -0.9994>}
 spine_02 {parent=spine_01, loc=<7.2518 -0.035 0>, rot=<0 0 0.0238 -0.9997>}
 spine_03 {parent=spine_02, loc=<6.0104 0 0>, rot=<0 0 0.0149 -0.9999>}
 spine_04 {parent=spine_03, loc=<8.1024 0 0>, rot=<0 0 -0.0132 -0.9999>}
 spine_05 {parent=spine_04, loc=<10.0146 0 0>, rot=<0 0 -0.0015 -1>}
 neck_01 {parent=spine_05, loc=<11.3138 0.1529 0>, rot=<0 0 -0.2003 -0.9797>}
 neck_02 {parent=neck_01, loc=<4.5627 0 0>, rot=<0 0 0.0504 -0.9987>}
 head {parent=neck_02, loc=<5.3914 0 0>, rot=<0 0 0.115 -0.9934>}
 faceAttach {parent=head, loc=<-4.1329 -0.9158 0>, rot=<0.0082 -0.7071 -0.0082 -0.7071>, scale=<0.89 0.89 0.89>}
 C_brow_mid {parent=faceAttach, loc=<-0.0254 6.9445 17.5604>, rot=<0 0 0 -1>}
 L_brow_mid {parent=faceAttach, loc=<3.2242 6.6669 17.3148>, rot=<-0.0001 -0.0042 0.2553 -0.9669>}
 L_brow_outer {parent=faceAttach, loc=<5.7748 4.3473 16.5646>, rot=<-0.0287 -0.04 0.4491 -0.8921>}
 L_cheek_inner {parent=faceAttach, loc=<4.737 9.4714 5.8382>, rot=<-0.0535 -0.2264 0.3241 -0.917>}
 C_nose_bridge {parent=faceAttach, loc=<-0.0198 10.9565 13.7375>, rot=<-0.2218 0 0 -0.9751>}
 R_brow_mid {parent=faceAttach, loc=<-3.2663 6.6583 17.3148>, rot=<-0.0001 0.0042 -0.2553 -0.9669>}
 R_brow_outer {parent=faceAttach, loc=<-5.8027 4.334 16.5656>, rot=<-0.0286 0.0399 -0.4491 -0.8921>}
 R_cheek_inner {parent=faceAttach, loc=<-4.7247 9.4777 5.8276>, rot=<-0.0515 0.2214 -0.3247 -0.9181>}
 R_eye {parent=faceAttach, loc=<-18.6704 -0.1574 16.4081>, rot=<0 0 0.6936 -0.7203>}
 R_eye_lid_upper_mid {parent=R_eye, loc=<-0.0102 -0.2915 1.2083>, rot=<0.5751 -0.0096 0.0137 -0.8179>, scale=<1 1.0359 0.9653>}
 R_eye_lid_lower_mid {parent=R_eye, loc=<0.0111 -0.3564 -1.124>, rot=<-0.2796 0.0047 0.016 -0.96>, scale=<1 1.0111 0.989>}
 dyn_R_eye {parent=R_eye, loc=<0 0 0>, rot=<0 0 0 1>}
 L_eye {parent=faceAttach, loc=<18.6704 -0.1574 16.4081>, rot=<0 0 -0.6936 -0.7203>}
 L_eye_lid_upper_mid {parent=L_eye, loc=<0.0352 -0.2993 1.148>, rot=<0.5715 0.0096 -0.0137 -0.8204>}
 L_eye_lid_lower_mid {parent=L_eye, loc=<0.0331 -0.3246 -1.1431>, rot=<-0.2696 -0.0045 -0.0161 -0.9628>}
 dyn_L_eye {parent=L_eye, loc=<0 0 0>, rot=<0 0 0 1>}
 dyn_fin_l {parent=head, loc=<14.1456 0.3502 -3.1493>, rot=<-0.6582 0.0651 0.2556 -0.7051>}
 dyn_fin_l_02 {parent=dyn_fin_l, loc=<4.0643 0 0>, rot=<-0.003 0.0704 -0.0038 -0.9975>}
 dyn_fin_r {parent=head, loc=<14.1456 0.3502 3.1493>, rot=<-0.7048 0.256 0.0658 -0.6583>}
 dyn_fin_r_02 {parent=dyn_fin_r, loc=<4.0643 0 0>, rot=<0.0016 0.0676 0.0027 -0.9977>}
 C_glabella_down_pose {type=Curve, loc=<0 0 0>, rot=<0 0 0 1>}
 tongue_up_pose {type=Curve, loc=<0 0 0>, rot=<0 0 0 1>}
 phoneme_mbp_pose {type=Curve, loc=<0 0 0>, rot=<0 0 0 1>}
 L_upper_lip_lower_pose {type=Curve, loc=<0 0 0>, rot=<0 0 0 1>}
 R_blink_pose {type=Curve, loc=<0 0 0>, rot=<0 0 0 1>}
 R_wide_pose {type=Curve, loc=<0 0 0>, rot=<0 0 0 1>}
 look_right_pose {type=Curve, loc=<0 0 0>, rot=<0 0 0 1>}
 jaw_up_pose {type=Curve, loc=<0 0 0>, rot=<0 0 0 1>}
 L_squeeze_pose {type=Curve, loc=<0 0 0>, rot=<0 0 0 1>}
 R_brow_up_pose {type=Curve, loc=<0 0 0>, rot=<0 0 0 1>}
 look_left_pose {type=Curve, loc=<0 0 0>, rot=<0 0 0 1>}
 tongue_back_pose {type=Curve, loc=<0 0 0>, rot=<0 0 0 1>}
 R_squint_inner_pose {type=Curve, loc=<0 0 0>, rot=<0 0 0 1>}
 look_up_pose {type=Curve, loc=<0 0 0>, rot=<0 0 0 1>}
 phoneme_fv_pose {type=Curve, loc=<0 0 0>, rot=<0 0 0 1>}
 jaw_back_pose {type=Curve, loc=<0 0 0>, rot=<0 0 0 1>}
 R_smile_pose {type=Curve, loc=<0 0 0>, rot=<0 0 0 1>}
 R_lip_corner_narrow_pose {type=Curve, loc=<0 0 0>, rot=<0 0 0 1>}
 eye_convergence_pose {type=Curve, loc=<0 0 0>, rot=<0 0 0 1>}
 R_brow_down_pose {type=Curve, loc=<0 0 0>, rot=<0 0 0 1>}
 R_frown_pose {type=Curve, loc=<0 0 0>, rot=<0 0 0 1>}
 L_lower_lip_down_pose {type=Curve, loc=<0 0 0>, rot=<0 0 0 1>}
 L_blink_pose {type=Curve, loc=<0 0 0>, rot=<0 0 0 1>}
 L_lip_corner_narrow_pose {type=Curve, loc=<0 0 0>, rot=<0 0 0 1>}
 look_down_pose {type=Curve, loc=<0 0 0>, rot=<0 0 0 1>}
 R_upper_lip_raiser_pose {type=Curve, loc=<0 0 0>, rot=<0 0 0 1>}
 phoneme_oo_pose {type=Curve, loc=<0 0 0>, rot=<0 0 0 1>}
 R_brow_squeeze_in_pose {type=Curve, loc=<0 0 0>, rot=<0 0 0 1>}
 jaw_left_pose {type=Curve, loc=<0 0 0>, rot=<0 0 0 1>}
 L_lip_corner_wide_pose {type=Curve, loc=<0 0 0>, rot=<0 0 0 1>}
 C_nose_wrinkler_pose {type=Curve, loc=<0 0 0>, rot=<0 0 0 1>}
 R_lower_lip_up_pose {type=Curve, loc=<0 0 0>, rot=<0 0 0 1>}
 jaw_open_pose {type=Curve, loc=<0 0 0>, rot=<0 0 0 1>}
 L_brow_up_pose {type=Curve, loc=<0 0 0>, rot=<0 0 0 1>}
 R_lip_corner_wide_pose {type=Curve, loc=<0 0 0>, rot=<0 0 0 1>}
 L_lower_lip_up_pose {type=Curve, loc=<0 0 0>, rot=<0 0 0 1>}
 R_lower_lip_down_pose {type=Curve, loc=<0 0 0>, rot=<0 0 0 1>}
 L_upper_lip_raiser_pose {type=Curve, loc=<0 0 0>, rot=<0 0 0 1>}
 R_cheek_up_pose {type=Curve, loc=<0 0 0>, rot=<0 0 0 1>}
 L_brow_down_pose {type=Curve, loc=<0 0 0>, rot=<0 0 0 1>}
 jaw_forward_pose {type=Curve, loc=<0 0 0>, rot=<0 0 0 1>}
 L_squint_inner_pose {type=Curve, loc=<0 0 0>, rot=<0 0 0 1>}
 tongue_down_pose {type=Curve, loc=<0 0 0>, rot=<0 0 0 1>}
 R_upper_lip_lower_pose {type=Curve, loc=<0 0 0>, rot=<0 0 0 1>}
 L_brow_squeeze_in_pose {type=Curve, loc=<0 0 0>, rot=<0 0 0 1>}
 jaw_right_pose {type=Curve, loc=<0 0 0>, rot=<0 0 0 1>}
 mouth_down_pose {type=Curve, loc=<0 0 0>, rot=<0 0 0 1>}
 C_glabella_up_pose {type=Curve, loc=<0 0 0>, rot=<0 0 0 1>}
 phoneme_ch_pose {type=Curve, loc=<0 0 0>, rot=<0 0 0 1>}
 L_frown_pose {type=Curve, loc=<0 0 0>, rot=<0 0 0 1>}
 R_squeeze_pose {type=Curve, loc=<0 0 0>, rot=<0 0 0 1>}
 tongue_forward_pose {type=Curve, loc=<0 0 0>, rot=<0 0 0 1>}
 L_upperLidDown_pose {type=Curve, loc=<0 0 0>, rot=<0 0 0 1>}
 R_upperLidDown_pose {type=Curve, loc=<0 0 0>, rot=<0 0 0 1>}
 L_wide_pose {type=Curve, loc=<0 0 0>, rot=<0 0 0 1>}
 L_smile_pose {type=Curve, loc=<0 0 0>, rot=<0 0 0 1>}
 mouth_up_pose {type=Curve, loc=<0 0 0>, rot=<0 0 0 1>}
 L_cheek_up_pose {type=Curve, loc=<0 0 0>, rot=<0 0 0 1>}
 base_pose {type=Curve, loc=<0 0 0>, rot=<0 0 0 1>}
 blendParent1 {type=Curve, loc=<0 0 0>, rot=<0 0 0 1>}
]
