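# Physics asset for the Shark part of the male medium Alien Trooper character body, covering the head and two side fins.
# Capsule bodies sit on the head and on a mirrored two-bone dynamic chain for each left and right fin.
# Each chain constrains outer fin to inner fin and inner fin to head, with identical narrow swing limits and almost no twist, restricting the fins to slight flexing.
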PhysicsAsset /BRCosmetics/Characters/Player/Male/Medium/Bodies/M_MED_Alien_Trooper/Meshes/Parts/Shark/M_MED_Alien_Trooper_Shark_Physics
bodies {
 dyn_fin_l [{shapeType=Capsule, length=2.9035, radius=1.399, center=<2.3644 -0.0749 0.1417>, rotation=<rot 0 90.843 -86.468>}]
 dyn_fin_l_02 [{shapeType=Capsule, length=2.0523, radius=0.8965, center=<4.1394 -0.073 0.0061>, rotation=<rot 0 90.2044 -88.2508>}]
 dyn_fin_r [{shapeType=Capsule, length=2.9035, radius=1.399, center=<2.3646 0.0753 0.1379>, rotation=<rot 0 89.1686 -86.5591>}]
 dyn_fin_r_02 [{shapeType=Capsule, length=2.0523, radius=0.8966, center=<4.1386 0.0668 0.0166>, rotation=<rot 0 89.6955 -88.0155>}]
 head [{shapeType=Capsule, length=3.25, radius=4.3185, center=<7.1116 2.7395 0.0072>, rotation=<rot -0.0417 -90 0.0486>}]
}
constraints [
 {bone1Name=dyn_fin_l_02, bone2Name=dyn_fin_l, swing1LimitDegrees=10, swing2LimitDegrees=16, twistLimitDegrees=1}
 {bone1Name=dyn_fin_r_02, bone2Name=dyn_fin_r, swing1LimitDegrees=10, swing2LimitDegrees=16, twistLimitDegrees=1}
 {bone1Name=dyn_fin_l, bone2Name=head, swing1LimitDegrees=10, swing2LimitDegrees=16, twistLimitDegrees=1}
 {bone1Name=dyn_fin_r, bone2Name=head, swing1LimitDegrees=10, swing2LimitDegrees=16, twistLimitDegrees=1}
]
