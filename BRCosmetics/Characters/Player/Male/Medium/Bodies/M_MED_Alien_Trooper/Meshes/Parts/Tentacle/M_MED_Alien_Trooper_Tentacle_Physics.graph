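# Physics asset for the head tentacle part of the Alien Trooper player body cosmetic, defining capsule bodies for dynamic tentacle bones.
# Tentacles B and C are two-segment chains jointed to the head collision body, with wide, asymmetric swing limits.
# Tentacles A and D are single segments jointed to the same head body with tighter swing, and all joints allow little twist.
# A large, unconstrained neck capsule supplies an additional collision volume.
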PhysicsAsset /BRCosmetics/Characters/Player/Male/Medium/Bodies/M_MED_Alien_Trooper/Meshes/Parts/Tentacle/M_MED_Alien_Trooper_Tentacle_Physics
bodies {
 dyn_head_tentacle_a [{shapeType=Capsule, length=2.6426, radius=0.9631, center=<3.5 0 0>, rotation=<rot 90 0 0>}]
 dyn_head_tentacle_b [{shapeType=Capsule, length=3.2324, radius=1.2285, center=<2.5 0 0>, rotation=<rot 90 0 0>}]
 dyn_head_tentacle_b_01 [{shapeType=Capsule, length=4.3479, radius=0.9788, center=<3.5 0 0>, rotation=<rot 90 0 0>}]
 dyn_head_tentacle_c [{shapeType=Capsule, length=3.2324, radius=1.2285, center=<2.5 0 0>, rotation=<rot 90 0 0>}]
 dyn_head_tentacle_c_01 [{shapeType=Capsule, length=4.5843, radius=1.015, center=<3.5 0 0>, rotation=<rot 90 0 0>}]
 dyn_head_tentacle_d [{shapeType=Capsule, length=2.403, radius=1.0586, center=<3.5 0 0>, rotation=<rot 90 0 0>}]
 dyn_tentacle_head_collision [{shapeType=Capsule, length=1.538, radius=5.0481, center=<-1.8795 1.3117 0>}]
 dyn_tentacle_neck_01_collision [{shapeType=Capsule, length=7.4771, radius=12.5625, center=<-2.5411 2.8238 0.0001>, rotation=<rot -90 -20.556 -64.4532>}]
}
constraints [
 {bone1Name=dyn_head_tentacle_b_01, bone2Name=dyn_head_tentacle_b, swing1LimitDegrees=25, swing2LimitDegrees=80, twistLimitDegrees=3}
 {bone1Name=dyn_head_tentacle_c_01, bone2Name=dyn_head_tentacle_c, swing1LimitDegrees=25, swing2LimitDegrees=80, twistLimitDegrees=3}
 {bone1Name=dyn_head_tentacle_b, bone2Name=dyn_tentacle_head_collision, swing1LimitDegrees=25, swing2LimitDegrees=80, twistLimitDegrees=3}
 {bone1Name=dyn_head_tentacle_c, bone2Name=dyn_tentacle_head_collision, swing1LimitDegrees=25, swing2LimitDegrees=80, twistLimitDegrees=3}
 {bone1Name=dyn_head_tentacle_a, bone2Name=dyn_tentacle_head_collision, swing1LimitDegrees=10, swing2LimitDegrees=15, twistLimitDegrees=3}
 {bone1Name=dyn_head_tentacle_d, bone2Name=dyn_tentacle_head_collision, swing1LimitDegrees=10, swing2LimitDegrees=15, twistLimitDegrees=3}
]
