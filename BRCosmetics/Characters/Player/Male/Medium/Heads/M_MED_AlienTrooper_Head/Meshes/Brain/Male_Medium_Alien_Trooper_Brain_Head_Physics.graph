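# Physics asset for the Alien Trooper "Brain" head cosmetic on the male medium player character, providing collision bodies and joints for its left and right whiskers.
# Each side has a small capsule on a master whisker bone and three identical capsules on individual whisker bones.
# Each whisker is constrained directly to its side's master in a fan rather than a chain, with moderate swing and tight twist limits.
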PhysicsAsset /BRCosmetics/Characters/Player/Male/Medium/Heads/M_MED_AlienTrooper_Head/Meshes/Brain/Male_Medium_Alien_Trooper_Brain_Head_Physics
bodies {
 dyn_L_whiskers_MSTR [{shapeType=Capsule, length=0.3566, radius=0.37, rotation=<rot 0 -90 0>}]
 dyn_R_whiskers_MSTR [{shapeType=Capsule, length=0.3566, radius=0.37, rotation=<rot 0 -90 0>}]
 dyn_whiskers_L_01 [{shapeType=Capsule, length=1.0249, radius=0.4394, center=<1.1209 0 -0.0001>, rotation=<rot 90 0 0>}]
 dyn_whiskers_L_02 [{shapeType=Capsule, length=1.0249, radius=0.4394, center=<1.1209 0 -0.0001>, rotation=<rot 90 0 0>}]
 dyn_whiskers_L_03 [{shapeType=Capsule, length=1.0249, radius=0.4394, center=<1.1209 0 -0.0001>, rotation=<rot 90 0 0>}]
 dyn_whiskers_R_01 [{shapeType=Capsule, length=1.0249, radius=0.4394, center=<1.1209 0 -0.0001>, rotation=<rot 90 0 0>}]
 dyn_whiskers_R_02 [{shapeType=Capsule, length=1.0249, radius=0.4394, center=<1.1209 0 -0.0001>, rotation=<rot 90 0 0>}]
 dyn_whiskers_R_03 [{shapeType=Capsule, length=1.0249, radius=0.4394, center=<1.1209 0 -0.0001>, rotation=<rot 90 0 0>}]
}
constraints [
 {bone1Name=dyn_whiskers_L_01, bone2Name=dyn_L_whiskers_MSTR, swing1LimitDegrees=20, swing2LimitDegrees=20, twistLimitDegrees=4}
 {bone1Name=dyn_whiskers_L_02, bone2Name=dyn_L_whiskers_MSTR, swing1LimitDegrees=20, swing2LimitDegrees=20, twistLimitDegrees=4}
 {bone1Name=dyn_whiskers_L_03, bone2Name=dyn_L_whiskers_MSTR, swing1LimitDegrees=20, swing2LimitDegrees=20, twistLimitDegrees=4}
 {bone1Name=dyn_whiskers_R_01, bone2Name=dyn_R_whiskers_MSTR, swing1LimitDegrees=20, swing2LimitDegrees=20, twistLimitDegrees=4}
 {bone1Name=dyn_whiskers_R_02, bone2Name=dyn_R_whiskers_MSTR, swing1LimitDegrees=20, swing2LimitDegrees=20, twistLimitDegrees=4}
 {bone1Name=dyn_whiskers_R_03, bone2Name=dyn_R_whiskers_MSTR, swing1LimitDegrees=20, swing2LimitDegrees=20, twistLimitDegrees=4}
]
