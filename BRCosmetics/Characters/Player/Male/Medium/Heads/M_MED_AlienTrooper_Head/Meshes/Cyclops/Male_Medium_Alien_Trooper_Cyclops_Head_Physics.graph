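# Physics asset for the male medium Alien Trooper Cyclops head cosmetic, providing capsule bodies for the head and two dynamic pincer chains.
# Each pincer has two segments joined by constraints with swing and twist limits of a few degrees.
# The chains are asymmetric: the left runs through a small master body whose joint to the head has free swing and twist, while the right attaches directly to the head with tight limits.
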PhysicsAsset /BRCosmetics/Characters/Player/Male/Medium/Heads/M_MED_AlienTrooper_Head/Meshes/Cyclops/Male_Medium_Alien_Trooper_Cyclops_Head_Physics
bodies {
 dyn_Cyclops_pincer_l_01 [{shapeType=Capsule, length=1.9158, radius=0.8371, center=<2.2114 0 0.0001>, rotation=<rot 90 0 0>}]
 dyn_Cyclops_pincer_l_02 [{shapeType=Capsule, length=4.7908, radius=1.3371, center=<4.5506 0 0.0002>, rotation=<rot 90 0 0>}]
 dyn_Cyclops_pincer_l_MSTR [{shapeType=Capsule, length=0.505, radius=0.505, rotation=<rot 0 -90 0>}]
 dyn_Cyclops_pincer_r_01 [{shapeType=Capsule, length=1.9158, radius=0.8371, center=<2.2114 0 0.0001>, rotation=<rot 90 0 0>}]
 dyn_Cyclops_pincer_r_02 [{shapeType=Capsule, length=4.7908, radius=1.3371, center=<4.5506 0 0.0002>, rotation=<rot 90 0 0>}]
 head [{shapeType=Capsule, length=8.1291, radius=3.2492, center=<-1.5397 3.3009 0.0027>, rotation=<rot 0 17.1389 -89.8924>}]
}
constraints [
 {bone1Name=dyn_Cyclops_pincer_r_02, bone2Name=dyn_Cyclops_pincer_r_01, swing1LimitDegrees=5, swing2LimitDegrees=2, twistLimitDegrees=1}
 {bone1Name=dyn_Cyclops_pincer_l_01, bone2Name=dyn_Cyclops_pincer_l_MSTR, swing1LimitDegrees=5, swing2LimitDegrees=2, twistLimitDegrees=1}
 {bone1Name=dyn_Cyclops_pincer_l_02, bone2Name=dyn_Cyclops_pincer_l_01, swing1LimitDegrees=5, swing2LimitDegrees=2, twistLimitDegrees=1}
 {
  bone1Name=dyn_Cyclops_pincer_l_MSTR
  bone2Name=head
  swing1LimitDegrees=45
  swing1Motion=Free
  swing2LimitDegrees=45
  swing2Motion=Free
  twistLimitDegrees=45
  twistMotion=Free
 }
 {bone1Name=dyn_Cyclops_pincer_r_01, bone2Name=head, swing1LimitDegrees=5, swing2LimitDegrees=2, twistLimitDegrees=1}
]
