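# Control Rig for the Alien Trooper Owl head that procedurally adjusts the beard and moustache master bones as the head moves.
# It reads the head's rotation relative to spine_05 and takes the Z Euler angle as the driver.
# Clamped remaps turn that angle into small offsets and rotations on four ModifyTransforms steps, with mirrored yaw for the moustache sides.
# Brow, ponytail, mohawk, ear bones and face pose curves are left untouched.
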ControlRigBlueprint /BRCosmetics/Characters/Player/Male/Medium/Heads/M_MED_AlienTrooper_Head/Meshes/Owl/Male_Medium_Alien_Trooper_Owl_Head_CtrlRig
graphs [
 RigVMFunctionLibrary {links [], nodes []}
 RigVMModel {nodes [
   RigUnit_BeginExecution = &RigUnit_BeginExecution()
   n1 = &RigVMFunction_MathQuaternionToEuler(RotationOrder=ZYX, Value=&RigUnit_GetRelativeBoneTransform(Bone=head, Space=spine_05).Transform.Rotation)
   &RigUnit_ModifyTransforms(
    ItemToModify=((Item=(Type=Bone,Name="dyn_beard_l_master"),Transform=(Rotation=(X=0,Y=0,Z=0,W=1),Translation=(X=0,Y=0,Z=0),Scale3D=(X=1,Y=1,Z=1))))
    Weight=1
    ItemToModify.0.Transform.Translation.Z=&RigVMFunction_MathFloatRemap(SourceMaximum=15, SourceMinimum=-15, TargetMinimum=0, bClamp=true, Value=@n1.Result.Z).Result
    ItemToModify.0.Transform.Rotation=/Script/RigVM.RigVMFunction_MathQuaternionFromRotator(
     Rotator=(Pitch=-60,Yaw=0,Roll=0)
     Rotator.Pitch=&RigVMFunction_MathFloatRemap(SourceMaximum=15, SourceMinimum=-15, TargetMaximum=35, TargetMinimum=0, bClamp=true, Value=@n1.Result.Z).Result
    ) {is=struct RigVMUnitNode}.Result
    ExecutePin=@RigUnit_BeginExecution.ExecutePin
   )
   > &RigUnit_ModifyTransforms(
    ItemToModify=((Item=(Type=Bone,Name="dyn_beard_r_master"),Transform=(Rotation=(X=0,Y=0,Z=0,W=1),Translation=(X=0,Y=0,Z=0),Scale3D=(X=1,Y=1,Z=1))))
    Weight=1
    ItemToModify.0.Transform.Translation.Z=&RigVMFunction_MathFloatRemap(SourceMaximum=15, SourceMinimum=-15, TargetMinimum=0, bClamp=true, Value=@n1.Result.Z).Result
    ItemToModify.0.Transform.Rotation=&RigVMFunction_MathQuaternionFromRotator(
     Rotator=(Pitch=-60,Yaw=0,Roll=0)
     Rotator.Pitch=&RigVMFunction_MathFloatRemap(SourceMaximum=15, SourceMinimum=-15, TargetMaximum=30, TargetMinimum=0, bClamp=true, Value=@n1.Result.Z).Result
    ).Result
   )
   n2 = &RigVMFunction_MathQuaternionToEuler(RotationOrder=ZYX, Value=&RigUnit_GetRelativeBoneTransform(Bone=head, Space=spine_05).Transform.Rotation)
   > &RigUnit_ModifyTransforms(
    ItemToModify=((Item=(Type=Bone,Name="dyn_moustache_l_master"),Transform=(Rotation=(X=0,Y=0,Z=0,W=1),Translation=(X=-1,Y=0,Z=0),Scale3D=(X=1,Y=1,Z=1))))
    Weight=1
    ItemToModify.0.Transform.Translation.Z=&RigVMFunction_MathFloatRemap(SourceMaximum=15, SourceMinimum=-15, TargetMaximum=0.75, TargetMinimum=0, bClamp=true, Value=@n2.Result.Z).Result
    ItemToModify.0.Transform.Translation.X=&RigVMFunction_MathFloatRemap(SourceMaximum=15, SourceMinimum=-15, TargetMaximum=-0.75, TargetMinimum=0, bClamp=true, Value=@n2.Result.Z).Result
    ItemToModify.0.Transform.Rotation=&RigVMFunction_MathQuaternionFromRotator(
     Rotator=(Pitch=-60,Yaw=40,Roll=0)
     Rotator.Pitch=&RigVMFunction_MathFloatRemap(SourceMaximum=20, SourceMinimum=0, TargetMaximum=35, TargetMinimum=0, bClamp=true, Value=@n2.Result.Z).Result
     Rotator.Yaw=&RigVMFunction_MathFloatRemap(SourceMaximum=20, SourceMinimum=0, TargetMaximum=40, TargetMinimum=0, bClamp=true, Value=@n2.Result.Z).Result
    ).Result
   )
   > &RigUnit_ModifyTransforms(
    ItemToModify=((Item=(Type=Bone,Name="dyn_moustache_r_master"),Transform=(Rotation=(X=0,Y=0,Z=0,W=1),Translation=(X=0,Y=0,Z=0),Scale3D=(X=1,Y=1,Z=1))))
    Weight=1
    ItemToModify.0.Transform.Translation.X=&RigVMFunction_MathFloatRemap(SourceMaximum=15, SourceMinimum=-15, TargetMaximum=-0.5, TargetMinimum=0, bClamp=true, Value=@n2.Result.Z).Result
    ItemToModify.0.Transform.Translation.Z=&RigVMFunction_MathFloatRemap(SourceMaximum=15, SourceMinimum=-15, TargetMaximum=0.5, TargetMinimum=0, bClamp=true, Value=@n2.Result.Z).Result
    ItemToModify.0.Transform.Rotation=&RigVMFunction_MathQuaternionFromRotator(
     Rotator=(Pitch=-60,Yaw=0,Roll=0)
     Rotator.Pitch=&RigVMFunction_MathFloatRemap(SourceMaximum=20, SourceMinimum=0, TargetMaximum=35, TargetMinimum=0, bClamp=true, Value=@n2.Result.Z).Result
     Rotator.Yaw=&RigVMFunction_MathFloatRemap(SourceMaximum=20, SourceMinimum=0, TargetMaximum=-40, TargetMinimum=0, bClamp=true, Value=@n2.Result.Z).Result
    ).Result
   )
  ]}
]
hierarchy [
 root {loc=<0 0 0>, rot=<0 0 0 1>}
 pelvis {parent=root, loc=<0 -0.7071 93.1191>, rot=<0.0127 0.707 -0.0127 -0.707>}
 spine_01 {parent=pelvis, loc=<6.7494 -0.6388 0>, rot=<0 0 0.034 -0.9994>}
 spine_02 {parent=spine_01, loc=<7.2518 -0.035 0>, rot=<0 0 0.0238 -0.9997>}
 spine_03 {parent=spine_02, loc=<6.0104 0 0>, rot=<0 0 0.0149 -0.9999>}
 spine_04 {parent=spine_03, loc=<8.1024 0 0>, rot=<0 0 -0.0132 -0.9999>}
 spine_05 {parent=spine_04, loc=<10.0146 0 0>, rot=<0 0 -0.0015 -1>}
 neck_01 {parent=spine_05, loc=<11.3138 0.1529 0>, rot=<0 0 -0.2003 -0.9797>}
 neck_02 {parent=neck_01, loc=<4.5627 0 0>, rot=<0 0 0.0504 -0.9987>}
 head {parent=neck_02, loc=<5.3914 0 0>, rot=<0 0 0.115 -0.9934>}
 faceAttach {parent=head, loc=<-2.737 0.5477 -0.066>, rot=<0.0082 -0.7071 -0.0082 -0.7071>, scale=<0.89 0.89 0.89>}
 C_brow_mid {parent=faceAttach, loc=<-0.095 9.875 13.5609>, rot=<0 0.0124 -0.0025 -0.9999>}
 L_brow_mid {parent=faceAttach, loc=<4.3193 8.2187 14.0853>, rot=<-0.0001 -0.0043 0.2553 -0.9669>}
 L_brow_outer {parent=faceAttach, loc=<6.3552 5.9631 13.503>, rot=<0.1033 0.1262 0.4123 -0.8963>}
 dyn_brow_l_master {parent=L_brow_outer, loc=<0.1315 0.1179 1.7088>, rot=<0.8312 0.2252 0.2768 -0.4263>, scale=<1.1236 1.1236 1.1236>}
 dyn_brow_l_01 {parent=dyn_brow_l_master, loc=<0 0 0>, rot=<0.1027 -0.0295 -0.0064 -0.9943>}
 dyn_brow_l_02 {parent=dyn_brow_l_01, loc=<2.9361 0 0>, rot=<-0.0824 -0.2131 0.0126 -0.9735>}
 L_cheek_inner {parent=faceAttach, loc=<4.6628 7.8637 4.2323>, rot=<-0.0535 -0.2264 0.3241 -0.917>}
 C_lip_upper_mid {parent=faceAttach, loc=<-0.0604 9.9813 1.4968>, rot=<0.2225 0.0026 -0.001 -0.9749>}
 L_lip_upper_outer {parent=faceAttach, loc=<3.6507 8.2872 1.4655>, rot=<0.0444 -0.049 0.2536 -0.965>}
 dyn_moustache_l_master {parent=L_lip_upper_outer, loc=<-0.0086 0.1534 -0.8055>, rot=<-0.0297 -0.3358 -0.083 -0.9378>, scale=<1.1236 1.1236 1.1236>}
 dyn_moustache_l_01 {parent=dyn_moustache_l_master, loc=<0 0 0>, rot=<0.1179 0.0554 0.0126 -0.9914>}
 dyn_moustache_l_02 {parent=dyn_moustache_l_01, loc=<3.061 0 0>, rot=<-0.068 0.3106 0.0857 -0.9442>}
 L_lip_corner {parent=faceAttach, loc=<4.7856 6.1574 1.6249>, rot=<0.0021 0.0088 -0.0145 -0.9999>}
 C_nose_bridge {parent=faceAttach, loc=<-0.0577 10.2283 6.2388>, rot=<-0.2218 -0.0078 -0.0022 -0.9751>}
 R_brow_mid {parent=faceAttach, loc=<-4.4557 8.1978 14.0853>, rot=<-0.0001 0.0043 -0.2553 -0.9669>}
 R_brow_outer {parent=faceAttach, loc=<-6.4154 5.8895 13.5123>, rot=<0.1033 -0.1262 -0.4123 -0.8963>}
 dyn_brow_r_master {parent=R_brow_outer, loc=<-0.1315 0.2331 1.7088>, rot=<-0.2527 -0.8743 -0.3232 0.2593>, scale=<1.1236 1.1236 1.1236>}
 dyn_brow_r_01 {parent=dyn_brow_r_master, loc=<0 0 0>, rot=<-0.0144 -0.0287 0.0317 -0.999>}
 dyn_brow_r_02 {parent=dyn_brow_r_01, loc=<2.9365 0 0>, rot=<0.0092 -0.2122 0.0152 -0.9771>}
 R_cheek_inner {parent=faceAttach, loc=<-4.7989 7.8699 4.2217>, rot=<-0.0515 0.2214 -0.3247 -0.9181>}
 R_lip_corner {parent=faceAttach, loc=<-4.934 6.1574 1.6249>, rot=<0.0021 -0.0088 0.0145 -0.9999>}
 C_jaw {parent=faceAttach, loc=<-0.1795 2.9678 1.3919>, rot=<0 0 0 1>}
 L_lip_lower_outer {parent=C_jaw, loc=<3.6208 4.9011 -0.9263>, rot=<0.0196 0.0805 0.3326 -0.9394>}
 C_lip_lower_mid {parent=C_jaw, loc=<0.1018 6.364 -1.4706>, rot=<0 0.0004 0.0001 -1>}
 R_lip_lower_outer {parent=C_jaw, loc=<-3.4102 4.9011 -0.9263>, rot=<0.0196 -0.0805 -0.3326 -0.9394>}
 teeth_lower {parent=C_jaw, loc=<-0.0266 4.4674 -1.5253>, rot=<0 0 0 1>}
 tongue {parent=C_jaw, loc=<0.1043 1.732 -0.4624>, rot=<-0.1568 0 0 -0.9876>}
 dyn_beard_l_master {parent=C_jaw, loc=<3.0742 4.4638 -4.3897>, rot=<0.1887 -0.3331 -0.3728 -0.8452>, scale=<1.1236 1.1236 1.1236>}
 dyn_beard_l_01 {parent=dyn_beard_l_master, loc=<0 0 0>, rot=<0 0 0 1>}
 dyn_beard_l_02 {parent=dyn_beard_l_01, loc=<4.3085 0.0267 0.0483>, rot=<-0.0033 0.1606 -0.019 -0.9868>}
 dyn_beard_r_master {parent=C_jaw, loc=<-2.8637 4.4638 -4.3899>, rot=<-0.3378 0.178 0.8476 0.3686>, scale=<1.1236 1.1236 1.1236>}
 dyn_beard_r_01 {parent=dyn_beard_r_master, loc=<0 0 0>, rot=<0 0 0 -1>}
 dyn_beard_r_02 {parent=dyn_beard_r_01, loc=<4.3086 -0.028 0.0476>, rot=<-0.0172 0.1605 0.0198 -0.9867>}
 R_eye {parent=faceAttach, loc=<-4.8652 6.7109 10.3329>, rot=<0 0 0.9998 -0.0187>}
 R_eye_lid_upper_mid {parent=R_eye, loc=<-0.0396 -1.1714 0.2631>, rot=<0.0523 -0.0009 0.0167 -0.9985>}
 R_eye_lid_lower_mid {parent=R_eye, loc=<-0.0085 -0.9425 -0.7051>, rot=<0.0329 -0.0005 0.0167 -0.9993>}
 L_eye {parent=faceAttach, loc=<4.7168 6.7109 10.3329>, rot=<0 0 0.9998 0.0187>}
 L_eye_lid_upper_mid {parent=L_eye, loc=<0.0251 -1.1623 0.2388>, rot=<0.0523 0 0 -0.9986>}
 L_eye_lid_lower_mid {parent=L_eye, loc=<0.0223 -0.923 -0.7488>, rot=<0.0329 0 0 -0.9995>}
 teeth_upper {parent=faceAttach, loc=<-0.0719 7.6457 2.5471>, rot=<0 0 0 1>}
 R_lip_upper_outer {parent=faceAttach, loc=<-3.7991 8.2872 1.4655>, rot=<0.0444 0.049 -0.2536 -0.965>}
 dyn_moustache_r_master {parent=R_lip_upper_outer, loc=<0.0086 0.1534 -0.805>, rot=<-0.3369 -0.016 0.9404 0.0446>, scale=<1.1236 1.1236 1.1236>}
 dyn_moustache_r_01 {parent=dyn_moustache_r_master, loc=<0 0 0>, rot=<-0.1156 0.0602 0.028 -0.9911>}
 dyn_moustache_r_02 {parent=dyn_moustache_r_01, loc=<3.0612 0 0>, rot=<0.1059 0.3081 -0.0817 -0.9419>}
 dyn_ponytail_master {parent=head, loc=<9.0194 -4.8652 0>, rot=<0.0514 -0.7052 0.7052 -0.0514>}
 dyn_ponytail_01 {parent=dyn_ponytail_master, loc=<0 0 0>, rot=<0 0 0 1>}
 dyn_ponytail_02 {parent=dyn_ponytail_01, loc=<5.2861 0 0>, rot=<0 -0.0685 0 -0.9976>}
 dyn_ponytail_03 {parent=dyn_ponytail_02, loc=<5.2679 0 0>, rot=<0 0.245 0 -0.9695>}
 dyn_ponytail_04 {parent=dyn_ponytail_03, loc=<4.0271 0 0>, rot=<0 0.0815 0 -0.9967>}
 dyn_mohawk_mid_master {parent=head, loc=<9.4791 3.7715 0.0526>, rot=<0.7064 0.0374 -0.0312 -0.7061>}
 dyn_mohawk_mid_01 {parent=dyn_mohawk_mid_master, loc=<0 0 0>, rot=<0 0 0 -1>}
 dyn_mohawk_front_master {parent=head, loc=<8.3507 8.4799 0>, rot=<0.6939 0.1426 -0.1359 -0.6926>}
 dyn_mohawk_front_01 {parent=dyn_mohawk_front_master, loc=<0 0 0>, rot=<0 0 0 1>}
 dyn_mohawk_front_02 {parent=dyn_mohawk_front_01, loc=<4.5837 0 0>, rot=<-0.002 0.0418 -0.0044 -0.9991>}
 dyn_ear_l_master {parent=head, loc=<8.9259 -1.7407 -4.4039>, rot=<-0.6959 -0.0198 0.1221 -0.7074>}
 dyn_ear_l_01 {parent=dyn_ear_l_master, loc=<0 0 0>, rot=<0 0 0 -1>}
 dyn_ear_l_02 {parent=dyn_ear_l_01, loc=<4.742 0 0>, rot=<-0.0017 0.0009 -0.011 -0.9999>}
 dyn_ear_l_03 {parent=dyn_ear_l_02, loc=<4.79 0 0>, rot=<0.0003 0.185 0.0009 -0.9827>}
 dyn_ear_l_04 {parent=dyn_ear_l_03, loc=<5.2771 0 0>, rot=<0.018 0.1964 0.0203 -0.9802>}
 dyn_ear_r_master {parent=head, loc=<8.9255 -1.7407 4.4039>, rot=<-0.7074 0.1221 -0.0198 -0.6959>}
 dyn_ear_r_01 {parent=dyn_ear_r_master, loc=<0 0 0>, rot=<0 0 0 -1>}
 dyn_ear_r_02 {parent=dyn_ear_r_01, loc=<4.742 0 0>, rot=<0.0017 0.0009 0.011 -0.9999>}
 dyn_ear_r_03 {parent=dyn_ear_r_02, loc=<4.7909 0 0>, rot=<-0.0003 0.185 -0.0009 -0.9827>}
 dyn_ear_r_04 {parent=dyn_ear_r_03, loc=<5.2769 0 0>, rot=<-0.018 0.1964 -0.0203 -0.9801>}
 C_glabella_down_pose {type=Curve, loc=<0 0 0>, rot=<0 0 0 1>}
 tongue_up_pose {type=Curve, loc=<0 0 0>, rot=<0 0 0 1>}
 phoneme_mbp_pose {type=Curve, loc=<0 0 0>, rot=<0 0 0 1>}
 L_upper_lip_lower_pose {type=Curve, loc=<0 0 0>, rot=<0 0 0 1>}
 R_blink_pose {type=Curve, loc=<0 0 0>, rot=<0 0 0 1>}
 R_wide_pose {type=Curve, loc=<0 0 0>, rot=<0 0 0 1>}
 look_right_pose {type=Curve, loc=<0 0 0>, rot=<0 0 0 1>}
 jaw_up_pose {type=Curve, loc=<0 0 0>, rot=<0 0 0 1>}
 L_squeeze_pose {type=Curve, loc=<0 0 0>, rot=<0 0 0 1>}
 R_brow_up_pose {type=Curve, loc=<0 0 0>, rot=<0 0 0 1>}
 look_left_pose {type=Curve, loc=<0 0 0>, rot=<0 0 0 1>}
 tongue_back_pose {type=Curve, loc=<0 0 0>, rot=<0 0 0 1>}
 R_squint_inner_pose {type=Curve, loc=<0 0 0>, rot=<0 0 0 1>}
 look_up_pose {type=Curve, loc=<0 0 0>, rot=<0 0 0 1>}
 phoneme_fv_pose {type=Curve, loc=<0 0 0>, rot=<0 0 0 1>}
 jaw_back_pose {type=Curve, loc=<0 0 0>, rot=<0 0 0 1>}
 R_smile_pose {type=Curve, loc=<0 0 0>, rot=<0 0 0 1>}
 R_lip_corner_narrow_pose {type=Curve, loc=<0 0 0>, rot=<0 0 0 1>}
 eye_convergence_pose {type=Curve, loc=<0 0 0>, rot=<0 0 0 1>}
 R_brow_down_pose {type=Curve, loc=<0 0 0>, rot=<0 0 0 1>}
 R_frown_pose {type=Curve, loc=<0 0 0>, rot=<0 0 0 1>}
 L_lower_lip_down_pose {type=Curve, loc=<0 0 0>, rot=<0 0 0 1>}
 L_blink_pose {type=Curve, loc=<0 0 0>, rot=<0 0 0 1>}
 L_lip_corner_narrow_pose {type=Curve, loc=<0 0 0>, rot=<0 0 0 1>}
 look_down_pose {type=Curve, loc=<0 0 0>, rot=<0 0 0 1>}
 R_upper_lip_raiser_pose {type=Curve, loc=<0 0 0>, rot=<0 0 0 1>}
 phoneme_oo_pose {type=Curve, loc=<0 0 0>, rot=<0 0 0 1>}
 R_brow_squeeze_in_pose {type=Curve, loc=<0 0 0>, rot=<0 0 0 1>}
 jaw_left_pose {type=Curve, loc=<0 0 0>, rot=<0 0 0 1>}
 L_lip_corner_wide_pose {type=Curve, loc=<0 0 0>, rot=<0 0 0 1>}
 C_nose_wrinkler_pose {type=Curve, loc=<0 0 0>, rot=<0 0 0 1>}
 R_lower_lip_up_pose {type=Curve, loc=<0 0 0>, rot=<0 0 0 1>}
 jaw_open_pose {type=Curve, loc=<0 0 0>, rot=<0 0 0 1>}
 L_brow_up_pose {type=Curve, loc=<0 0 0>, rot=<0 0 0 1>}
 R_lip_corner_wide_pose {type=Curve, loc=<0 0 0>, rot=<0 0 0 1>}
 L_lower_lip_up_pose {type=Curve, loc=<0 0 0>, rot=<0 0 0 1>}
 R_lower_lip_down_pose {type=Curve, loc=<0 0 0>, rot=<0 0 0 1>}
 L_upper_lip_raiser_pose {type=Curve, loc=<0 0 0>, rot=<0 0 0 1>}
 R_cheek_up_pose {type=Curve, loc=<0 0 0>, rot=<0 0 0 1>}
 L_brow_down_pose {type=Curve, loc=<0 0 0>, rot=<0 0 0 1>}
 jaw_forward_pose {type=Curve, loc=<0 0 0>, rot=<0 0 0 1>}
 L_squint_inner_pose {type=Curve, loc=<0 0 0>, rot=<0 0 0 1>}
 tongue_down_pose {type=Curve, loc=<0 0 0>, rot=<0 0 0 1>}
 R_upper_lip_lower_pose {type=Curve, loc=<0 0 0>, rot=<0 0 0 1>}
 L_brow_squeeze_in_pose {type=Curve, loc=<0 0 0>, rot=<0 0 0 1>}
 jaw_right_pose {type=Curve, loc=<0 0 0>, rot=<0 0 0 1>}
 mouth_down_pose {type=Curve, loc=<0 0 0>, rot=<0 0 0 1>}
 C_glabella_up_pose {type=Curve, loc=<0 0 0>, rot=<0 0 0 1>}
 phoneme_ch_pose {type=Curve, loc=<0 0 0>, rot=<0 0 0 1>}
 L_frown_pose {type=Curve, loc=<0 0 0>, rot=<0 0 0 1>}
 R_squeeze_pose {type=Curve, loc=<0 0 0>, rot=<0 0 0 1>}
 tongue_forward_pose {type=Curve, loc=<0 0 0>, rot=<0 0 0 1>}
 L_upperLidDown_pose {type=Curve, loc=<0 0 0>, rot=<0 0 0 1>}
 R_upperLidDown_pose {type=Curve, loc=<0 0 0>, rot=<0 0 0 1>}
 L_wide_pose {type=Curve, loc=<0 0 0>, rot=<0 0 0 1>}
 L_smile_pose {type=Curve, loc=<0 0 0>, rot=<0 0 0 1>}
 mouth_up_pose {type=Curve, loc=<0 0 0>, rot=<0 0 0 1>}
 L_cheek_up_pose {type=Curve, loc=<0 0 0>, rot=<0 0 0 1>}
 base_pose {type=Curve, loc=<0 0 0>, rot=<0 0 0 1>}
 blendParent1 {type=Curve, loc=<0 0 0>, rot=<0 0 0 1>}
]
