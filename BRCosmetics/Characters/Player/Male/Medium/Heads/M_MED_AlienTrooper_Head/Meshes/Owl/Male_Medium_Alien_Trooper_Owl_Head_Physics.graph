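# Physics asset for the Owl variant of the male medium Alien Trooper cosmetic head, covering its beard and moustache dynamic bones.
# Left and right beard and moustache each form a master-01-02 capsule chain, linked by constraints with small swing and twist limits, tighter in swing on the moustache.
# An unconstrained spine_05 owl body adds two larger capsules, apparently as collision volume for the chains.
# No constraint ties the master bodies to the spine or each other.
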PhysicsAsset /BRCosmetics/Characters/Player/Male/Medium/Heads/M_MED_AlienTrooper_Head/Meshes/Owl/Male_Medium_Alien_Trooper_Owl_Head_Physics
bodies {
 dyn_beard_l_01 [{shapeType=Capsule, length=2, radius=2, center=<2.536 0 0>, rotation=<rot 90 0 0>}]
 dyn_beard_l_02 [{shapeType=Capsule, length=1.974, radius=1.3454, center=<2.536 0 0>, rotation=<rot 90 0 0>}]
 dyn_beard_l_master [{shapeType=Capsule, length=2, radius=2, rotation=<rot 90 0 0>}]
 dyn_beard_r_01 [{shapeType=Capsule, length=2, radius=2, center=<2.536 0 0>, rotation=<rot 90 0 0>}]
 dyn_beard_r_02 [{shapeType=Capsule, length=2, radius=2, center=<2.536 0 0>, rotation=<rot 90 0 0>}]
 dyn_beard_r_master [{shapeType=Capsule, length=2, radius=2, rotation=<rot 90 0 0>}]
 dyn_moustache_l_01 [{shapeType=Capsule, length=2, radius=2, rotation=<rot 90 0 0>}]
 dyn_moustache_l_02 [{shapeType=Capsule, length=2, radius=2, rotation=<rot 90 0 0>}]
 dyn_moustache_l_master [{shapeType=Capsule, length=2, radius=2, rotation=<rot 90 0 0>}]
 dyn_moustache_r_01 [{shapeType=Capsule, length=2, radius=2, rotation=<rot 90 0 0>}]
 dyn_moustache_r_02 [{shapeType=Capsule, length=2, radius=2, rotation=<rot 90 0 0>}]
 dyn_moustache_r_master [{shapeType=Capsule, length=2, radius=2, rotation=<rot 90 0 0>}]
 dyn_spine_05_owl [
  {shapeType=Capsule, length=9.7728, radius=9.0545, center=<1.8215 8.2196 0>}
  {shapeType=Capsule, length=15.2818, radius=3.5455, center=<9.3114 7.6635 -11.7329>, rotation=<rot 0 36.4306 85.9607>}
 ]
}
constraints [
 {bone1Name=dyn_moustache_l_01, bone2Name=dyn_moustache_l_master, swing1LimitDegrees=5, swing2LimitDegrees=5, twistLimitDegrees=5}
 {bone1Name=dyn_moustache_l_02, bone2Name=dyn_moustache_l_01, swing1LimitDegrees=5, swing2LimitDegrees=5, twistLimitDegrees=5}
 {bone1Name=dyn_beard_l_01, bone2Name=dyn_beard_l_master, swing1LimitDegrees=10, swing2LimitDegrees=15, twistLimitDegrees=3}
 {bone1Name=dyn_beard_l_02, bone2Name=dyn_beard_l_01, swing1LimitDegrees=10, swing2LimitDegrees=15, twistLimitDegrees=3}
 {bone1Name=dyn_beard_r_01, bone2Name=dyn_beard_r_master, swing1LimitDegrees=10, swing2LimitDegrees=15, twistLimitDegrees=3}
 {bone1Name=dyn_beard_r_02, bone2Name=dyn_beard_r_01, swing1LimitDegrees=10, swing2LimitDegrees=15, twistLimitDegrees=3}
 {bone1Name=dyn_moustache_r_01, bone2Name=dyn_moustache_r_master, swing1LimitDegrees=5, swing2LimitDegrees=5, twistLimitDegrees=5}
 {bone1Name=dyn_moustache_r_02, bone2Name=dyn_moustache_r_01, swing1LimitDegrees=5, swing2LimitDegrees=5, twistLimitDegrees=5}
]
