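# Physics asset for the Alien Trooper Shark head cosmetic on medium male characters, setting up physics for the dynamic upper outer lip bones.
# Capsule bodies form mirrored left and right three-bone chains (MSTR, 01, 02), each attached to a larger head capsule.
# Joints within each chain have limited swing and very tight twist, while MSTR-to-head joints leave swing and twist free, so their angular limit values do not apply.
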PhysicsAsset /BRCosmetics/Characters/Player/Male/Medium/Heads/M_MED_AlienTrooper_Head/Meshes/Shark/Male_Medium_Alien_Trooper_Shark_Head_Physics
bodies {
 dyn_l_lip_upper_outer_01 [{shapeType=Capsule, length=0.505, radius=0.505, center=<0.6404 0.0983 -0.071>, rotation=<rot 0 85.6407 -86.2532>}]
 dyn_l_lip_upper_outer_02 [{shapeType=Capsule, length=0.9882, radius=0.554, center=<0.9494 -0.0115 0.0525>, rotation=<rot 0 88.1539 -89.7465>}]
 dyn_l_lip_upper_outer_MSTR [{shapeType=Capsule, length=0.505, radius=0.505, rotation=<rot 0 -90 0>}]
 dyn_r_lip_upper_outer_01 [{shapeType=Capsule, length=0.505, radius=0.505, center=<0.6404 0.0983 -0.071>, rotation=<rot 0 85.6407 -86.2532>}]
 dyn_r_lip_upper_outer_02 [{shapeType=Capsule, length=0.9882, radius=0.554, center=<0.9494 -0.0115 0.0525>, rotation=<rot 0 88.1539 -89.7465>}]
 dyn_r_lip_upper_outer_MSTR [{shapeType=Capsule, length=0.505, radius=0.505, rotation=<rot 0 -90 0>}]
 head [{shapeType=Capsule, length=2.875, radius=3.4156, center=<-2.2895 2.7009 0>, rotation=<rot 90 92.4556 -177.5374>}]
}
constraints [
 {bone1Name=dyn_l_lip_upper_outer_02, bone2Name=dyn_l_lip_upper_outer_01, swing1LimitDegrees=20, swing2LimitDegrees=40, twistLimitDegrees=1}
 {bone1Name=dyn_r_lip_upper_outer_02, bone2Name=dyn_r_lip_upper_outer_01, swing1LimitDegrees=20, swing2LimitDegrees=40, twistLimitDegrees=1}
 {bone1Name=dyn_l_lip_upper_outer_01, bone2Name=dyn_l_lip_upper_outer_MSTR, swing1LimitDegrees=20, swing2LimitDegrees=40, twistLimitDegrees=1}
 {bone1Name=dyn_r_lip_upper_outer_01, bone2Name=dyn_r_lip_upper_outer_MSTR, swing1LimitDegrees=20, swing2LimitDegrees=40, twistLimitDegrees=1}
 {
  bone1Name=dyn_r_lip_upper_outer_MSTR
  bone2Name=head
  swing1LimitDegrees=45
  swing1Motion=Free
  swing2LimitDegrees=45
  swing2Motion=Free
  twistLimitDegrees=45
  twistMotion=Free
 }
 {
  bone1Name=dyn_l_lip_upper_outer_MSTR
  bone2Name=head
  swing1LimitDegrees=45
  swing1Motion=Free
  swing2LimitDegrees=45
  swing2Motion=Free
  twistLimitDegrees=45
  twistMotion=Free
 }
]
