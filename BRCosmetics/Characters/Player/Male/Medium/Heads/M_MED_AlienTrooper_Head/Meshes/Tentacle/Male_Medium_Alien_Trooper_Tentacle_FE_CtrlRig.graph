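# Control Rig for the Alien Trooper male head that poses and animates six mouth tentacle chains.
# From BeginExecution, fixed rotation offsets set the pose, then corrective transforms blend in by clamped remaps of head-to-neck_02 Euler X.
# Per-chain ChainHarmonics then adds pendulum, reach and wave motion with debug draw on.
# A separate chain of head-rotation-driven rotate, translate and scale correctives is not connected to BeginExecution.
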ControlRigBlueprint /BRCosmetics/Characters/Player/Male/Medium/Heads/M_MED_AlienTrooper_Head/Meshes/Tentacle/Male_Medium_Alien_Trooper_Tentacle_FE_CtrlRig
graphs [
 RigVMFunctionLibrary {links [], nodes []}
 RigVMModel {nodes [
   n1 = &RigUnit_GetRelativeTransformForItem(Child=(Type=Bone,Name="head"), Parent=(Type=Bone,Name="neck_02"))
   n2 = &RigVMFunction_MathQuaternionToEuler(RotationOrder=ZYX, Value=@n1.RelativeTransform.Rotation)
   n3 = &RigVMFunction_MathFloatRemap(SourceMaximum=0, SourceMinimum=-15, TargetMinimum=0, bClamp=true, Value=@n2.Result.Z)
   n4 = &RigVMFunction_MathQuaternionToEuler(RotationOrder=ZYX, Value=@n1.RelativeTransform.Rotation)
   n5 = &RigVMFunction_MathFloatRemap(SourceMaximum=0, SourceMinimum=-10, TargetMaximum=0.9, TargetMinimum=1, bClamp=true, Value=@n4.Result.Z)
   n6 = &RigVMFunction_MathFloatRemap(SourceMaximum=0, SourceMinimum=-10, TargetMaximum=-30, TargetMinimum=0, bClamp=true, Value=@n2.Result.Z)
   n7 = &RigVMFunction_MathFloatRemap(SourceMaximum=0, SourceMinimum=-10, TargetMaximum=-20, TargetMinimum=0, bClamp=true, Value=@n2.Result.Z)
   &RigUnit_ModifyTransforms(
    ExecutePin=()
    ItemToModify="((Item=(Type=Bone,Name=\"dyn_l_mouth_tentacle_a_MSTR\"),Transform=(Rotation=(X=0,Y=0,Z=0,W=1),Translation=(X=0,Y=0,Z=0),Scale3D=(X=1,Y=1,Z=1))),(Item=(Type=Bone,Name=\"dyn_r_mouth_tentacle_a_MSTR\"),Transform=(Rotation=(X=0,Y=0,Z=0,W=1),Translation=(X=0,Y=0,Z= ...(truncated)"
    Weight=1
    ItemToModify.1.Transform.Translation.Z=@n3.Result
    ItemToModify.3.Transform.Translation.Z=@n3.Result
    ItemToModify.4.Transform.Scale3D.X=@n5.Result
    ItemToModify.4.Transform.Scale3D.Y=@n5.Result
    ItemToModify.4.Transform.Scale3D.Z=@n5.Result
    ItemToModify.0.Transform.Rotation=&RigVMFunction_MathQuaternionFromEuler(
     Euler=(X=0,Y=0,Z=0)
     Euler.Z=&RigVMFunction_MathFloatRemap(SourceMaximum=0, SourceMinimum=-15, TargetMaximum=30, TargetMinimum=0, bClamp=true, Value=@n2.Result.Z).Result
     Euler.Y=@n6.Result
    ).Result
    ItemToModify.1.Transform.Rotation=&RigVMFunction_MathQuaternionFromEuler(
     Euler=(X=0,Y=0,Z=0)
     Euler.Z=&RigVMFunction_MathFloatRemap(SourceMaximum=-4, SourceMinimum=-10, TargetMaximum=-40, TargetMinimum=0, bClamp=true, Value=@n2.Result.Z).Result
     Euler.Y=@n6.Result
    ).Result
    ItemToModify.3.Transform.Rotation=&RigVMFunction_MathQuaternionFromEuler(
     Euler=(X=0,Y=0,Z=0)
     Euler.Y=@n7.Result
     Euler.X=&RigVMFunction_MathFloatRemap(SourceMaximum=0, SourceMinimum=-15, TargetMaximum=-30, TargetMinimum=0, bClamp=true, Value=@n2.Result.Z).Result
     Euler.Z=&RigVMFunction_MathFloatRemap(SourceMaximum=0, SourceMinimum=-15, TargetMaximum=-20, TargetMinimum=0, bClamp=true, Value=@n2.Result.Z).Result
    ).Result
    ItemToModify.4.Transform.Rotation=&RigVMFunction_MathQuaternionFromEuler(
     Euler=(X=0,Y=0,Z=0)
     Euler.Y=&RigVMFunction_MathFloatRemap(SourceMaximum=0, SourceMinimum=-10, TargetMaximum=0, TargetMinimum=0, bClamp=true, Value=@n4.Result.Z).Result
     Euler.Z=&RigVMFunction_MathFloatRemap(SourceMaximum=-4, SourceMinimum=-10, TargetMaximum=0, TargetMinimum=0, bClamp=true, Value=@n4.Result.Z).Result
    ).Result
    ItemToModify.2.Transform.Rotation=&RigVMFunction_MathQuaternionFromEuler(
     Euler=(X=0,Y=0,Z=0)
     Euler.Y=@n7.Result
     Euler.X=&RigVMFunction_MathFloatRemap(SourceMaximum=0, SourceMinimum=-15, TargetMaximum=30, TargetMinimum=0, bClamp=true, Value=@n2.Result.Z).Result
    ).Result
   )
   n8 = &RigVMFunction_MathQuaternionToEuler(
    RotationOrder=ZYX
    Value=&RigUnit_GetRelativeTransformForItem(Child=(Type=Bone,Name="head"), Parent=(Type=Bone,Name="neck_02")).RelativeTransform.Rotation
   )
   n9 = &RigVMFunction_MathFloatRemap(SourceMaximum=-20, SourceMinimum=0, TargetMaximum=0.8, TargetMinimum=1, bClamp=true, Value=@n8.Result.X)
   > &RigUnit_ModifyTransforms(
    ItemToModify="((Item=(Type=Bone,Name=\"dyn_l_mouth_tentacle_a_MSTR\"),Transform=(Rotation=(X=0,Y=0,Z=0,W=1),Translation=(X=0,Y=0,Z=0),Scale3D=(X=1,Y=1,Z=1))),(Item=(Type=Bone,Name=\"dyn_r_mouth_tentacle_a_MSTR\"),Transform=(Rotation=(X=0,Y=0,Z=0,W=1),Translation=(X=0,Y=0,Z= ...(truncated)"
    Weight=1
    ItemToModify.2.Transform.Translation.X=&RigVMFunction_MathFloatRemap(SourceMaximum=-20, SourceMinimum=0, TargetMaximum=-2, TargetMinimum=0, bClamp=true, Value=@n8.Result.X).Result
    ItemToModify.2.Transform.Translation.Z=&RigVMFunction_MathFloatRemap(SourceMaximum=-20, SourceMinimum=0, TargetMaximum=-1, TargetMinimum=0, bClamp=true, Value=@n8.Result.X).Result
    ItemToModify.0.Transform.Scale3D.X=@n9.Result
    ItemToModify.0.Transform.Scale3D.Y=@n9.Result
    ItemToModify.0.Transform.Scale3D.Z=@n9.Result
    ItemToModify.2.Transform.Scale3D.X=@n9.Result
    ItemToModify.2.Transform.Scale3D.Y=@n9.Result
    ItemToModify.2.Transform.Scale3D.Z=@n9.Result
    ItemToModify.0.Transform.Rotation=&RigVMFunction_MathQuaternionFromEuler(
     Euler=(X=0,Y=0,Z=0)
     Euler.Y=&RigVMFunction_MathFloatRemap(SourceMaximum=-30, SourceMinimum=0, TargetMaximum=-60, TargetMinimum=0, bClamp=true, Value=@n8.Result.X).Result
    ).Result
    ItemToModify.2.Transform.Rotation=&RigVMFunction_MathQuaternionFromEuler(
     Euler=(X=0,Y=0,Z=0)
     Euler.Y=&RigVMFunction_MathFloatRemap(SourceMaximum=-30, SourceMinimum=0, TargetMaximum=30, TargetMinimum=0, bClamp=true, Value=@n8.Result.X).Result
     Euler.X=&RigVMFunction_MathFloatRemap(SourceMaximum=-30, SourceMinimum=0, TargetMaximum=-20, TargetMinimum=0, bClamp=true, Value=@n8.Result.X).Result
    ).Result
    ItemToModify.3.Transform.Rotation=&RigVMFunction_MathQuaternionFromEuler(
     Euler=(X=0,Y=0,Z=0)
     Euler.Z=&RigVMFunction_MathFloatRemap(SourceMaximum=-30, SourceMinimum=0, TargetMaximum=25, TargetMinimum=0, bClamp=true, Value=@n8.Result.X).Result
     Euler.Y=&RigVMFunction_MathFloatRemap(SourceMaximum=-30, SourceMinimum=0, TargetMaximum=20, TargetMinimum=0, bClamp=true, Value=@n8.Result.X).Result
    ).Result
   )
   n10 = &RigVMFunction_MathQuaternionToEuler(
    RotationOrder=ZYX
    Value=&RigUnit_GetRelativeTransformForItem(Child=(Type=Bone,Name="head"), Parent=(Type=Bone,Name="neck_02")).RelativeTransform.Rotation
   )
   n11 = &RigVMFunction_MathFloatRemap(SourceMaximum=0, SourceMinimum=-10, TargetMaximum=0.8, TargetMinimum=1, bClamp=true, Value=@n10.Result.Z)
   > &RigUnit_ModifyTransforms(
    ItemToModify="((Item=(Type=Bone),Transform=(Rotation=(X=0,Y=0,Z=0,W=1),Translation=(X=0,Y=0,Z=0),Scale3D=(X=1,Y=1,Z=1))),(Item=(Type=Bone,Name=\"dyn_r_mouth_tentacle_c_MSTR\"),Transform=(Rotation=(X=0,Y=0,Z=0,W=1),Translation=(X=0,Y=0,Z=0),Scale3D=(X=1,Y=1,Z=1))),(Item=(T ...(truncated)"
    Weight=1
    ItemToModify.1.Transform.Translation.Y=&RigVMFunction_MathFloatRemap(SourceMaximum=0, SourceMinimum=-10, TargetMaximum=-2, TargetMinimum=0, bClamp=true, Value=@n10.Result.Z).Result
    ItemToModify.1.Transform.Scale3D.X=@n11.Result
    ItemToModify.1.Transform.Scale3D.Z=@n11.Result
    ItemToModify.1.Transform.Translation.Z=&RigVMFunction_MathFloatRemap(SourceMaximum=0, SourceMinimum=-10, TargetMaximum=-1, TargetMinimum=0, bClamp=true, Value=@n10.Result.Z).Result
    ItemToModify.1.Transform.Rotation=&RigVMFunction_MathQuaternionFromEuler(
     Euler=(X=0,Y=0,Z=0)
     Euler.Z=&RigVMFunction_MathFloatRemap(SourceMaximum=0, SourceMinimum=-10, TargetMaximum=-25, TargetMinimum=0, bClamp=true, Value=@n10.Result.Z).Result
     Euler.X=&RigVMFunction_MathFloatRemap(SourceMaximum=0, SourceMinimum=-10, TargetMaximum=-10, TargetMinimum=0, bClamp=true, Value=@n10.Result.Z).Result
    ).Result
    ItemToModify.2.Transform.Rotation=&RigVMFunction_MathQuaternionFromEuler(
     Euler=(X=0,Y=0,Z=0)
     Euler.Y=&RigVMFunction_MathFloatRemap(SourceMaximum=0, SourceMinimum=-10, TargetMaximum=-20, TargetMinimum=0, bClamp=true, Value=@n10.Result.Z).Result
     Euler.X=&RigVMFunction_MathFloatRemap(SourceMaximum=0, SourceMinimum=-10, TargetMaximum=30, TargetMinimum=0, bClamp=true, Value=@n10.Result.Z).Result
    ).Result
    ItemToModify.3.Transform.Rotation=&RigVMFunction_MathQuaternionFromEuler(
     Euler=(X=0,Y=0,Z=0)
     Euler.X=&RigVMFunction_MathFloatRemap(SourceMaximum=0, SourceMinimum=-10, TargetMaximum=40, TargetMinimum=0, bClamp=true, Value=@n10.Result.Z).Result
     Euler.Y=&RigVMFunction_MathFloatRemap(SourceMaximum=0, SourceMinimum=-10, TargetMaximum=-50, TargetMinimum=0, bClamp=true, Value=@n10.Result.Z).Result
    ).Result
    ItemToModify.4.Transform.Rotation=&RigVMFunction_MathQuaternionFromEuler(
     Euler=(X=0,Y=0,Z=0)
     Euler.X=&RigVMFunction_MathFloatRemap(SourceMaximum=0, SourceMinimum=-10, TargetMaximum=30, TargetMinimum=0, bClamp=true, Value=@n10.Result.Z).Result
     Euler.Y=&RigVMFunction_MathFloatRemap(SourceMaximum=0, SourceMinimum=-10, TargetMaximum=-50, TargetMinimum=0, bClamp=true, Value=@n10.Result.Z).Result
    ).Result
   )
   n12 = &RigVMFunction_MathQuaternionToEuler(
    RotationOrder=ZYX
    Value=&RigUnit_GetRelativeTransformForItem(Child=(Type=Bone,Name="head"), Parent=(Type=Bone,Name="neck_02")).RelativeTransform.Rotation
   )
   > &RigUnit_ModifyTransforms(
    ItemToModify=((Item=(Type=Bone),Transform=(Rotation=(X=0,Y=0,Z=0,W=1),Translation=(X=0,Y=0,Z=0),Scale3D=(X=1,Y=1,Z=1))),(Item=(Type=Bone,Name="dyn_r_mouth_tentacle_c_MSTR"),Transform=(Rotation=(X=0,Y=0,Z=0,W=1),Translation=(X=0,Y=0,Z=0),Scale3D=(X=1,Y=1,Z=1))))
    Weight=1
    ItemToModify.1.Transform.Rotation=&RigVMFunction_MathQuaternionFromEuler(
     Euler=(X=0,Y=0,Z=0)
     Euler.Y=&RigVMFunction_MathFloatRemap(SourceMaximum=-15, SourceMinimum=-9, TargetMaximum=15, TargetMinimum=0, bClamp=true, Value=@n12.Result.Z).Result
    ).Result
   )
   > &RigUnit_ModifyTransforms(
    ItemToModify=((Item=(Type=Bone),Transform=(Rotation=(X=0,Y=0,Z=0,W=1),Translation=(X=0,Y=0,Z=0),Scale3D=(X=1,Y=1,Z=1))),(Item=(Type=Bone,Name="dyn_l_mouth_tentacle_c_MSTR"),Transform=(Rotation=(X=0,Y=0,Z=0,W=1),Translation=(X=0,Y=0,Z=0),Scale3D=(X=1,Y=1,Z=1))))
    Weight=1
   )
   RigUnit_BeginExecution = &RigUnit_BeginExecution()
   &RigUnit_ModifyTransforms(
    ItemToModify="((Item=(Type=Bone),Transform=(Rotation=(X=0,Y=0,Z=0,W=1),Translation=(X=0,Y=0,Z=0),Scale3D=(X=1,Y=1,Z=1))),(Item=(Type=Bone,Name=\"dyn_r_mouth_tentacle_a_MSTR\"),Transform=(Rotation=(X=0,Y=0,Z=0,W=1),Translation=(X=0,Y=0,Z=0),Scale3D=(X=1,Y=1,Z=1))),(Item=(T ...(truncated)"
    Weight=1
    ItemToModify.2.Transform.Rotation=&RigVMFunction_MathQuaternionFromEuler(Euler=(X=0,Y=-40,Z=25)).Result
    ItemToModify.1.Transform.Rotation=&RigVMFunction_MathQuaternionFromEuler(Euler=(X=0,Y=-35,Z=-25)).Result
    ExecutePin=@RigUnit_BeginExecution.ExecutePin
   )
   > &RigUnit_ModifyTransforms(
    ItemToModify="((Item=(Type=Bone),Transform=(Rotation=(X=0,Y=0,Z=0,W=1),Translation=(X=0,Y=0,Z=0),Scale3D=(X=1,Y=1,Z=1))),(Item=(Type=Bone,Name=\"dyn_r_mouth_tentacle_b_MSTR\"),Transform=(Rotation=(X=0,Y=0,Z=0,W=1),Translation=(X=0,Y=0,Z=0),Scale3D=(X=1,Y=1,Z=1))),(Item=(T ...(truncated)"
    Weight=1
    ItemToModify.1.Transform.Rotation=&RigVMFunction_MathQuaternionFromEuler(Euler=(X=0,Y=0,Z=-20)).Result
    ItemToModify.2.Transform.Rotation=&RigVMFunction_MathQuaternionFromEuler(Euler=(X=0,Y=0,Z=12)).Result
   )
   > &RigUnit_ModifyTransforms(
    ItemToModify="((Item=(Type=Bone),Transform=(Rotation=(X=0,Y=0,Z=0,W=1),Translation=(X=0,Y=0,Z=0),Scale3D=(X=1,Y=1,Z=1))),(Item=(Type=Bone,Name=\"dyn_r_mouth_tentacle_c_MSTR\"),Transform=(Rotation=(X=0,Y=0,Z=0,W=1),Translation=(X=0,Y=0,Z=0),Scale3D=(X=1,Y=1,Z=1))),(Item=(T ...(truncated)"
    Weight=1
    ItemToModify.1.Transform.Rotation=&RigVMFunction_MathQuaternionFromEuler(Euler=(X=5,Y=15,Z=-5)).Result
    ItemToModify.2.Transform.Rotation=&RigVMFunction_MathQuaternionFromEuler(Euler=(X=15,Y=-15,Z=0)).Result
    ItemToModify.3.Transform.Rotation=&RigVMFunction_MathQuaternionFromEuler(Euler=(X=20,Y=-20,Z=0)).Result
    ItemToModify.4.Transform.Rotation=&RigVMFunction_MathQuaternionFromEuler(Euler=(X=0,Y=0,Z=-10)).Result
   )
   n13 = &RigVMFunction_MathFloatRemap(
    SourceMaximum=-14
    SourceMinimum=-6
    TargetMinimum=0
    bClamp=true
    Value=&RigVMFunction_MathQuaternionToEuler(
     RotationOrder=ZYX
     Value=&RigUnit_GetRelativeTransformForItem(Child=(Type=Bone,Name="head"), Parent=(Type=Bone,Name="neck_02")).RelativeTransform.Rotation
    ).Result.X
   )
   > &RigUnit_ModifyTransforms(
    ItemToModify="((Item=(Type=Bone),Transform=(Rotation=(X=0,Y=0,Z=0,W=1),Translation=(X=0,Y=0,Z=0),Scale3D=(X=1,Y=1,Z=1))),(Item=(Type=Bone,Name=\"dyn_l_mouth_tentacle_b_03\"),Transform=(Rotation=(X=0,Y=0,Z=0,W=1),Translation=(X=0,Y=0,Z=0),Scale3D=(X=1,Y=1,Z=1))),(Item=(Typ ...(truncated)"
    Weight=@n13.Result
    ItemToModify.2.Transform.Rotation=&RigVMFunction_MathQuaternionFromEuler(Euler=(X=-10,Y=5,Z=0)).Result
    ItemToModify.1.Transform.Rotation=&RigVMFunction_MathQuaternionFromEuler(Euler=(X=-45,Y=-55,Z=0)).Result
    ItemToModify.3.Transform.Rotation=&RigVMFunction_MathQuaternionFromEuler(Euler=(X=0,Y=-27,Z=0)).Result
    ItemToModify.4.Transform.Rotation=&RigVMFunction_MathQuaternionFromEuler(Euler=(X=20,Y=25,Z=0)).Result
   )
   > &RigUnit_ModifyTransforms(
    ItemToModify="((Item=(Type=Bone),Transform=(Rotation=(X=0,Y=0,Z=0,W=1),Translation=(X=0,Y=0,Z=0),Scale3D=(X=1,Y=1,Z=1))),(Item=(Type=Bone,Name=\"dyn_r_mouth_tentacle_c_MSTR\"),Transform=(Rotation=(X=0,Y=0,Z=0,W=1),Translation=(X=0,Y=0,Z=0),Scale3D=(X=1,Y=1,Z=1))),(Item=(T ...(truncated)"
    Weight=@n13.Result
    ItemToModify.2.Transform.Rotation=&RigVMFunction_MathQuaternionFromEuler(Euler=(X=0,Y=0,Z=15)).Result
    ItemToModify.1.Transform.Rotation=&RigVMFunction_MathQuaternionFromEuler(Euler=(X=0,Y=-5,Z=-10)).Result
    ItemToModify.4.Transform.Rotation=&RigVMFunction_MathQuaternionFromEuler(Euler=(X=12,Y=-10,Z=0)).Result
    ItemToModify.3.Transform.Rotation=&RigVMFunction_MathQuaternionFromEuler(Euler=(X=10,Y=-7,Z=0)).Result
   )
   n14 = &RigUnit_GetRelativeTransformForItem(Child=(Type=Bone,Name="head"), Parent=(Type=Bone,Name="neck_02"))
   > &RigUnit_ModifyTransforms(
    ItemToModify=((Item=(Type=Bone,Name="dyn_r_mouth_tentacle_c_MSTR"),Transform=(Rotation=(X=0,Y=0,Z=0,W=1),Translation=(X=0,Y=0,Z=-0.7),Scale3D=(X=1,Y=1,Z=1))))
    Weight=&RigVMFunction_MathFloatRemap(
     SourceMaximum=25
     SourceMinimum=8
     TargetMinimum=0
     bClamp=true
     Value=&RigVMFunction_MathQuaternionToEuler(RotationOrder=ZYX, Value=@n14.RelativeTransform.Rotation).Result.X
    ).Result
    ItemToModify.0.Transform.Rotation=&RigVMFunction_MathQuaternionFromEuler(Euler=(X=0,Y=10,Z=-10)).Result
   )
   n15 = &RigUnit_GetRelativeTransformForItem(Child=(Type=Bone,Name="head"), Parent=(Type=Bone,Name="neck_02"))
   > &RigUnit_ModifyTransforms(
    ItemToModify="((Item=(Type=Bone,Name=\"dyn_r_mouth_tentacle_b_MSTR\"),Transform=(Rotation=(X=0,Y=0,Z=0,W=1),Translation=(X=0,Y=0,Z=0),Scale3D=(X=1,Y=1,Z=1))),(Item=(Type=Bone,Name=\"dyn_r_mouth_tentacle_b_02\"),Transform=(Rotation=(X=0,Y=0,Z=0,W=1),Translation=(X=0,Y=0,Z=0) ...(truncated)"
    Weight=&RigVMFunction_MathFloatRemap(
     SourceMaximum=35
     SourceMinimum=15
     TargetMinimum=0
     bClamp=true
     Value=&RigVMFunction_MathQuaternionToEuler(RotationOrder=ZYX, Value=@n15.RelativeTransform.Rotation).Result.X
    ).Result
    ItemToModify.0.Transform.Rotation=&RigVMFunction_MathQuaternionFromEuler(Euler=(X=0,Y=15,Z=-15)).Result
    ItemToModify.1.Transform.Rotation=&RigVMFunction_MathQuaternionFromEuler(Euler=(X=0,Y=25,Z=-25)).Result
   )
   n16 = &RigUnit_GetRelativeTransformForItem(Child=(Type=Bone,Name="head"), Parent=(Type=Bone,Name="neck_02"))
   > &RigUnit_ModifyTransforms(
    ItemToModify=((Item=(Type=Bone,Name="dyn_r_mouth_tentacle_a_MSTR"),Transform=(Rotation=(X=0,Y=0,Z=0,W=1),Translation=(X=0,Y=0,Z=0),Scale3D=(X=1,Y=1,Z=1))))
    Weight=&RigVMFunction_MathFloatRemap(
     SourceMaximum=30
     SourceMinimum=10
     TargetMinimum=0
     bClamp=true
     Value=&RigVMFunction_MathQuaternionToEuler(RotationOrder=ZYX, Value=@n16.RelativeTransform.Rotation).Result.X
    ).Result
    ItemToModify.0.Transform.Rotation=&RigVMFunction_MathQuaternionFromEuler(Euler=(X=0,Y=-35,Z=-25)).Result
   )
   > /Script/ControlRig.RigUnit_ChainHarmonicsPerItem(
    ChainRoot=(Type=Bone,Name="dyn_l_mouth_tentacle_a_01")
    DrawWorldOffset=(Rotation=(X=0,Y=0,Z=0,W=1),Translation=(X=0,Y=0,Z=0),Scale3D=(X=1,Y=1,Z=1))
    Pendulum=(bEnabled=True,PendulumStiffness=2,PendulumGravity=(X=0,Y=0,Z=0),PendulumBlend=0.75,PendulumDrag=0.98,PendulumMaximum=0.1,UnwindAxis=(X=0,Y=1,Z=0),UnwindMinimum=0.2,UnwindMaximum=0.05)
    Reach=(bEnabled=True,ReachTarget=(X=0,Y=0,Z=0),ReachAxis=(X=1,Y=0,Z=0))
    Speed=(X=1.7,Y=0.9,Z=0.6)
    Wave=(bEnabled=True,WaveFrequency=(X=0.75,Y=0.3,Z=0.2),WaveAmplitude=(X=0.4,Y=0.75,Z=0.3),WaveOffset=(X=1.5,Y=0.6,Z=1.7),WaveNoise=(X=0.8,Y=0.3,Z=0.5),WaveMaximum=1,WaveEase=BounceEaseInOut)
    WaveCurve=(EditorCurveData=(Keys=((),(Time=1,Value=1)),DefaultValue=340282346638528859811704183484516925440,PreInfinityExtrap=RCCE_Constant,PostInfinityExtrap=RCCE_Constant))
    bDrawDebug=True
   ) {is=struct RigVMUnitNode}
   > &RigUnit_ChainHarmonicsPerItem(
    ChainRoot=(Type=Bone,Name="dyn_r_mouth_tentacle_a_01")
    DrawWorldOffset=(Rotation=(X=0,Y=0,Z=0,W=1),Translation=(X=0,Y=0,Z=0),Scale3D=(X=1,Y=1,Z=1))
    Pendulum=(bEnabled=True,PendulumStiffness=2,PendulumGravity=(X=0,Y=0,Z=0),PendulumBlend=0.75,PendulumDrag=0.98,PendulumMaximum=0.1,UnwindAxis=(X=0,Y=1,Z=0),UnwindMinimum=0.2,UnwindMaximum=0.05)
    Reach=(bEnabled=True,ReachTarget=(X=0,Y=0,Z=0),ReachAxis=(X=1,Y=0,Z=0))
    Speed=(X=1.2,Y=1.1,Z=0.4)
    Wave=(bEnabled=True,WaveFrequency=(X=0.4,Y=0.5,Z=0.3),WaveAmplitude=(X=0.4,Y=0.5,Z=0.5),WaveOffset=(X=1.5,Y=0.6,Z=1.7),WaveNoise=(X=0.6,Y=0.2,Z=0.7),WaveMaximum=1,WaveEase=BounceEaseInOut)
    WaveCurve=(EditorCurveData=(Keys=((),(Time=1,Value=1)),DefaultValue=340282346638528859811704183484516925440,PreInfinityExtrap=RCCE_Constant,PostInfinityExtrap=RCCE_Constant))
    bDrawDebug=True
   )
   > &RigUnit_ChainHarmonicsPerItem(
    ChainRoot=(Type=Bone,Name="dyn_l_mouth_tentacle_b_01")
    DrawWorldOffset=(Rotation=(X=0,Y=0,Z=0,W=1),Translation=(X=0,Y=0,Z=0),Scale3D=(X=1,Y=1,Z=1))
    Pendulum=(bEnabled=True,PendulumStiffness=2,PendulumGravity=(X=0,Y=0,Z=0),PendulumBlend=0.75,PendulumDrag=0.98,PendulumMaximum=0.1,UnwindAxis=(X=0,Y=1,Z=0),UnwindMinimum=0.2,UnwindMaximum=0.05)
    Reach=(bEnabled=True,ReachTarget=(X=0,Y=0,Z=0),ReachAxis=(X=1,Y=0,Z=0))
    Speed=(X=1.5,Y=0.8,Z=0.9)
    Wave=(bEnabled=True,WaveFrequency=(X=0.75,Y=0.3,Z=0.2),WaveAmplitude=(X=0.4,Y=0.75,Z=0.3),WaveOffset=(X=0.6,Y=1.2,Z=1),WaveNoise=(X=0.5,Y=0.6,Z=0.9),WaveMaximum=0.6,WaveEase=BounceEaseInOut)
    WaveCurve=(EditorCurveData=(Keys=((),(Time=1,Value=1)),DefaultValue=340282346638528859811704183484516925440,PreInfinityExtrap=RCCE_Constant,PostInfinityExtrap=RCCE_Constant))
    bDrawDebug=True
   )
   > &RigUnit_ChainHarmonicsPerItem(
    ChainRoot=(Type=Bone,Name="dyn_r_mouth_tentacle_b_01")
    DrawWorldOffset=(Rotation=(X=0,Y=0,Z=0,W=1),Translation=(X=0,Y=0,Z=0),Scale3D=(X=1,Y=1,Z=1))
    Pendulum=(bEnabled=True,PendulumStiffness=2,PendulumGravity=(X=0,Y=0,Z=0),PendulumBlend=0.75,PendulumDrag=0.98,PendulumMaximum=0.1,UnwindAxis=(X=0,Y=1,Z=0),UnwindMinimum=0.2,UnwindMaximum=0.05)
    Reach=(bEnabled=True,ReachTarget=(X=0,Y=0,Z=0),ReachAxis=(X=1,Y=0,Z=0))
    Speed=(X=1.7,Y=0.9,Z=0.3)
    Wave=(bEnabled=True,WaveFrequency=(X=0.45,Y=0.3,Z=0.2),WaveAmplitude=(X=0.2,Y=0.25,Z=0.15),WaveOffset=(X=1.5,Y=0.6,Z=0.75),WaveNoise=(X=0.4,Y=0.3,Z=0.25),WaveMaximum=1,WaveEase=BounceEaseInOut)
    WaveCurve=(EditorCurveData=(Keys=((),(Time=1,Value=1)),DefaultValue=340282346638528859811704183484516925440,PreInfinityExtrap=RCCE_Constant,PostInfinityExtrap=RCCE_Constant))
    bDrawDebug=True
   )
   > &RigUnit_ChainHarmonicsPerItem(
    ChainRoot=(Type=Bone,Name="dyn_l_mouth_tentacle_c_01")
    DrawWorldOffset=(Rotation=(X=0,Y=0,Z=0,W=1),Translation=(X=0,Y=0,Z=0),Scale3D=(X=1,Y=1,Z=1))
    Pendulum=(bEnabled=True,PendulumStiffness=2,PendulumGravity=(X=0,Y=0,Z=0),PendulumBlend=0.75,PendulumDrag=0.98,PendulumMaximum=0.1,UnwindAxis=(X=0,Y=1,Z=0),UnwindMinimum=0.2,UnwindMaximum=0.05)
    Reach=(bEnabled=True,ReachTarget=(X=0,Y=0,Z=0),ReachAxis=(X=1,Y=0,Z=0))
    Speed=(X=1.7,Y=0.9,Z=0.6)
    Wave=(bEnabled=True,WaveFrequency=(X=0.75,Y=0.3,Z=0.2),WaveAmplitude=(X=0.4,Y=0.75,Z=0.3),WaveOffset=(X=1.5,Y=0.6,Z=1.7),WaveNoise=(X=0.8,Y=0.3,Z=0.5),WaveMaximum=0.8,WaveEase=BounceEaseInOut)
    WaveCurve=(EditorCurveData=(Keys=((),(Time=1,Value=1)),DefaultValue=340282346638528859811704183484516925440,PreInfinityExtrap=RCCE_Constant,PostInfinityExtrap=RCCE_Constant))
    bDrawDebug=True
   )
   > &RigUnit_ChainHarmonicsPerItem(
    ChainRoot=(Type=Bone,Name="dyn_r_mouth_tentacle_c_01")
    DrawWorldOffset=(Rotation=(X=0,Y=0,Z=0,W=1),Translation=(X=0,Y=0,Z=0),Scale3D=(X=1,Y=1,Z=1))
    Pendulum=(bEnabled=True,PendulumStiffness=2,PendulumGravity=(X=0,Y=0,Z=0),PendulumBlend=0.75,PendulumDrag=0.98,PendulumMaximum=0.1,UnwindAxis=(X=0,Y=1,Z=0),UnwindMinimum=0.2,UnwindMaximum=0.05)
    Reach=(bEnabled=True,ReachTarget=(X=0,Y=0,Z=0),ReachAxis=(X=1,Y=0,Z=0))
    Speed=(X=0.75,Y=1.1,Z=0.75)
    Wave=(bEnabled=True,WaveFrequency=(X=0.2,Y=3,Z=0.2),WaveAmplitude=(X=0.2,Y=0.35,Z=0.25),WaveOffset=(X=1,Y=1,Z=0.5),WaveNoise=(X=0.2,Y=0.3,Z=0.5),WaveMaximum=0.5,WaveEase=BounceEaseInOut)
    WaveCurve=(EditorCurveData=(Keys=((),(Time=1,Value=1)),DefaultValue=340282346638528859811704183484516925440,PreInfinityExtrap=RCCE_Constant,PostInfinityExtrap=RCCE_Constant))
    bDrawDebug=True
   )
   &RigVMFunction_MathFloatRemap(
    SourceMaximum=0
    SourceMinimum=-15
    TargetMaximum=0.6
    TargetMinimum=1
    bClamp=true
    Value=&RigVMFunction_MathQuaternionToEuler(
     RotationOrder=ZYX
     Value=&RigUnit_GetRelativeTransformForItem(Child=(Type=Bone,Name="head"), Parent=(Type=Bone,Name="neck_02")).RelativeTransform.Rotation
    ).Result.Z
   )
   n17 = &RigVMFunction_MathFloatRemap(
    SourceMaximum=0
    SourceMinimum=-10
    TargetMaximum=-10
    TargetMinimum=0
    bClamp=true
    Value=&RigVMFunction_MathQuaternionToEuler(
     RotationOrder=ZYX
     Value=&RigUnit_GetRelativeTransformForItem(Child=(Type=Bone,Name="head"), Parent=(Type=Bone,Name="neck_02")).RelativeTransform.Rotation
    ).Result.Z
   )
   &RigVMFunction_MathQuaternionFromEuler(Euler=(X=0,Y=0,Z=0), Euler.Y=@n17.Result, Euler.Z=@n17.Result)
  ]}
]
hierarchy [
 root {loc=<0 0 0>, rot=<0 0 0 1>}
 pelvis {parent=root, loc=<0 -0.7071 93.1191>, rot=<0.0127 0.707 -0.0127 -0.707>}
 spine_01 {parent=pelvis, loc=<6.7494 -0.6388 0>, rot=<0 0 0.034 -0.9994>}
 spine_02 {parent=spine_01, loc=<7.2518 -0.035 0>, rot=<0 0 0.0238 -0.9997>}
 spine_03 {parent=spine_02, loc=<6.0104 0 0>, rot=<0 0 0.0149 -0.9999>}
 spine_04 {parent=spine_03, loc=<8.1024 0 0>, rot=<0 0 -0.0132 -0.9999>}
 spine_05 {parent=spine_04, loc=<10.0146 0 0>, rot=<0 0 -0.0015 -1>}
 neck_01 {parent=spine_05, loc=<11.3138 0.1529 0>, rot=<0 0 -0.2003 -0.9797>}
 neck_02 {parent=neck_01, loc=<4.5627 0 0>, rot=<0 0 0.0504 -0.9987>}
 head {parent=neck_02, loc=<5.3914 0 0>, rot=<0 0 0.115 -0.9934>}
 faceAttach {parent=head, loc=<-2.737 0.5477 -0.066>, rot=<0.0082 -0.7071 -0.0082 -0.7071>, scale=<0.89 0.89 0.89>}
 L_brow_outer {parent=faceAttach, loc=<6.3394 6.6402 13.5269>, rot=<-0.0225 -0.0322 0.4479 -0.8932>}
 L_cheek_inner {parent=faceAttach, loc=<4.6628 7.8637 4.2323>, rot=<-0.0674 -0.2616 0.3195 -0.9083>}
 C_lip_upper_mid {parent=faceAttach, loc=<-0.0591 9.2002 1.4688>, rot=<0.2225 0.0013 -0.0005 -0.9749>}
 L_lip_upper_outer {parent=faceAttach, loc=<3.2966 8.3483 -1.281>, rot=<0.0359 -0.0728 0.2627 -0.9615>}
 dyn_l_mouth_tentacle_a_MSTR {parent=L_lip_upper_outer, loc=<-0.2319 0.3884 0.0206>, rot=<0.1725 -0.3287 -0.4098 -0.8332>, scale=<1.1236 1.1236 1.1236>}
 dyn_l_mouth_tentacle_a_01 {parent=dyn_l_mouth_tentacle_a_MSTR, loc=<0 0 0>, rot=<0 0 0 -1>}
 dyn_l_mouth_tentacle_a_02 {parent=dyn_l_mouth_tentacle_a_01, loc=<3.5427 0 0>, rot=<-0.1051 0.4236 0.2042 -0.8763>}
 dyn_l_mouth_tentacle_a_03 {parent=dyn_l_mouth_tentacle_a_02, loc=<4.0705 0 0>, rot=<0.2447 0.2859 0.6271 -0.682>}
 dyn_l_mouth_tentacle_a_04 {parent=dyn_l_mouth_tentacle_a_03, loc=<3.455 0 0>, rot=<0.1567 -0.3708 0.6764 -0.6168>}
 L_lip_corner {parent=faceAttach, loc=<5.3232 4.1001 0.4669>, rot=<0.0029 0.008 -0.0146 -0.9999>}
 dyn_l_mouth_tentacle_c_MSTR {parent=L_lip_corner, loc=<0.2075 -0.0756 -0.7814>, rot=<0.089 -0.605 -0.1351 -0.7796>, scale=<1.1236 1.1236 1.1236>}
 dyn_l_mouth_tentacle_c_01 {parent=dyn_l_mouth_tentacle_c_MSTR, loc=<0 0 0>, rot=<0 0 0 -1>}
 dyn_l_mouth_tentacle_c_02 {parent=dyn_l_mouth_tentacle_c_01, loc=<2.4033 0 0>, rot=<0.2469 0.0796 -0.0889 -0.9617>}
 dyn_l_mouth_tentacle_c_03 {parent=dyn_l_mouth_tentacle_c_02, loc=<3.5398 0 0>, rot=<0.4753 0.282 -0.5336 -0.6402>}
 dyn_l_mouth_tentacle_c_04 {parent=dyn_l_mouth_tentacle_c_03, loc=<3.2297 0 0>, rot=<-0.0084 0.2873 -0.4664 -0.8366>}
 C_nose_bridge {parent=faceAttach, loc=<-0.0941 10.1759 8.1302>, rot=<-0.2218 -0.0001 0 -0.9751>}
 R_brow_outer {parent=faceAttach, loc=<-6.4563 6.649 13.5081>, rot=<-0.0281 0.0393 -0.449 -0.8922>}
 R_cheek_inner {parent=faceAttach, loc=<-4.7989 7.8699 4.2217>, rot=<-0.0676 0.2621 -0.3194 -0.9081>}
 R_lip_corner {parent=faceAttach, loc=<-5.4542 4.6934 0.4647>, rot=<0.0029 -0.008 0.0146 -0.9999>}
 dyn_r_mouth_tentacle_c_MSTR {parent=R_lip_corner, loc=<-0.2916 -0.666 -0.3008>, rot=<-0.6352 0.0295 0.7699 0.0535>, scale=<1.1236 1.1236 1.1236>}
 dyn_r_mouth_tentacle_c_01 {parent=dyn_r_mouth_tentacle_c_MSTR, loc=<0 0 0>, rot=<0 0 0 1>}
 dyn_r_mouth_tentacle_c_02 {parent=dyn_r_mouth_tentacle_c_01, loc=<2.482 0 0>, rot=<0.8267 -0.067 -0.2983 0.4723>}
 dyn_r_mouth_tentacle_c_03 {parent=dyn_r_mouth_tentacle_c_02, loc=<3.3479 0 0>, rot=<-0.0862 0.3837 0.1053 -0.9134>}
 dyn_r_mouth_tentacle_c_04 {parent=dyn_r_mouth_tentacle_c_03, loc=<4.6564 0 0>, rot=<0.4129 -0.2044 -0.6097 -0.6449>}
 dyn_r_mouth_tentacle_c_05 {parent=dyn_r_mouth_tentacle_c_04, loc=<2.7301 0 0>, rot=<0.2436 0.3763 -0.6019 -0.6609>}
 C_jaw {parent=faceAttach, loc=<-0.1795 2.9678 1.3919>, rot=<0 0 0 1>}
 L_lip_lower_outer {parent=C_jaw, loc=<4.6046 3.4476 -2.4342>, rot=<0.013 0.0445 0.3057 -0.951>}
 dyn_l_mouth_tentacle_b_MSTR {parent=L_lip_lower_outer, loc=<-0.2621 0.368 0.0302>, rot=<0.2723 -0.5491 -0.4168 -0.6713>, scale=<1.1236 1.1236 1.1236>}
 dyn_l_mouth_tentacle_b_01 {parent=dyn_l_mouth_tentacle_b_MSTR, loc=<0 0 0>, rot=<0 0 0 -1>}
 dyn_l_mouth_tentacle_b_02 {parent=dyn_l_mouth_tentacle_b_01, loc=<2.7922 0 0>, rot=<0.1357 0.2554 -0.0845 -0.9535>}
 dyn_l_mouth_tentacle_b_03 {parent=dyn_l_mouth_tentacle_b_02, loc=<4.1341 0 0>, rot=<0.3011 0.1243 -0.4354 -0.8392>}
 dyn_l_mouth_tentacle_b_04 {parent=dyn_l_mouth_tentacle_b_03, loc=<4.6648 0 0>, rot=<-0.0828 0.3786 -0.7117 -0.5859>}
 C_lip_lower_mid {parent=C_jaw, loc=<0.1017 5.5043 -2.7562>, rot=<0 0 0 -1>}
 R_lip_lower_outer {parent=C_jaw, loc=<-4.394 3.4476 -2.4342>, rot=<0.013 -0.0445 -0.3057 -0.951>}
 dyn_r_mouth_tentacle_b_MSTR {parent=R_lip_lower_outer, loc=<0.2621 0.368 0.0302>, rot=<-0.5491 0.2723 0.6713 0.4168>, scale=<1.1236 1.1236 1.1236>}
 dyn_r_mouth_tentacle_b_01 {parent=dyn_r_mouth_tentacle_b_MSTR, loc=<0 0 0>, rot=<0 0 0 -1>}
 dyn_r_mouth_tentacle_b_02 {parent=dyn_r_mouth_tentacle_b_01, loc=<2.7922 0 0>, rot=<-0.1357 0.2554 0.0845 -0.9535>}
 dyn_r_mouth_tentacle_b_03 {parent=dyn_r_mouth_tentacle_b_02, loc=<4.1341 0 0>, rot=<-0.2838 0.1141 0.3969 -0.8654>}
 dyn_r_mouth_tentacle_b_04 {parent=dyn_r_mouth_tentacle_b_03, loc=<4.3838 0 0>, rot=<0.0482 0.372 0.7095 -0.5965>}
 tongue {parent=C_jaw, loc=<0.106 3.478 -0.7518>, rot=<-0.1568 0 0 -0.9876>}
 dyn_r_mouth_tongue_MSTR {parent=tongue, loc=<-1.1716 2.8181 -1.3046>, rot=<0.5439 -0.5933 -0.4303 -0.4087>, scale=<1.1236 1.1236 1.1236>}
 dyn_r_mouth_tongue {parent=dyn_r_mouth_tongue_MSTR, loc=<0 0 0>, rot=<0 0 0 -1>}
 dyn_r_mouth_tongue_01 {parent=dyn_r_mouth_tongue, loc=<3.1937 0 0>, rot=<0.0231 -0.2268 0.0435 -0.9727>}
 dyn_l_mouth_tongue_MSTR {parent=tongue, loc=<1.1703 2.8181 -1.3046>, rot=<0.5933 -0.5439 -0.4087 -0.4303>, scale=<1.1236 1.1236 1.1236>}
 dyn_l_mouth_tongue {parent=dyn_l_mouth_tongue_MSTR, loc=<0 0 0>, rot=<0 0 0 -1>}
 dyn_l_mouth_tongue_01 {parent=dyn_l_mouth_tongue, loc=<3.1937 0 0>, rot=<-0.0263 -0.2208 -0.044 -0.974>}
 R_eye {parent=faceAttach, loc=<-0.9684 5.1828 11.3153>, rot=<0 0 0.9998 -0.0187>}
 R_eye_lid_lower_mid {parent=R_eye, loc=<-0.0085 -0.9422 -0.7046>, rot=<0.0329 -0.0005 0.0167 -0.9993>}
 L_eye {parent=faceAttach, loc=<0.8201 5.1828 11.3153>, rot=<0 0 0.9998 0.0187>}
 L_eye_lid_lower_mid {parent=L_eye, loc=<0.0223 -0.923 -0.7488>, rot=<0.0329 0 0 -0.9995>}
 R_lip_upper_outer {parent=faceAttach, loc=<-3.4449 8.3483 -1.281>, rot=<0.0359 0.0728 -0.2627 -0.9615>}
 dyn_r_mouth_tentacle_a_MSTR {parent=R_lip_upper_outer, loc=<0.2319 0.3884 0.0206>, rot=<-0.3525 0.1949 0.8149 0.4168>, scale=<1.1236 1.1236 1.1236>}
 dyn_r_mouth_tentacle_a_01 {parent=dyn_r_mouth_tentacle_a_MSTR, loc=<0 0 0>, rot=<0 0 0 -1>}
 dyn_r_mouth_tentacle_a_02 {parent=dyn_r_mouth_tentacle_a_01, loc=<2.9378 0 0>, rot=<0.108 0.388 -0.1626 -0.9007>}
 dyn_r_mouth_tentacle_a_03 {parent=dyn_r_mouth_tentacle_a_02, loc=<4.1242 0 0>, rot=<-0.1692 0.3612 -0.5726 -0.7163>}
 dyn_r_mouth_tentacle_a_04 {parent=dyn_r_mouth_tentacle_a_03, loc=<3.6815 0 0>, rot=<-0.2038 -0.2987 -0.7414 -0.5653>}
 C_glabella_down_pose {type=Curve, loc=<0 0 0>, rot=<0 0 0 1>}
 tongue_up_pose {type=Curve, loc=<0 0 0>, rot=<0 0 0 1>}
 phoneme_mbp_pose {type=Curve, loc=<0 0 0>, rot=<0 0 0 1>}
 L_upper_lip_lower_pose {type=Curve, loc=<0 0 0>, rot=<0 0 0 1>}
 R_blink_pose {type=Curve, loc=<0 0 0>, rot=<0 0 0 1>}
 R_wide_pose {type=Curve, loc=<0 0 0>, rot=<0 0 0 1>}
 look_right_pose {type=Curve, loc=<0 0 0>, rot=<0 0 0 1>}
 jaw_up_pose {type=Curve, loc=<0 0 0>, rot=<0 0 0 1>}
 L_squeeze_pose {type=Curve, loc=<0 0 0>, rot=<0 0 0 1>}
 R_brow_up_pose {type=Curve, loc=<0 0 0>, rot=<0 0 0 1>}
 look_left_pose {type=Curve, loc=<0 0 0>, rot=<0 0 0 1>}
 tongue_back_pose {type=Curve, loc=<0 0 0>, rot=<0 0 0 1>}
 R_squint_inner_pose {type=Curve, loc=<0 0 0>, rot=<0 0 0 1>}
 look_up_pose {type=Curve, loc=<0 0 0>, rot=<0 0 0 1>}
 phoneme_fv_pose {type=Curve, loc=<0 0 0>, rot=<0 0 0 1>}
 jaw_back_pose {type=Curve, loc=<0 0 0>, rot=<0 0 0 1>}
 R_smile_pose {type=Curve, loc=<0 0 0>, rot=<0 0 0 1>}
 R_lip_corner_narrow_pose {type=Curve, loc=<0 0 0>, rot=<0 0 0 1>}
 eye_convergence_pose {type=Curve, loc=<0 0 0>, rot=<0 0 0 1>}
 R_brow_down_pose {type=Curve, loc=<0 0 0>, rot=<0 0 0 1>}
 R_frown_pose {type=Curve, loc=<0 0 0>, rot=<0 0 0 1>}
 L_lower_lip_down_pose {type=Curve, loc=<0 0 0>, rot=<0 0 0 1>}
 L_blink_pose {type=Curve, loc=<0 0 0>, rot=<0 0 0 1>}
 L_lip_corner_narrow_pose {type=Curve, loc=<0 0 0>, rot=<0 0 0 1>}
 look_down_pose {type=Curve, loc=<0 0 0>, rot=<0 0 0 1>}
 R_upper_lip_raiser_pose {type=Curve, loc=<0 0 0>, rot=<0 0 0 1>}
 phoneme_oo_pose {type=Curve, loc=<0 0 0>, rot=<0 0 0 1>}
 R_brow_squeeze_in_pose {type=Curve, loc=<0 0 0>, rot=<0 0 0 1>}
 jaw_left_pose {type=Curve, loc=<0 0 0>, rot=<0 0 0 1>}
 L_lip_corner_wide_pose {type=Curve, loc=<0 0 0>, rot=<0 0 0 1>}
 C_nose_wrinkler_pose {type=Curve, loc=<0 0 0>, rot=<0 0 0 1>}
 R_lower_lip_up_pose {type=Curve, loc=<0 0 0>, rot=<0 0 0 1>}
 jaw_open_pose {type=Curve, loc=<0 0 0>, rot=<0 0 0 1>}
 L_brow_up_pose {type=Curve, loc=<0 0 0>, rot=<0 0 0 1>}
 R_lip_corner_wide_pose {type=Curve, loc=<0 0 0>, rot=<0 0 0 1>}
 L_lower_lip_up_pose {type=Curve, loc=<0 0 0>, rot=<0 0 0 1>}
 R_lower_lip_down_pose {type=Curve, loc=<0 0 0>, rot=<0 0 0 1>}
 L_upper_lip_raiser_pose {type=Curve, loc=<0 0 0>, rot=<0 0 0 1>}
 R_cheek_up_pose {type=Curve, loc=<0 0 0>, rot=<0 0 0 1>}
 L_brow_down_pose {type=Curve, loc=<0 0 0>, rot=<0 0 0 1>}
 jaw_forward_pose {type=Curve, loc=<0 0 0>, rot=<0 0 0 1>}
 L_squint_inner_pose {type=Curve, loc=<0 0 0>, rot=<0 0 0 1>}
 tongue_down_pose {type=Curve, loc=<0 0 0>, rot=<0 0 0 1>}
 R_upper_lip_lower_pose {type=Curve, loc=<0 0 0>, rot=<0 0 0 1>}
 L_brow_squeeze_in_pose {type=Curve, loc=<0 0 0>, rot=<0 0 0 1>}
 jaw_right_pose {type=Curve, loc=<0 0 0>, rot=<0 0 0 1>}
 mouth_down_pose {type=Curve, loc=<0 0 0>, rot=<0 0 0 1>}
 C_glabella_up_pose {type=Curve, loc=<0 0 0>, rot=<0 0 0 1>}
 phoneme_ch_pose {type=Curve, loc=<0 0 0>, rot=<0 0 0 1>}
 L_frown_pose {type=Curve, loc=<0 0 0>, rot=<0 0 0 1>}
 R_squeeze_pose {type=Curve, loc=<0 0 0>, rot=<0 0 0 1>}
 tongue_forward_pose {type=Curve, loc=<0 0 0>, rot=<0 0 0 1>}
 L_upperLidDown_pose {type=Curve, loc=<0 0 0>, rot=<0 0 0 1>}
 R_upperLidDown_pose {type=Curve, loc=<0 0 0>, rot=<0 0 0 1>}
 L_wide_pose {type=Curve, loc=<0 0 0>, rot=<0 0 0 1>}
 L_smile_pose {type=Curve, loc=<0 0 0>, rot=<0 0 0 1>}
 mouth_up_pose {type=Curve, loc=<0 0 0>, rot=<0 0 0 1>}
 L_cheek_up_pose {type=Curve, loc=<0 0 0>, rot=<0 0 0 1>}
 base_pose {type=Curve, loc=<0 0 0>, rot=<0 0 0 1>}
 blendParent1 {type=Curve, loc=<0 0 0>, rot=<0 0 0 1>}
]
