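# Physics asset for the tentacled Alien Trooper head cosmetic on the male medium player, covering the dynamic mouth tentacle and tongue bones.
# Left and right tentacle chains (a, b, c) and two tongues each link one or two capsules to a small MSTR root capsule.
# Constraints set moderate swing and very tight twist limits.
# Large clavicle and upper-spine capsules have no constraints, apparently serving as body collision for the tentacles.
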PhysicsAsset /BRCosmetics/Characters/Player/Male/Medium/Heads/M_MED_AlienTrooper_Head/Meshes/Tentacle/Male_Medium_Alien_Trooper_Tentacle_Head_Physics
bodies {
 dyn_l_mouth_tentacle_a_01 [{shapeType=Capsule, length=2.0957, radius=1.3696, center=<3 0 0>, rotation=<rot 90 0 0>}]
 dyn_l_mouth_tentacle_a_MSTR [{shapeType=Capsule, length=0.505, radius=0.505, rotation=<rot 0 -90 0>}]
 dyn_l_mouth_tentacle_b_02 [{shapeType=Capsule, length=1.4017, radius=1.7527, center=<1.9691 0 0>, rotation=<rot 90 0 0>}]
 dyn_l_mouth_tentacle_b_03 [{shapeType=Capsule, length=1.2271, radius=1.8496, center=<2.3194 0 0>, rotation=<rot 90 0 0>}]
 dyn_l_mouth_tentacle_b_MSTR [{shapeType=Capsule, length=0.505, radius=0.505, rotation=<rot 0 -90 0>}]
 dyn_l_mouth_tentacle_c_01 [{shapeType=Capsule, length=1.2963, radius=1.2281, center=<3 0 0>, rotation=<rot 90 0 0>}]
 dyn_l_mouth_tentacle_c_MSTR [{shapeType=Capsule, length=0.505, radius=0.505, rotation=<rot 0 -90 0>}]
 dyn_l_mouth_tongue [{shapeType=Capsule, length=0.4715, radius=0.9802, center=<3 0 0>, rotation=<rot 90 0 0>}]
 dyn_l_mouth_tongue_MSTR [{shapeType=Capsule, length=0.505, radius=0.505}]
 dyn_r_mouth_tentacle_a_01 [{shapeType=Capsule, length=1.4421, radius=1.3158, center=<3 0 0>, rotation=<rot 90 0 0>}]
 dyn_r_mouth_tentacle_a_MSTR [{shapeType=Capsule, length=0.505, radius=0.505, rotation=<rot 0 -90 0>}]
 dyn_r_mouth_tentacle_b_02 [{shapeType=Capsule, length=1.96, radius=1.7517, center=<2.2747 0 0>, rotation=<rot 90 0 0>}]
 dyn_r_mouth_tentacle_b_03 [{shapeType=Capsule, length=1.46, radius=1.5527, center=<2.8343 0 0>, rotation=<rot 90 0 0>}]
 dyn_r_mouth_tentacle_b_MSTR [{shapeType=Capsule, length=0.505, radius=0.505, rotation=<rot 0 -90 0>}]
 dyn_r_mouth_tentacle_c_02 [{shapeType=Capsule, length=1.6509, radius=1.2886, center=<2.5 0 0>, rotation=<rot 90 0 0>}]
 dyn_r_mouth_tentacle_c_03 [{shapeType=Capsule, length=1.9992, radius=1.1967, center=<3 0 0>, rotation=<rot 90 0 0>}]
 dyn_r_mouth_tentacle_c_MSTR [{shapeType=Capsule, length=0.505, radius=0.505, rotation=<rot 0 -90 0>}]
 dyn_r_mouth_tongue [{shapeType=Capsule, length=0.3062, radius=0.9523, center=<3 0 0>, rotation=<rot 90 0 0>}]
 dyn_r_mouth_tongue_MSTR [{shapeType=Capsule, length=0.505, radius=0.505}]
 dyn_tentacle_clavicle_l [{shapeType=Capsule, length=14.5976, radius=10.2338, center=<8 0 1.5>, rotation=<rot 0 0 -60.2429>}]
 dyn_tentacle_clavicle_r [{shapeType=Capsule, length=14.5976, radius=10.2338, center=<-8 0 -1.5>, rotation=<rot 0 0 -60.2433>}]
 dyn_tentacle_spine_05 [{shapeType=Capsule, length=6.3959, radius=18.1764, center=<-3.5 0 0>, rotation=<rot 90 -7.125 56.3265>}]
}
constraints [
 {
  bone1Name=dyn_l_mouth_tentacle_a_01
  bone2Name=dyn_l_mouth_tentacle_a_MSTR
  swing1LimitDegrees=15
  swing2LimitDegrees=20
  twistLimitDegrees=3
 }
 {
  bone1Name=dyn_l_mouth_tentacle_c_01
  bone2Name=dyn_l_mouth_tentacle_c_MSTR
  swing1LimitDegrees=20
  swing2LimitDegrees=35
  twistLimitDegrees=3
 }
 {bone1Name=dyn_r_mouth_tentacle_c_03, bone2Name=dyn_r_mouth_tentacle_c_02, swing1LimitDegrees=10, swing2LimitDegrees=15, twistLimitDegrees=3}
 {
  bone1Name=dyn_r_mouth_tentacle_c_02
  bone2Name=dyn_r_mouth_tentacle_c_MSTR
  swing1LimitDegrees=10
  swing2LimitDegrees=15
  twistLimitDegrees=3
 }
 {bone1Name=dyn_r_mouth_tongue, bone2Name=dyn_r_mouth_tongue_MSTR, swing1LimitDegrees=10, swing2LimitDegrees=25, twistLimitDegrees=3}
 {bone1Name=dyn_l_mouth_tongue, bone2Name=dyn_l_mouth_tongue_MSTR, swing1LimitDegrees=10, swing2LimitDegrees=25, twistLimitDegrees=3}
 {
  bone1Name=dyn_r_mouth_tentacle_a_01
  bone2Name=dyn_r_mouth_tentacle_a_MSTR
  swing1LimitDegrees=15
  swing2LimitDegrees=20
  twistLimitDegrees=3
 }
 {
  bone1Name=dyn_r_mouth_tentacle_b_02
  bone2Name=dyn_r_mouth_tentacle_b_MSTR
  swing1LimitDegrees=20
  swing2LimitDegrees=35
  twistLimitDegrees=1
 }
 {bone1Name=dyn_r_mouth_tentacle_b_03, bone2Name=dyn_r_mouth_tentacle_b_02, swing1LimitDegrees=20, swing2LimitDegrees=35, twistLimitDegrees=1}
 {
  bone1Name=dyn_l_mouth_tentacle_b_02
  bone2Name=dyn_l_mouth_tentacle_b_MSTR
  swing1LimitDegrees=20
  swing2LimitDegrees=35
  twistLimitDegrees=1
 }
 {bone1Name=dyn_l_mouth_tentacle_b_03, bone2Name=dyn_l_mouth_tentacle_b_02, swing1LimitDegrees=20, swing2LimitDegrees=35, twistLimitDegrees=1}
]
