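# Animation Blueprint for the male medium Alien Trooper head, serving Shark, Tentacle, Brain, Owl and Cyclops variants.
# Copied mesh poses take montage slots and curve-gated facial animation from faceAttach down.
# A pose asset feeds chained Control Rigs, RigidBody and AnimDynamics for secondary motion, with tentacle setups switched for front end.
# Weighted constraints drive spike bones, and a player-list flag selects a pose with the head shrunk.
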AnimBlueprint /BRCosmetics/Characters/Player/Male/Medium/Heads/M_MED_AlienTrooper_Head/Meshes/Human/Male_Medium_Alien_Trooper_Base_Head_AnimBp
target_skeleton=&/Male_Medium_Alien_Trooper_Head_Skeleton
graphs [
 AnimGraph {nodes [
   n1 = Anim Dynamics - Bone: dyn_mohawk_front_01(Alpha=1, ExternalForce="0,0,0") {
    angularDampingOverride=0.9
    angularSpringConstant=3000
    bAngularSpring=true
    bGravityOverrideInSimSpace=true
    bOverrideAngularDamping=true
    bUseGravityOverride=true
    simulationSpace=RootRelative
    alphaBoolBlend {blendOption=Linear}
    boundBone {boneName=dyn_mohawk_front_01}
    componentAppliedLinearAccClamp=<1500 1500 1500>
    componentLinearAccScale=<0.3 0.3 0.4>
    componentLinearVelScale=<0.2 0.2 0.3>
    gravityOverride {z=-980}
    physicsBodyDefinitions [{
      sphereCollisionRadius=0
      boundBone {boneName=dyn_mohawk_front_01}
      boxExtents=<16 40 40>
      constraintSetup {angularLimitsMax=<2 10 5>, angularLimitsMin=<-2 -10 -5>, angularTarget=<1 0 0>}
      localJointOffset=<8 0 0>
     }]
    relativeSpaceBone {boneName=spine_05}
   }
   Anim Dynamics - Bone: dyn_mohawk_front_02(ComponentPose=@n1, Alpha=1, ExternalForce="0,0,0") {
    angularDampingOverride=0.9
    angularSpringConstant=3000
    bAngularSpring=true
    bGravityOverrideInSimSpace=true
    bOverrideAngularDamping=true
    bUseGravityOverride=true
    simulationSpace=RootRelative
    alphaBoolBlend {blendOption=Linear}
    boundBone {boneName=dyn_mohawk_front_02}
    componentAppliedLinearAccClamp=<1500 1500 1500>
    componentLinearAccScale=<0.3 0.3 0.4>
    componentLinearVelScale=<0.2 0.2 0.3>
    gravityOverride {z=-980}
    physicsBodyDefinitions [{
      sphereCollisionRadius=0
      boundBone {boneName=dyn_mohawk_front_02}
      boxExtents=<16 40 40>
      constraintSetup {angularLimitsMax=<5 5 5>, angularLimitsMin=<-5 -10 -5>, angularTarget=<1 0 0>}
      localJointOffset=<8 0 0>
     }]
    relativeSpaceBone {boneName=spine_05}
   }
   n2 = Get MeshToCopy() -> (MeshToCopy:Skeletal Mesh Component Object Reference)
   Blend Poses by bool(
    BlendPose_0=Layered blend per bone(BasePose=Copy Pose From Mesh(SourceMeshComponent=@n2), BlendPoses_0=Slot 'DefaultSlot'()) {layerSetup [{branchFilters [{boneName=faceAttach}]}]}
    BlendPose_1=Layered blend per bone(BasePose=Copy Pose From Mesh(SourceMeshComponent=@n2), BlendPoses_0=Slot 'inGameSlot'() {slotName=inGameSlot}) {layerSetup [{branchFilters [{boneName=faceAttach}]}]}
    bActiveValue=Get bCanPlayCustomAnimations() -> (bCanPlayCustomAnimations:bool)
   ) {blendType=Linear}
   n3 = "Transform (Modify) Bone - Bone: head"!(ComponentPose=Local To Component(LocalPose=Use cached pose 'preScaleCheck'()), Alpha=1) -> (Pose) {scaleMode=BMM_Replace, alphaBoolBlend {blendOption=Linear}, boneToModify {boneName=head}, scale=<0.01 0.01 0.01>}
   Output Pose(Result=Blend Poses by bool(
     BlendPose_0=Component To Local(ComponentPose=@n3)
     BlendPose_1=Use cached pose 'preScaleCheck'()
     BlendTime_0=0
     BlendTime_1=0
     bActiveValue=Get GetPlayerListUsesCustomCharacterParts() -> (GetPlayerListUsesCustomCharacterParts:bool)
    ) {blendType=Linear}) {layerGroup=None}
   n4 = Sequence Player 'AlienTrooper_Male_Facial'() {sequence=/BRCosmetics/Animation/Game/MainPlayer/Facial/AlienTrooper/AlienTrooper_Male_Facial}
   n5 = Use cached pose 'IncomingPose'()
   n6 = "Blend Poses (EFortFacialAnimTypes)"!(
    BlendPose_0:Pose Link Structure=Two Way Blend!(
     A=Layered blend per bone(BasePose=@n5, BlendPoses_0=@n4) {curveBlendOption=DoNotOverride, layerSetup [{branchFilters [{boneName=faceAttach}]}]}
     B=Use cached pose 'IncomingPose'()
     AlphaCurveName:Name=DisableFaceOverride
    ) -> (Pose) {blendNode {alphaInputType=Curve, alphaBoolBlend {blendOption=Linear}}}
    BlendPose_1:Pose Link Structure=Use cached pose 'IncomingPose'()
    BlendPose_2:Pose Link Structure=Use cached pose 'IncomingPose'()
    BlendTime_0:float=0.1
    BlendTime_1:float=0.1
    BlendTime_2:float=0.1
    ActiveEnumValue:Byte=Get!() -> (
     FacialInput_SubjectName:Live Link Subject Name Structure
     FacialInput_CurrentAnimType:EFortFacialAnimTypes Enum
     FacialInput_bCurvesOnly:bool=~
    ).FacialInput_CurrentAnimType
   ) -> (Pose:Pose Link Structure) {
    is=title AnimGraphNode_BlendListByEnum
    blendType=Linear
    blendPose [{linkId=-1, sourceLinkId=-1}, {linkId=-1, sourceLinkId=-1}, {linkId=-1, sourceLinkId=-1}]
    blendTime [0.1,0.1,0.1]
   }
   Save cached pose 'GameFacial'(Pose=@n6) {cacheName=GameFacial}
   n7 = M_MED_Alien_Trooper_Base_Head_Facial_PoseAsset!(SourcePose:Pose Link Structure=Use cached pose 'post driven'()) -> (Pose:Pose Link Structure) {
    is=title AnimGraphNode_PoseBlendNode
    blendOption=Linear
    poseAsset=&/M_MED_Alien_Trooper_Base_Head_Facial_PoseAsset
   }
   n8 = Control Rig!(Source=@n7, bAlphaBoolEnabled:bool=Get Chin_Active() -> (Chin_Active:bool)) -> (Pose) {
    alphaInputType=Bool
    alphaBoolBlend {blendOption=Linear}
    controlRigAssetReference {blueprintRigClass=/BRCosmetics/Characters/Player/Male/Medium/Heads/M_MED_AlienTrooper_Head/Meshes/Tentacle/Male_Medium_Alien_Trooper_Tentacle_Head_CtrlRig.Male_Medium_Alien_Trooper_Tentacle_Head_CtrlRig_C}
   }
   n9 = "Sequence Player (additive) 'Male_Medium_Alien_Trooper_T_Head_Sequence'"() {sequence=/BRCosmetics/Characters/Player/Male/Medium/Heads/M_MED_AlienTrooper_Head/Meshes/Tentacle/Male_Medium_Alien_Trooper_T_Head_Sequence}
   n10 = Get bRigidBodyEnabled()
   n11 = RigidBody(ComponentPose=Local To Component(LocalPose=Apply Additive!(Base=@n8, Additive=@n9, bAlphaBoolEnabled:bool=@n10) -> (Pose) {alphaInputType=Bool, alphaBoolBlend {blendOption=Linear}})) {
    bClampLinearTranslationLimitToRefPose=true
    bOverrideWorldGravity=true
    overridePhysicsAsset=/BRCosmetics/Characters/Player/Male/Medium/Heads/M_MED_AlienTrooper_Head/Meshes/Shark/Male_Medium_Alien_Trooper_Shark_Head_Physics
    simulationSpace=BaseBoneSpace
    alphaBoolBlend {blendOption=Linear}
    baseBoneRef {boneName=root}
    componentAppliedLinearAccClamp=<3000 3000 3000>
    componentLinearAccScale=<0.1 0.1 0.4>
    componentLinearVelScale=<0.05 0.05 0.2>
    overrideWorldGravity {z=-980}
   }
   n12 = RigidBody(ComponentPose=@n11) {
    bClampLinearTranslationLimitToRefPose=true
    bOverrideWorldGravity=true
    overridePhysicsAsset=/BRCosmetics/Characters/Player/Male/Medium/Bodies/M_MED_Alien_Trooper/Meshes/Parts/Shark/M_MED_Alien_Trooper_Shark_Physics
    simulationSpace=BaseBoneSpace
    alphaBoolBlend {blendOption=Linear}
    baseBoneRef {boneName=root}
    componentAppliedLinearAccClamp=<3000 3000 3000>
    componentLinearAccScale=<0.1 0.1 0.4>
    componentLinearVelScale=<0.05 0.05 0.2>
    overrideWorldGravity {z=-980}
   }
   n13 = Control Rig!(Source=Component To Local(ComponentPose=@n12), Alpha:float) -> (Pose) {
    alphaBoolBlend {blendOption=Linear}
    controlRigAssetReference {blueprintRigClass=/BRCosmetics/Characters/Player/Male/Medium/Bodies/M_MED_Alien_Trooper/Meshes/Parts/Shark/M_MED_Alien_Trooper_Shark_CtrlRig.M_MED_Alien_Trooper_Shark_CtrlRig_C}
   }
   n14 = RigidBody!(ComponentPose=Local To Component(LocalPose=@n13), Alpha, OverrideWorldGravity:Vector=Get GravityOverride(), ExternalForce) -> (Pose) {
    bClampLinearTranslationLimitToRefPose=true
    bOverrideWorldGravity=true
    overridePhysicsAsset=/BRCosmetics/Characters/Player/Male/Medium/Bodies/M_MED_Alien_Trooper/Meshes/Parts/Tentacle/M_MED_Alien_Trooper_Tentacle_Physics
    simulationSpace=BaseBoneSpace
    alphaBoolBlend {blendOption=Linear}
    baseBoneRef {boneName=spine_05}
    componentAppliedLinearAccClamp=<5000 5000 5000>
    componentLinearAccScale=<0.3 0.3 0.5>
    componentLinearVelScale=<0.15 0.15 0.3>
    overrideWorldGravity {z=-980}
   }
   n15 = Get bIsInFrontEnd()
   n16 = NOT Boolean(A=@n15)
   n17 = RigidBody!(ComponentPose=@n14, bAlphaBoolEnabled:bool=@n16, ExternalForce) -> (Pose) {
    alphaInputType=Bool
    bClampLinearTranslationLimitToRefPose=true
    bOverrideWorldGravity=true
    overridePhysicsAsset=/BRCosmetics/Characters/Player/Male/Medium/Heads/M_MED_AlienTrooper_Head/Meshes/Tentacle/Male_Medium_Alien_Trooper_Tentacle_Head_Physics
    simulationSpace=BaseBoneSpace
    alphaBoolBlend {blendOption=Linear}
    baseBoneRef {boneName=spine_05}
    componentAppliedLinearAccClamp=<5000 5000 5000>
    componentLinearAccScale=<0.3 0.3 0.5>
    componentLinearVelScale=<0.15 0.15 0.3>
    overrideWorldGravity {z=-980}
   }
   n18 = RigidBody(ComponentPose=@n17, bAlphaBoolEnabled=@n15) {
    alphaInputType=Bool
    bClampLinearTranslationLimitToRefPose=true
    bOverrideWorldGravity=true
    overridePhysicsAsset=/BRCosmetics/Characters/Player/Male/Medium/Heads/M_MED_AlienTrooper_Head/Meshes/Tentacle/Male_Medium_Alien_Trooper_Tentacle_FE_Physics
    simulationSpace=BaseBoneSpace
    alphaBoolBlend {blendOption=Linear}
    baseBoneRef {boneName=spine_05}
    componentAppliedLinearAccClamp=<5000 5000 5000>
    componentLinearAccScale=<0.3 0.3 0.5>
    componentLinearVelScale=<0.15 0.15 0.3>
    overrideWorldGravity {z=-980}
   }
   n19 = Control Rig!(Source=Component To Local(ComponentPose=@n18), bAlphaBoolEnabled:bool=@n16) -> (Pose) {
    alphaInputType=Bool
    alphaBoolBlend {blendOption=Linear}
    controlRigAssetReference {blueprintRigClass=/BRCosmetics/Characters/Player/Male/Medium/Heads/M_MED_AlienTrooper_Head/Meshes/Tentacle/Male_Medium_Alien_Trooper_Tentacle_CtrlRig.Male_Medium_Alien_Trooper_Tentacle_CtrlRig_C}
   }
   n20 = Control Rig(Source=@n19, bAlphaBoolEnabled=@n15) {
    alphaInputType=Bool
    alphaBoolBlend {blendOption=Linear}
    controlRigAssetReference {blueprintRigClass=/BRCosmetics/Characters/Player/Male/Medium/Heads/M_MED_AlienTrooper_Head/Meshes/Tentacle/Male_Medium_Alien_Trooper_Tentacle_FE_CtrlRig.Male_Medium_Alien_Trooper_Tentacle_FE_CtrlRig_C}
   }
   n21 = RigidBody!(ComponentPose=Local To Component(LocalPose=@n20), Alpha:float, ExternalForce) -> (Pose) {
    bClampLinearTranslationLimitToRefPose=true
    bOverrideWorldGravity=true
    overridePhysicsAsset=/BRCosmetics/Characters/Player/Male/Medium/Heads/M_MED_AlienTrooper_Head/Meshes/Brain/Male_Medium_Alien_Trooper_Brain_Head_Physics
    simulationSpace=BaseBoneSpace
    alphaBoolBlend {blendOption=Linear}
    baseBoneRef {boneName=root}
    componentAppliedLinearAccClamp=<3000 3000 3000>
    componentLinearAccScale=<0.1 0.1 0.4>
    componentLinearVelScale=<0.05 0.05 0.2>
    overrideWorldGravity {z=-980}
   }
   n22 = RigidBody(ComponentPose=@n21, Alpha=0.5) {
    bClampLinearTranslationLimitToRefPose=true
    bOverrideWorldGravity=true
    overridePhysicsAsset=/BRCosmetics/Characters/Player/Male/Medium/Bodies/M_MED_Alien_Trooper/Meshes/Parts/Brain/M_MED_Alien_Trooper_Brain_Physics
    simulationSpace=BaseBoneSpace
    alphaBoolBlend {blendOption=Linear}
    baseBoneRef {boneName=root}
    componentAppliedLinearAccClamp=<3000 3000 3000>
    componentLinearAccScale=<0.1 0.1 0.4>
    componentLinearVelScale=<0.05 0.05 0.2>
    overrideWorldGravity {z=-980}
   }
   n23 = Control Rig!(Source=Component To Local(ComponentPose=@n22), Alpha:float) -> (Pose) {
    alphaBoolBlend {blendOption=Linear}
    controlRigAssetReference {blueprintRigClass=/BRCosmetics/Characters/Player/Male/Medium/Bodies/M_MED_Alien_Trooper/Meshes/Parts/Brain/M_MED_Alien_Trooper_Brain_CtrlRig.M_MED_Alien_Trooper_Brain_CtrlRig_C}
   }
   n24 = Control Rig(Source=@n23) {
    alphaBoolBlend {blendOption=Linear}
    controlRigAssetReference {blueprintRigClass=/BRCosmetics/Characters/Player/Male/Medium/Bodies/M_MED_Alien_Trooper/Meshes/Parts/Owl/M_MED_Alien_Trooper_Owl_CtrlRig.M_MED_Alien_Trooper_Owl_CtrlRig_C}
   }
   n25 = Control Rig(Source=@n24) {
    alphaBoolBlend {blendOption=Linear}
    controlRigAssetReference {blueprintRigClass=/BRCosmetics/Characters/Player/Male/Medium/Heads/M_MED_AlienTrooper_Head/Meshes/Owl/Male_Medium_Alien_Trooper_Owl_Head_CtrlRig.Male_Medium_Alien_Trooper_Owl_Head_CtrlRig_C}
   }
   n26 = RigidBody!(ComponentPose=Local To Component(LocalPose=@n25), Alpha, OverrideWorldGravity:Vector=Get GravityOverride(), ExternalForce) -> (Pose) {
    bClampLinearTranslationLimitToRefPose=true
    bOverrideWorldGravity=true
    overridePhysicsAsset=/BRCosmetics/Characters/Player/Male/Medium/Bodies/M_MED_Alien_Trooper/Meshes/Parts/Owl/M_MED_Alien_Trooper_Owl_Physics
    simulationSpace=BaseBoneSpace
    alphaBoolBlend {blendOption=Linear}
    baseBoneRef {boneName=spine_05}
    componentAppliedLinearAccClamp=<5000 5000 5000>
    componentLinearAccScale=<0.3 0.3 0.5>
    componentLinearVelScale=<0.15 0.15 0.3>
    overrideWorldGravity {z=-980}
   }
   n27 = RigidBody(ComponentPose=@n26, OverrideWorldGravity="0,0,-980") {
    bClampLinearTranslationLimitToRefPose=true
    bOverrideWorldGravity=true
    overridePhysicsAsset=/BRCosmetics/Characters/Player/Male/Medium/Heads/M_MED_AlienTrooper_Head/Meshes/Owl/Male_Medium_Alien_Trooper_Owl_Head_Physics
    simulationSpace=BaseBoneSpace
    alphaBoolBlend {blendOption=Linear}
    baseBoneRef {boneName=spine_05}
    componentAppliedLinearAccClamp=<5000 5000 5000>
    componentLinearAccScale=<0.3 0.3 0.5>
    componentLinearVelScale=<0.15 0.15 0.3>
    overrideWorldGravity {z=-980}
   }
   n28 = Anim Dynamics - Bone: dyn_mohawk_front_01(ComponentPose=@n27, Alpha=1, ExternalForce="0,0,0") {
    angularDampingOverride=0.9
    angularSpringConstant=3000
    bAngularSpring=true
    bGravityOverrideInSimSpace=true
    bOverrideAngularDamping=true
    bUseGravityOverride=true
    simulationSpace=RootRelative
    alphaBoolBlend {blendOption=Linear}
    boundBone {boneName=dyn_mohawk_front_01}
    componentAppliedLinearAccClamp=<1500 1500 1500>
    componentLinearAccScale=<0.3 0.3 0.4>
    componentLinearVelScale=<0.2 0.2 0.3>
    gravityOverride {z=-980}
    physicsBodyDefinitions [{
      sphereCollisionRadius=0
      boundBone {boneName=dyn_mohawk_front_01}
      boxExtents=<8 20 20>
      constraintSetup {twistAxis=AxisY, angularLimitsMax=<2 10 5>, angularLimitsMin=<-2 -10 -5>, angularTarget=<1 0 0>}
      localJointOffset=<8 0 0>
     }]
    relativeSpaceBone {boneName=spine_05}
   }
   n29 = Anim Dynamics - Bone: dyn_mohawk_front_02(ComponentPose=@n28, Alpha=1, ExternalForce="0,0,0") {
    angularDampingOverride=0.9
    angularSpringConstant=3000
    bAngularSpring=true
    bGravityOverrideInSimSpace=true
    bOverrideAngularDamping=true
    bUseGravityOverride=true
    simulationSpace=RootRelative
    alphaBoolBlend {blendOption=Linear}
    boundBone {boneName=dyn_mohawk_front_02}
    componentAppliedLinearAccClamp=<1500 1500 1500>
    componentLinearAccScale=<0.3 0.3 0.4>
    componentLinearVelScale=<0.2 0.2 0.3>
    gravityOverride {z=-980}
    physicsBodyDefinitions [{
      sphereCollisionRadius=0
      boundBone {boneName=dyn_mohawk_front_02}
      boxExtents=<8 20 20>
      constraintSetup {twistAxis=AxisY, angularLimitsMax=<5 5 5>, angularLimitsMin=<-5 -10 -5>, angularTarget=<1 0 0>}
      localJointOffset=<8 0 0>
     }]
    relativeSpaceBone {boneName=spine_05}
   }
   n30 = Anim Dynamics - Bone: dyn_mohawk_mid_01(ComponentPose=@n29, Alpha=1, ExternalForce="0,0,0") {
    angularDampingOverride=0.9
    angularSpringConstant=3000
    bAngularSpring=true
    bGravityOverrideInSimSpace=true
    bOverrideAngularDamping=true
    bUseGravityOverride=true
    simulationSpace=BoneRelative
    alphaBoolBlend {blendOption=Linear}
    boundBone {boneName=dyn_mohawk_mid_01}
    componentAppliedLinearAccClamp=<1500 1500 1500>
    componentLinearAccScale=<0.3 0.3 0.4>
    componentLinearVelScale=<0.2 0.2 0.3>
    gravityOverride {z=-980}
    physicsBodyDefinitions [{
      sphereCollisionRadius=0
      boundBone {boneName=dyn_mohawk_mid_01}
      boxExtents=<8 20 20>
      constraintSetup {twistAxis=AxisY, angularLimitsMax=<5 5 5>, angularLimitsMin=<-5 -10 -5>, angularTarget=<1 0 0>}
      localJointOffset=<8 0 0>
     }]
    relativeSpaceBone {boneName=root}
   }
   n31 = Anim Dynamics - Bone: dyn_brow_l_01(ComponentPose=@n30, Alpha=1, ExternalForce="0,0,0") {
    angularDampingOverride=0.7
    angularSpringConstant=50
    bAngularSpring=true
    bGravityOverrideInSimSpace=true
    bOverrideAngularDamping=true
    bUseGravityOverride=true
    simulationSpace=BoneRelative
    alphaBoolBlend {blendOption=Linear}
    boundBone {boneName=dyn_brow_l_01}
    componentAppliedLinearAccClamp=<1500 1500 1500>
    componentLinearAccScale=<0.4 0.4 0.5>
    componentLinearVelScale=<0.2 0.2 0.3>
    gravityOverride {z=-980}
    physicsBodyDefinitions [{
      sphereCollisionRadius=0
      boundBone {boneName=dyn_brow_l_01}
      boxExtents=<5 5 5>
      constraintSetup {angularLimitsMax=<10 15 10>, angularLimitsMin=<-10 -25 -10>, angularTarget=<1 0 0>}
      localJointOffset=<1 0 0>
     }]
    relativeSpaceBone {boneName=root}
   }
   n32 = Anim Dynamics - Bone: dyn_brow_l_02(ComponentPose=@n31, Alpha=1, ExternalForce="0,0,0") {
    angularDampingOverride=0.7
    angularSpringConstant=50
    bAngularSpring=true
    bGravityOverrideInSimSpace=true
    bOverrideAngularDamping=true
    bUseGravityOverride=true
    simulationSpace=BoneRelative
    alphaBoolBlend {blendOption=Linear}
    boundBone {boneName=dyn_brow_l_02}
    componentAppliedLinearAccClamp=<1500 1500 1500>
    componentLinearAccScale=<0.4 0.4 0.5>
    componentLinearVelScale=<0.2 0.2 0.3>
    gravityOverride {z=-980}
    physicsBodyDefinitions [{
      sphereCollisionRadius=0
      boundBone {boneName=dyn_brow_l_02}
      boxExtents=<5 5 5>
      constraintSetup {angularLimitsMax=<10 15 10>, angularLimitsMin=<-10 -25 -10>, angularTarget=<1 0 0>}
      localJointOffset=<1 0 0>
     }]
    relativeSpaceBone {boneName=root}
   }
   n33 = Anim Dynamics - Bone: dyn_brow_r_01(ComponentPose=@n32, Alpha=1, ExternalForce="0,0,0") {
    angularDampingOverride=0.7
    angularSpringConstant=50
    bAngularSpring=true
    bGravityOverrideInSimSpace=true
    bOverrideAngularDamping=true
    bUseGravityOverride=true
    simulationSpace=BoneRelative
    alphaBoolBlend {blendOption=Linear}
    boundBone {boneName=dyn_brow_r_01}
    componentAppliedLinearAccClamp=<1500 1500 1500>
    componentLinearAccScale=<0.4 0.4 0.5>
    componentLinearVelScale=<0.2 0.2 0.3>
    gravityOverride {z=-980}
    physicsBodyDefinitions [{
      sphereCollisionRadius=0
      boundBone {boneName=dyn_brow_r_01}
      boxExtents=<5 5 5>
      constraintSetup {angularLimitsMax=<10 15 10>, angularLimitsMin=<-10 -25 -10>, angularTarget=<1 0 0>}
      localJointOffset=<1 0 0>
     }]
    relativeSpaceBone {boneName=root}
   }
   n34 = Anim Dynamics - Bone: dyn_brow_r_02(ComponentPose=@n33, Alpha=1, ExternalForce="0,0,0") {
    angularDampingOverride=0.7
    angularSpringConstant=50
    bAngularSpring=true
    bGravityOverrideInSimSpace=true
    bOverrideAngularDamping=true
    bUseGravityOverride=true
    simulationSpace=BoneRelative
    alphaBoolBlend {blendOption=Linear}
    boundBone {boneName=dyn_brow_r_02}
    componentAppliedLinearAccClamp=<1500 1500 1500>
    componentLinearAccScale=<0.4 0.4 0.5>
    componentLinearVelScale=<0.2 0.2 0.3>
    gravityOverride {z=-980}
    physicsBodyDefinitions [{
      sphereCollisionRadius=0
      boundBone {boneName=dyn_brow_r_02}
      boxExtents=<5 5 5>
      constraintSetup {angularLimitsMax=<10 15 10>, angularLimitsMin=<-10 -25 -10>, angularTarget=<1 0 0>}
      localJointOffset=<1 0 0>
     }]
    relativeSpaceBone {boneName=root}
   }
   n35 = Anim Dynamics - Bone: dyn_ear_l_01(ComponentPose=@n34, Alpha=1, ExternalForce="0,0,0") {
    angularDampingOverride=1
    angularSpringConstant=1500
    bAngularSpring=true
    bGravityOverrideInSimSpace=true
    bOverrideAngularDamping=true
    bUseGravityOverride=true
    gravityScale=0.35
    simulationSpace=RootRelative
    alphaBoolBlend {blendOption=Linear}
    boundBone {boneName=dyn_ear_l_01}
    componentAppliedLinearAccClamp=<1000 1000 1000>
    componentLinearAccScale=<0.2 0.2 0.2>
    componentLinearVelScale=<0.1 0.1 0.2>
    gravityOverride {z=-980}
    physicsBodyDefinitions [{
      sphereCollisionRadius=0
      boundBone {boneName=dyn_ear_l_01}
      boxExtents=<16 8 16>
      constraintSetup {twistAxis=AxisY, angularLimitsMax=<5 5 2>, angularLimitsMin=<-5 -5 -2>, angularTarget=<0 0 0>}
      localJointOffset=<5 0 0>
     }]
   }
   n36 = Anim Dynamics - Bone: dyn_ear_l_02(ComponentPose=@n35, Alpha=1, ExternalForce="0,0,0") {
    angularDampingOverride=1
    angularSpringConstant=1500
    bAngularSpring=true
    bGravityOverrideInSimSpace=true
    bOverrideAngularDamping=true
    bUseGravityOverride=true
    gravityScale=0.35
    simulationSpace=RootRelative
    alphaBoolBlend {blendOption=Linear}
    boundBone {boneName=dyn_ear_l_02}
    componentAppliedLinearAccClamp=<1000 1000 1000>
    componentLinearAccScale=<0.2 0.2 0.2>
    componentLinearVelScale=<0.1 0.1 0.2>
    gravityOverride {z=-980}
    physicsBodyDefinitions [{
      sphereCollisionRadius=0
      boundBone {boneName=dyn_ear_l_02}
      boxExtents=<16 8 16>
      constraintSetup {twistAxis=AxisY, angularLimitsMax=<10 10 10>, angularLimitsMin=<-10 -10 -10>, angularTarget=<0 0 0>}
      localJointOffset=<5 0 0>
     }]
   }
   n37 = Anim Dynamics - Bone: dyn_ear_l_03(ComponentPose=@n36, Alpha=1, ExternalForce="0,0,0") {
    angularDampingOverride=1
    angularSpringConstant=1500
    bAngularSpring=true
    bGravityOverrideInSimSpace=true
    bOverrideAngularDamping=true
    bUseGravityOverride=true
    gravityScale=0.35
    simulationSpace=RootRelative
    alphaBoolBlend {blendOption=Linear}
    boundBone {boneName=dyn_ear_l_03}
    componentAppliedLinearAccClamp=<1000 1000 1000>
    componentLinearAccScale=<0.2 0.2 0.2>
    componentLinearVelScale=<0.1 0.1 0.2>
    gravityOverride {z=-980}
    physicsBodyDefinitions [{
      sphereCollisionRadius=0
      boundBone {boneName=dyn_ear_l_03}
      boxExtents=<16 8 16>
      constraintSetup {twistAxis=AxisY, angularLimitsMax=<10 10 10>, angularLimitsMin=<-10 -10 -10>, angularTarget=<0 0 0>}
      localJointOffset=<5 0 0>
     }]
   }
   n38 = Anim Dynamics - Bone: dyn_ear_l_04(ComponentPose=@n37, Alpha=1, ExternalForce="0,0,0") {
    angularDampingOverride=1
    angularSpringConstant=1500
    bAngularSpring=true
    bGravityOverrideInSimSpace=true
    bOverrideAngularDamping=true
    bUseGravityOverride=true
    gravityScale=0.35
    simulationSpace=RootRelative
    alphaBoolBlend {blendOption=Linear}
    boundBone {boneName=dyn_ear_l_04}
    componentAppliedLinearAccClamp=<1000 1000 1000>
    componentLinearAccScale=<0.2 0.2 0.2>
    componentLinearVelScale=<0.1 0.1 0.2>
    gravityOverride {z=-980}
    physicsBodyDefinitions [{
      sphereCollisionRadius=0
      boundBone {boneName=dyn_ear_l_04}
      boxExtents=<16 8 16>
      constraintSetup {twistAxis=AxisY, angularLimitsMax=<10 10 10>, angularLimitsMin=<-10 -10 -10>, angularTarget=<0 0 0>}
      localJointOffset=<5 0 0>
     }]
   }
   n39 = Anim Dynamics - Bone: dyn_ear_r_01(ComponentPose=@n38, Alpha=1, ExternalForce="0,0,0") {
    angularDampingOverride=1
    angularSpringConstant=1500
    bAngularSpring=true
    bGravityOverrideInSimSpace=true
    bOverrideAngularDamping=true
    bUseGravityOverride=true
    gravityScale=0.35
    simulationSpace=RootRelative
    alphaBoolBlend {blendOption=Linear}
    boundBone {boneName=dyn_ear_r_01}
    componentAppliedLinearAccClamp=<1000 1000 1000>
    componentLinearAccScale=<0.2 0.2 0.2>
    componentLinearVelScale=<0.1 0.1 0.2>
    gravityOverride {z=-980}
    physicsBodyDefinitions [{
      sphereCollisionRadius=0
      boundBone {boneName=dyn_ear_r_01}
      boxExtents=<16 8 16>
      constraintSetup {twistAxis=AxisY, angularLimitsMax=<5 5 2>, angularLimitsMin=<-5 -5 -2>, angularTarget=<0 0 0>}
      localJointOffset=<5 0 0>
     }]
   }
   n40 = Anim Dynamics - Bone: dyn_ear_r_02(ComponentPose=@n39, Alpha=1, ExternalForce="0,0,0") {
    angularDampingOverride=1
    angularSpringConstant=1500
    bAngularSpring=true
    bGravityOverrideInSimSpace=true
    bOverrideAngularDamping=true
    bUseGravityOverride=true
    gravityScale=0.35
    simulationSpace=RootRelative
    alphaBoolBlend {blendOption=Linear}
    boundBone {boneName=dyn_ear_r_02}
    componentAppliedLinearAccClamp=<1000 1000 1000>
    componentLinearAccScale=<0.2 0.2 0.2>
    componentLinearVelScale=<0.1 0.1 0.2>
    gravityOverride {z=-980}
    physicsBodyDefinitions [{
      sphereCollisionRadius=0
      boundBone {boneName=dyn_ear_r_02}
      boxExtents=<16 8 16>
      constraintSetup {twistAxis=AxisY, angularLimitsMax=<10 10 10>, angularLimitsMin=<-10 -10 -10>, angularTarget=<0 0 0>}
      localJointOffset=<5 0 0>
     }]
   }
   n41 = Anim Dynamics - Bone: dyn_ear_r_03(ComponentPose=@n40, Alpha=1, ExternalForce="0,0,0") {
    angularDampingOverride=1
    angularSpringConstant=1500
    bAngularSpring=true
    bGravityOverrideInSimSpace=true
    bOverrideAngularDamping=true
    bUseGravityOverride=true
    gravityScale=0.35
    simulationSpace=RootRelative
    alphaBoolBlend {blendOption=Linear}
    boundBone {boneName=dyn_ear_r_03}
    componentAppliedLinearAccClamp=<1000 1000 1000>
    componentLinearAccScale=<0.2 0.2 0.2>
    componentLinearVelScale=<0.1 0.1 0.2>
    gravityOverride {z=-980}
    physicsBodyDefinitions [{
      sphereCollisionRadius=0
      boundBone {boneName=dyn_ear_r_03}
      boxExtents=<16 8 16>
      constraintSetup {twistAxis=AxisY, angularLimitsMax=<10 10 10>, angularLimitsMin=<-10 -10 -10>, angularTarget=<0 0 0>}
      localJointOffset=<5 0 0>
     }]
   }
   n42 = Anim Dynamics - Bone: dyn_ear_r_04(ComponentPose=@n41, Alpha=1, ExternalForce="0,0,0") {
    angularDampingOverride=1
    angularSpringConstant=1500
    bAngularSpring=true
    bGravityOverrideInSimSpace=true
    bOverrideAngularDamping=true
    bUseGravityOverride=true
    gravityScale=0.35
    simulationSpace=RootRelative
    alphaBoolBlend {blendOption=Linear}
    boundBone {boneName=dyn_ear_r_04}
    componentAppliedLinearAccClamp=<1000 1000 1000>
    componentLinearAccScale=<0.2 0.2 0.2>
    componentLinearVelScale=<0.1 0.1 0.2>
    gravityOverride {z=-980}
    physicsBodyDefinitions [{
      sphereCollisionRadius=0
      boundBone {boneName=dyn_ear_r_04}
      boxExtents=<16 8 16>
      constraintSetup {twistAxis=AxisY, angularLimitsMax=<10 10 10>, angularLimitsMin=<-10 -10 -10>, angularTarget=<0 0 0>}
      localJointOffset=<5 0 0>
     }]
   }
   n43 = RigidBody!(ComponentPose=@n42, Alpha=0.7, ExternalForce) -> (Pose) {
    bClampLinearTranslationLimitToRefPose=true
    bOverrideWorldGravity=true
    overridePhysicsAsset=/BRCosmetics/Characters/Player/Male/Medium/Heads/M_MED_AlienTrooper_Head/Meshes/Cyclops/Male_Medium_Alien_Trooper_Cyclops_Head_Physics
    simulationSpace=BaseBoneSpace
    alphaBoolBlend {blendOption=Linear}
    baseBoneRef {boneName=root}
    componentAppliedLinearAccClamp=<3000 3000 3000>
    componentLinearAccScale=<0.1 0.1 0.2>
    componentLinearVelScale=<0.05 0.05 0.1>
    overrideWorldGravity {z=-980}
   }
   n44 = Control Rig!(Source=Component To Local(ComponentPose=@n43), bAlphaBoolEnabled:bool=Get Cyclops_Chin_Active() -> (Cyclops_Chin_Active:bool)) -> (Pose) {
    alphaInputType=Bool
    alphaBoolBlend {blendOption=Linear}
    controlRigAssetReference {blueprintRigClass=/BRCosmetics/Characters/Player/Male/Medium/Heads/M_MED_AlienTrooper_Head/Meshes/Cyclops/Male_Medium_Alien_Trooper_Cyclops_Head_CtrlRig.Male_Medium_Alien_Trooper_Cyclops_Head_CtrlRig_C}
   }
   n45 = Control Rig!(Source=@n44, Alpha:float) -> (Pose) {
    alphaBoolBlend {blendOption=Linear}
    controlRigAssetReference {blueprintRigClass=/BRCosmetics/Characters/Player/Male/Medium/Bodies/M_MED_Alien_Trooper/Meshes/Parts/Cyclops/M_MED_Alien_Trooper_Cyclops_CtrlRig.M_MED_Alien_Trooper_Cyclops_CtrlRig_C}
   }
   n46 = "Constraint - dyn_spikes_01 (PPPPPPPPPPPPP)"!(
    ComponentPose=Local To Component(LocalPose=@n45)
    Alpha=1
    ConstraintWeights_0=0
    ConstraintWeights_1:float=0.7937
    ConstraintWeights_2:float=0
    ConstraintWeights_3:float=0.0002
    ConstraintWeights_4:float=0.0009
    ConstraintWeights_5:float=0
    ConstraintWeights_6:float=0.0006
    ConstraintWeights_7:float=0
    ConstraintWeights_8:float=0.0002
    ConstraintWeights_9:float=0.2043
    ConstraintWeights_10:float=0
    ConstraintWeights_11:float=0
    ConstraintWeights_12:float=0.0001
   ) -> (Pose) {
    alphaBoolBlend {blendOption=Linear}
    boneToModify {boneName=dyn_spikes_01}
    constraintSetup [
     {targetBone {boneName=R_eye_lid_lower_mid}}
     {targetBone {boneName=head}}
     {targetBone {boneName=L_cheek_inner}}
     {targetBone {boneName=L_brow_outer}}
     {targetBone {boneName=C_brow_mid}}
     {targetBone {boneName=L_eye_lid_lower_mid}}
     {targetBone {boneName=L_brow_mid}}
     {targetBone {boneName=L_eye_lid_upper_mid}}
     {targetBone {boneName=R_brow_mid}}
     {targetBone {boneName=R_eye_lid_upper_mid}}
     {targetBone {boneName=C_nose_bridge}}
     {targetBone {boneName=R_cheek_inner}}
     {targetBone {boneName=R_brow_outer}}
    ]
    constraintWeights [1,1,1,1,1,1,1,1,1,1,1,1,1]
   }
   n47 = "Constraint - dyn_spikes_00 (PPPPPPPPPPPPP)"!(
    ComponentPose=@n46
    Alpha=1
    ConstraintWeights_0=0
    ConstraintWeights_1:float=0.824
    ConstraintWeights_2:float=0
    ConstraintWeights_3:float=0.0001
    ConstraintWeights_4:float=0.0008
    ConstraintWeights_5:float=0
    ConstraintWeights_6:float=0.0002
    ConstraintWeights_7:float=0
    ConstraintWeights_8:float=0.0005
    ConstraintWeights_9:float=0.174
    ConstraintWeights_10:float=0
    ConstraintWeights_11:float=0
    ConstraintWeights_12:float=0.0003
   ) -> (Pose) {
    alphaBoolBlend {blendOption=Linear}
    boneToModify {boneName=dyn_spikes_00}
    constraintSetup [
     {targetBone {boneName=R_eye_lid_lower_mid}}
     {targetBone {boneName=head}}
     {targetBone {boneName=L_cheek_inner}}
     {targetBone {boneName=L_brow_outer}}
     {targetBone {boneName=C_brow_mid}}
     {targetBone {boneName=L_eye_lid_lower_mid}}
     {targetBone {boneName=L_brow_mid}}
     {targetBone {boneName=L_eye_lid_upper_mid}}
     {targetBone {boneName=R_brow_mid}}
     {targetBone {boneName=R_eye_lid_upper_mid}}
     {targetBone {boneName=C_nose_bridge}}
     {targetBone {boneName=R_cheek_inner}}
     {targetBone {boneName=R_brow_outer}}
    ]
    constraintWeights [1,1,1,1,1,1,1,1,1,1,1,1,1]
   }
   n48 = "Constraint - dyn_spikes_02 (PPPPPPPPPPPPP)"!(
    ComponentPose=@n47
    Alpha=1
    ConstraintWeights_0=0.0001
    ConstraintWeights_1:float=0.844
    ConstraintWeights_2:float=0
    ConstraintWeights_3:float=0.0004
    ConstraintWeights_4:float=0.0008
    ConstraintWeights_5:float=0
    ConstraintWeights_6:float=0.0006
    ConstraintWeights_7:float=0
    ConstraintWeights_8:float=0.0001
    ConstraintWeights_9:float=0.154
    ConstraintWeights_10:float=0
    ConstraintWeights_11:float=0
    ConstraintWeights_12:float=0.0001
   ) -> (Pose) {
    alphaBoolBlend {blendOption=Linear}
    boneToModify {boneName=dyn_spikes_02}
    constraintSetup [
     {targetBone {boneName=R_eye_lid_lower_mid}}
     {targetBone {boneName=head}}
     {targetBone {boneName=L_cheek_inner}}
     {targetBone {boneName=L_brow_outer}}
     {targetBone {boneName=C_brow_mid}}
     {targetBone {boneName=L_eye_lid_lower_mid}}
     {targetBone {boneName=L_brow_mid}}
     {targetBone {boneName=L_eye_lid_upper_mid}}
     {targetBone {boneName=R_brow_mid}}
     {targetBone {boneName=R_eye_lid_upper_mid}}
     {targetBone {boneName=C_nose_bridge}}
     {targetBone {boneName=R_cheek_inner}}
     {targetBone {boneName=R_brow_outer}}
    ]
    constraintWeights [1,1,1,1,1,1,1,1,1,1,1,1,1]
   }
   n49 = "Constraint - dyn_spikes_04 (PPPPPPPPPPPPP)"!(
    ComponentPose=@n48
    Alpha=1
    ConstraintWeights_0=0.0002
    ConstraintWeights_1:float=0.9129
    ConstraintWeights_2:float=0
    ConstraintWeights_3:float=0.0005
    ConstraintWeights_4:float=0.0003
    ConstraintWeights_5:float=0
    ConstraintWeights_6:float=0.0004
    ConstraintWeights_7:float=0
    ConstraintWeights_8:float=0.0001
    ConstraintWeights_9:float=0.0854
    ConstraintWeights_10:float=0
    ConstraintWeights_11:float=0
    ConstraintWeights_12:float=0.0001
   ) -> (Pose) {
    alphaBoolBlend {blendOption=Linear}
    boneToModify {boneName=dyn_spikes_04}
    constraintSetup [
     {targetBone {boneName=R_eye_lid_lower_mid}}
     {targetBone {boneName=head}}
     {targetBone {boneName=L_cheek_inner}}
     {targetBone {boneName=L_brow_outer}}
     {targetBone {boneName=C_brow_mid}}
     {targetBone {boneName=L_eye_lid_lower_mid}}
     {targetBone {boneName=L_brow_mid}}
     {targetBone {boneName=L_eye_lid_upper_mid}}
     {targetBone {boneName=R_brow_mid}}
     {targetBone {boneName=R_eye_lid_upper_mid}}
     {targetBone {boneName=C_nose_bridge}}
     {targetBone {boneName=R_cheek_inner}}
     {targetBone {boneName=R_brow_outer}}
    ]
    constraintWeights [1,1,1,1,1,1,1,1,1,1,1,1,1]
   }
   n50 = "Constraint - dyn_spikes_03 (PPPPPPPPPPPPP)"!(
    ComponentPose=@n49
    Alpha=1
    ConstraintWeights_0=0.0001
    ConstraintWeights_1:float=0.8831
    ConstraintWeights_2:float=0
    ConstraintWeights_3:float=0.0001
    ConstraintWeights_4:float=0.0005
    ConstraintWeights_5:float=0.0001
    ConstraintWeights_6:float=0.0002
    ConstraintWeights_7:float=0
    ConstraintWeights_8:float=0.0004
    ConstraintWeights_9:float=0.1151
    ConstraintWeights_10:float=0
    ConstraintWeights_11:float=0
    ConstraintWeights_12:float=0.0004
   ) -> (Pose) {
    alphaBoolBlend {blendOption=Linear}
    boneToModify {boneName=dyn_spikes_03}
    constraintSetup [
     {targetBone {boneName=R_eye_lid_lower_mid}}
     {targetBone {boneName=head}}
     {targetBone {boneName=L_cheek_inner}}
     {targetBone {boneName=L_brow_outer}}
     {targetBone {boneName=C_brow_mid}}
     {targetBone {boneName=L_eye_lid_lower_mid}}
     {targetBone {boneName=L_brow_mid}}
     {targetBone {boneName=L_eye_lid_upper_mid}}
     {targetBone {boneName=R_brow_mid}}
     {targetBone {boneName=R_eye_lid_upper_mid}}
     {targetBone {boneName=C_nose_bridge}}
     {targetBone {boneName=R_cheek_inner}}
     {targetBone {boneName=R_brow_outer}}
    ]
    constraintWeights [1,1,1,1,1,1,1,1,1,1,1,1,1]
   }
   n51 = "Constraint - dyn_spikes_05 (PPPPPPPPPPPPP)"!(
    ComponentPose=@n50
    Alpha=1
    ConstraintWeights_0=0.0003
    ConstraintWeights_1:float=0.9322
    ConstraintWeights_2:float=0
    ConstraintWeights_3:float=0.0002
    ConstraintWeights_4:float=0.0002
    ConstraintWeights_5:float=0.0001
    ConstraintWeights_6:float=0.0001
    ConstraintWeights_7:float=0
    ConstraintWeights_8:float=0.0004
    ConstraintWeights_9:float=0.0654
    ConstraintWeights_10:float=0
    ConstraintWeights_11:float=0
    ConstraintWeights_12:float=0.0009
   ) -> (Pose) {
    alphaBoolBlend {blendOption=Linear}
    boneToModify {boneName=dyn_spikes_05}
    constraintSetup [
     {targetBone {boneName=R_eye_lid_lower_mid}}
     {targetBone {boneName=head}}
     {targetBone {boneName=L_cheek_inner}}
     {targetBone {boneName=L_brow_outer}}
     {targetBone {boneName=C_brow_mid}}
     {targetBone {boneName=L_eye_lid_lower_mid}}
     {targetBone {boneName=L_brow_mid}}
     {targetBone {boneName=L_eye_lid_upper_mid}}
     {targetBone {boneName=R_brow_mid}}
     {targetBone {boneName=R_eye_lid_upper_mid}}
     {targetBone {boneName=C_nose_bridge}}
     {targetBone {boneName=R_cheek_inner}}
     {targetBone {boneName=R_brow_outer}}
    ]
    constraintWeights [1,1,1,1,1,1,1,1,1,1,1,1,1]
   }
   n52 = "Constraint - dyn_spikes_06 (PPPPPPPPPPPPP)"!(
    ComponentPose=@n51
    Alpha=1
    ConstraintWeights_0=0.0001
    ConstraintWeights_1:float=0.9554
    ConstraintWeights_2:float=0
    ConstraintWeights_3:float=0.0002
    ConstraintWeights_4:float=0.0002
    ConstraintWeights_5:float=0
    ConstraintWeights_6:float=0.0002
    ConstraintWeights_7:float=0
    ConstraintWeights_8:float=0.0001
    ConstraintWeights_9:float=0.0435
    ConstraintWeights_10:float=0
    ConstraintWeights_11:float=0
    ConstraintWeights_12:float=0.0002
   ) -> (Pose) {
    alphaBoolBlend {blendOption=Linear}
    boneToModify {boneName=dyn_spikes_06}
    constraintSetup [
     {targetBone {boneName=R_eye_lid_lower_mid}}
     {targetBone {boneName=head}}
     {targetBone {boneName=L_cheek_inner}}
     {targetBone {boneName=L_brow_outer}}
     {targetBone {boneName=C_brow_mid}}
     {targetBone {boneName=L_eye_lid_lower_mid}}
     {targetBone {boneName=L_brow_mid}}
     {targetBone {boneName=L_eye_lid_upper_mid}}
     {targetBone {boneName=R_brow_mid}}
     {targetBone {boneName=R_eye_lid_upper_mid}}
     {targetBone {boneName=C_nose_bridge}}
     {targetBone {boneName=R_cheek_inner}}
     {targetBone {boneName=R_brow_outer}}
    ]
    constraintWeights [1,1,1,1,1,1,1,1,1,1,1,1,1]
   }
   n53 = "Constraint - dyn_spikes_07 (PPPPPPPPPPPPP)"!(
    ComponentPose=@n52
    Alpha=1
    ConstraintWeights_0=0.0002
    ConstraintWeights_1:float=0.9553
    ConstraintWeights_2:float=0
    ConstraintWeights_3:float=0.0004
    ConstraintWeights_4:float=0.0002
    ConstraintWeights_5:float=0
    ConstraintWeights_6:float=0.0003
    ConstraintWeights_7:float=0
    ConstraintWeights_8:float=0.0001
    ConstraintWeights_9:float=0.0435
    ConstraintWeights_10:float=0
    ConstraintWeights_11:float=0
    ConstraintWeights_12:float=0.0001
   ) -> (Pose) {
    alphaBoolBlend {blendOption=Linear}
    boneToModify {boneName=dyn_spikes_07}
    constraintSetup [
     {targetBone {boneName=R_eye_lid_lower_mid}}
     {targetBone {boneName=head}}
     {targetBone {boneName=L_cheek_inner}}
     {targetBone {boneName=L_brow_outer}}
     {targetBone {boneName=C_brow_mid}}
     {targetBone {boneName=L_eye_lid_lower_mid}}
     {targetBone {boneName=L_brow_mid}}
     {targetBone {boneName=L_eye_lid_upper_mid}}
     {targetBone {boneName=R_brow_mid}}
     {targetBone {boneName=R_eye_lid_upper_mid}}
     {targetBone {boneName=C_nose_bridge}}
     {targetBone {boneName=R_cheek_inner}}
     {targetBone {boneName=R_brow_outer}}
    ]
    constraintWeights [1,1,1,1,1,1,1,1,1,1,1,1,1]
   }
   n54 = "Constraint - dyn_spikes_09 (PPPPPPPPPPPPP)"!(
    ComponentPose=@n53
    Alpha=1
    ConstraintWeights_0=0.0001
    ConstraintWeights_1:float=0.9843
    ConstraintWeights_2:float=0
    ConstraintWeights_3:float=0.0001
    ConstraintWeights_4:float=0.0001
    ConstraintWeights_5:float=0
    ConstraintWeights_6:float=0.0001
    ConstraintWeights_7:float=0
    ConstraintWeights_8:float=0.0001
    ConstraintWeights_9:float=0.0152
    ConstraintWeights_10:float=0
    ConstraintWeights_11:float=0
    ConstraintWeights_12:float=0.0001
   ) -> (Pose) {
    alphaBoolBlend {blendOption=Linear}
    boneToModify {boneName=dyn_spikes_09}
    constraintSetup [
     {targetBone {boneName=R_eye_lid_lower_mid}}
     {targetBone {boneName=head}}
     {targetBone {boneName=L_cheek_inner}}
     {targetBone {boneName=L_brow_outer}}
     {targetBone {boneName=C_brow_mid}}
     {targetBone {boneName=L_eye_lid_lower_mid}}
     {targetBone {boneName=L_brow_mid}}
     {targetBone {boneName=L_eye_lid_upper_mid}}
     {targetBone {boneName=R_brow_mid}}
     {targetBone {boneName=R_eye_lid_upper_mid}}
     {targetBone {boneName=C_nose_bridge}}
     {targetBone {boneName=R_cheek_inner}}
     {targetBone {boneName=R_brow_outer}}
    ]
    constraintWeights [1,1,1,1,1,1,1,1,1,1,1,1,1]
   }
   n55 = "Constraint - dyn_spikes_08 (PPPPPPPPPPPPP)"!(
    ComponentPose=@n54
    Alpha=1
    ConstraintWeights_0=0
    ConstraintWeights_1:float=0.9922
    ConstraintWeights_2:float=0
    ConstraintWeights_3:float=0.0001
    ConstraintWeights_4:float=0
    ConstraintWeights_5:float=0
    ConstraintWeights_6:float=0
    ConstraintWeights_7:float=0
    ConstraintWeights_8:float=0
    ConstraintWeights_9:float=0.0076
    ConstraintWeights_10:float=0
    ConstraintWeights_11:float=0
    ConstraintWeights_12:float=0
   ) -> (Pose) {
    alphaBoolBlend {blendOption=Linear}
    boneToModify {boneName=dyn_spikes_08}
    constraintSetup [
     {targetBone {boneName=R_eye_lid_lower_mid}}
     {targetBone {boneName=head}}
     {targetBone {boneName=L_cheek_inner}}
     {targetBone {boneName=L_brow_outer}}
     {targetBone {boneName=C_brow_mid}}
     {targetBone {boneName=L_eye_lid_lower_mid}}
     {targetBone {boneName=L_brow_mid}}
     {targetBone {boneName=L_eye_lid_upper_mid}}
     {targetBone {boneName=R_brow_mid}}
     {targetBone {boneName=R_eye_lid_upper_mid}}
     {targetBone {boneName=C_nose_bridge}}
     {targetBone {boneName=R_cheek_inner}}
     {targetBone {boneName=R_brow_outer}}
    ]
    constraintWeights [1,1,1,1,1,1,1,1,1,1,1,1,1]
   }
   n56 = "Constraint - dyn_spikes_10 (PPPPPPPPPPPPP)"!(
    ComponentPose=@n55
    Alpha=1
    ConstraintWeights_0=0
    ConstraintWeights_1:float=0.9923
    ConstraintWeights_2:float=0
    ConstraintWeights_3:float=0.0001
    ConstraintWeights_4:float=0
    ConstraintWeights_5:float=0
    ConstraintWeights_6:float=0.0001
    ConstraintWeights_7:float=0
    ConstraintWeights_8:float=0
    ConstraintWeights_9:float=0.0074
    ConstraintWeights_10:float=0
    ConstraintWeights_11:float=0
    ConstraintWeights_12:float=0
   ) -> (Pose) {
    alphaBoolBlend {blendOption=Linear}
    boneToModify {boneName=dyn_spikes_10}
    constraintSetup [
     {targetBone {boneName=R_eye_lid_lower_mid}}
     {targetBone {boneName=head}}
     {targetBone {boneName=L_cheek_inner}}
     {targetBone {boneName=L_brow_outer}}
     {targetBone {boneName=C_brow_mid}}
     {targetBone {boneName=L_eye_lid_lower_mid}}
     {targetBone {boneName=L_brow_mid}}
     {targetBone {boneName=L_eye_lid_upper_mid}}
     {targetBone {boneName=R_brow_mid}}
     {targetBone {boneName=R_eye_lid_upper_mid}}
     {targetBone {boneName=C_nose_bridge}}
     {targetBone {boneName=R_cheek_inner}}
     {targetBone {boneName=R_brow_outer}}
    ]
    constraintWeights [1,1,1,1,1,1,1,1,1,1,1,1,1]
   }
   n57 = Anim Dynamics - Bone: dyn_spikes_19(ComponentPose=@n56, Alpha=1, ExternalForce="0,0,0") {
    angularDampingOverride=1
    angularSpringConstant=1500
    bAngularSpring=true
    bGravityOverrideInSimSpace=true
    bOverrideAngularDamping=true
    gravityScale=0.35
    simulationSpace=RootRelative
    alphaBoolBlend {blendOption=Linear}
    boundBone {boneName=dyn_spikes_19}
    componentAppliedLinearAccClamp=<3000 3000 3000>
    componentLinearAccScale=<0.2 0.2 4>
    componentLinearVelScale=<0.1 0.1 0.2>
    physicsBodyDefinitions [{
      sphereCollisionRadius=0
      boundBone {boneName=dyn_spikes_19}
      boxExtents=<10 10 4>
      constraintSetup {angularLimitsMax=<35 6 1>, angularLimitsMin=<-45 -6 -1>, angularTarget=<0 0 0>}
      localJointOffset=<0 0 -1>
     }]
   }
   n58 = Anim Dynamics - Bone: dyn_spikes_18(ComponentPose=@n57, Alpha=1, ExternalForce="0,0,0") {
    angularDampingOverride=1
    angularSpringConstant=1500
    bAngularSpring=true
    bGravityOverrideInSimSpace=true
    bOverrideAngularDamping=true
    gravityScale=0.35
    simulationSpace=RootRelative
    alphaBoolBlend {blendOption=Linear}
    boundBone {boneName=dyn_spikes_18}
    componentAppliedLinearAccClamp=<3000 3000 3000>
    componentLinearAccScale=<0.2 0.2 4>
    componentLinearVelScale=<0.1 0.1 0.2>
    physicsBodyDefinitions [{
      sphereCollisionRadius=0
      boundBone {boneName=dyn_spikes_18}
      boxExtents=<15 15 4>
      constraintSetup {angularLimitsMax=<35 6 1>, angularLimitsMin=<-45 -6 -1>, angularTarget=<0 0 0>}
      localJointOffset=<0 0 -3>
     }]
   }
   n59 = Anim Dynamics - Bone: dyn_spikes_17(ComponentPose=@n58, Alpha=1, ExternalForce="0,0,0") {
    angularDampingOverride=1
    angularSpringConstant=1500
    bAngularSpring=true
    bGravityOverrideInSimSpace=true
    bOverrideAngularDamping=true
    gravityScale=0.35
    simulationSpace=RootRelative
    alphaBoolBlend {blendOption=Linear}
    boundBone {boneName=dyn_spikes_17}
    componentAppliedLinearAccClamp=<3000 3000 3000>
    componentLinearAccScale=<0.2 0.2 4>
    componentLinearVelScale=<0.1 0.1 0.2>
    physicsBodyDefinitions [{
      sphereCollisionRadius=0
      boundBone {boneName=dyn_spikes_17}
      boxExtents=<20 20 4>
      constraintSetup {angularLimitsMax=<35 6 1>, angularLimitsMin=<-45 -6 -1>, angularTarget=<0 0 0>}
      localJointOffset=<0 0 -6>
     }]
   }
   n60 = Anim Dynamics - Bone: dyn_spikes_16(ComponentPose=@n59, Alpha=1, ExternalForce="0,0,0") {
    angularDampingOverride=1
    angularSpringConstant=1500
    bAngularSpring=true
    bGravityOverrideInSimSpace=true
    bOverrideAngularDamping=true
    gravityScale=0.35
    simulationSpace=RootRelative
    alphaBoolBlend {blendOption=Linear}
    boundBone {boneName=dyn_spikes_16}
    componentAppliedLinearAccClamp=<3000 3000 3000>
    componentLinearAccScale=<0.2 0.2 4>
    componentLinearVelScale=<0.1 0.1 0.2>
    physicsBodyDefinitions [{
      sphereCollisionRadius=0
      boundBone {boneName=dyn_spikes_16}
      boxExtents=<10 10 4>
      constraintSetup {angularLimitsMax=<35 6 1>, angularLimitsMin=<-45 -6 -1>, angularTarget=<0 0 0>}
      localJointOffset=<0 0 -1>
     }]
   }
   n61 = Anim Dynamics - Bone: dyn_spikes_15(ComponentPose=@n60, Alpha=1, ExternalForce="0,0,0") {
    angularDampingOverride=1
    angularSpringConstant=1500
    bAngularSpring=true
    bGravityOverrideInSimSpace=true
    bOverrideAngularDamping=true
    gravityScale=0.35
    simulationSpace=RootRelative
    alphaBoolBlend {blendOption=Linear}
    boundBone {boneName=dyn_spikes_15}
    componentAppliedLinearAccClamp=<3000 3000 3000>
    componentLinearAccScale=<0.2 0.2 4>
    componentLinearVelScale=<0.1 0.1 0.2>
    physicsBodyDefinitions [{
      sphereCollisionRadius=0
      boundBone {boneName=dyn_spikes_15}
      boxExtents=<15 15 4>
      constraintSetup {angularLimitsMax=<35 6 1>, angularLimitsMin=<-45 -6 -1>, angularTarget=<0 0 0>}
      localJointOffset=<0 0 -3>
     }]
   }
   n62 = Anim Dynamics - Bone: dyn_spikes_14(ComponentPose=@n61, Alpha=1, ExternalForce="0,0,0") {
    angularDampingOverride=1
    angularSpringConstant=1500
    bAngularSpring=true
    bGravityOverrideInSimSpace=true
    bOverrideAngularDamping=true
    gravityScale=0.35
    simulationSpace=RootRelative
    alphaBoolBlend {blendOption=Linear}
    boundBone {boneName=dyn_spikes_14}
    componentAppliedLinearAccClamp=<3000 3000 3000>
    componentLinearAccScale=<0.2 0.2 4>
    componentLinearVelScale=<0.1 0.1 0.2>
    physicsBodyDefinitions [{
      sphereCollisionRadius=0
      boundBone {boneName=dyn_spikes_14}
      boxExtents=<20 20 4>
      constraintSetup {angularLimitsMax=<35 6 1>, angularLimitsMin=<-45 -6 -1>, angularTarget=<0 0 0>}
      localJointOffset=<0 0 -6>
     }]
   }
   n63 = Anim Dynamics - Bone: dyn_spikes_13(ComponentPose=@n62, Alpha=1, ExternalForce="0,0,0") {
    angularDampingOverride=1
    angularSpringConstant=1500
    bAngularSpring=true
    bGravityOverrideInSimSpace=true
    bOverrideAngularDamping=true
    gravityScale=0.35
    simulationSpace=RootRelative
    alphaBoolBlend {blendOption=Linear}
    boundBone {boneName=dyn_spikes_13}
    componentAppliedLinearAccClamp=<3000 3000 3000>
    componentLinearAccScale=<0.2 0.2 4>
    componentLinearVelScale=<0.1 0.1 0.2>
    physicsBodyDefinitions [{
      sphereCollisionRadius=0
      boundBone {boneName=dyn_spikes_13}
      boxExtents=<10 10 4>
      constraintSetup {angularLimitsMax=<35 6 1>, angularLimitsMin=<-45 -6 -1>, angularTarget=<0 0 0>}
      localJointOffset=<0 0 -1>
     }]
   }
   n64 = Anim Dynamics - Bone: dyn_spikes_12(ComponentPose=@n63, Alpha=1, ExternalForce="0,0,0") {
    angularDampingOverride=1
    angularSpringConstant=1500
    bAngularSpring=true
    bGravityOverrideInSimSpace=true
    bOverrideAngularDamping=true
    gravityScale=0.35
    simulationSpace=RootRelative
    alphaBoolBlend {blendOption=Linear}
    boundBone {boneName=dyn_spikes_12}
    componentAppliedLinearAccClamp=<3000 3000 3000>
    componentLinearAccScale=<0.2 0.2 4>
    componentLinearVelScale=<0.1 0.1 0.2>
    physicsBodyDefinitions [{
      sphereCollisionRadius=0
      boundBone {boneName=dyn_spikes_12}
      boxExtents=<15 15 4>
      constraintSetup {angularLimitsMax=<35 6 1>, angularLimitsMin=<-45 -6 -1>, angularTarget=<0 0 0>}
      localJointOffset=<0 0 -3>
     }]
   }
   n65 = Anim Dynamics - Bone: dyn_spikes_11(ComponentPose=@n64, Alpha=1, ExternalForce="0,0,0") {
    angularDampingOverride=1
    angularSpringConstant=1500
    bAngularSpring=true
    bGravityOverrideInSimSpace=true
    bOverrideAngularDamping=true
    gravityScale=0.35
    simulationSpace=RootRelative
    alphaBoolBlend {blendOption=Linear}
    boundBone {boneName=dyn_spikes_11}
    componentAppliedLinearAccClamp=<3000 3000 3000>
    componentLinearAccScale=<0.2 0.2 4>
    componentLinearVelScale=<0.1 0.1 0.2>
    physicsBodyDefinitions [{
      sphereCollisionRadius=0
      boundBone {boneName=dyn_spikes_11}
      boxExtents=<20 20 4>
      constraintSetup {angularLimitsMax=<35 6 1>, angularLimitsMin=<-45 -6 -1>, angularTarget=<0 0 0>}
      localJointOffset=<0 0 -6>
     }]
   }
   n66 = Anim Dynamics - Bone: dyn_spikes_10(ComponentPose=@n65, Alpha=1, ExternalForce="0,0,0") {
    angularDampingOverride=1
    angularSpringConstant=1500
    bAngularSpring=true
    bGravityOverrideInSimSpace=true
    bOverrideAngularDamping=true
    gravityScale=0.35
    simulationSpace=RootRelative
    alphaBoolBlend {blendOption=Linear}
    boundBone {boneName=dyn_spikes_10}
    componentAppliedLinearAccClamp=<3000 3000 3000>
    componentLinearAccScale=<0.2 0.2 4>
    componentLinearVelScale=<0.1 0.1 0.2>
    physicsBodyDefinitions [{
      sphereCollisionRadius=0
      boundBone {boneName=dyn_spikes_10}
      boxExtents=<10 10 4>
      constraintSetup {angularLimitsMax=<35 6 1>, angularLimitsMin=<-45 -6 -1>, angularTarget=<0 0 0>}
      localJointOffset=<0 0 -1>
     }]
   }
   n67 = Anim Dynamics - Bone: dyn_spikes_09(ComponentPose=@n66, Alpha=1, ExternalForce="0,0,0") {
    angularDampingOverride=1
    angularSpringConstant=1500
    bAngularSpring=true
    bGravityOverrideInSimSpace=true
    bOverrideAngularDamping=true
    gravityScale=0.35
    simulationSpace=RootRelative
    alphaBoolBlend {blendOption=Linear}
    boundBone {boneName=dyn_spikes_09}
    componentAppliedLinearAccClamp=<3000 3000 3000>
    componentLinearAccScale=<0.2 0.2 4>
    componentLinearVelScale=<0.1 0.1 0.2>
    physicsBodyDefinitions [{
      sphereCollisionRadius=0
      boundBone {boneName=dyn_spikes_09}
      boxExtents=<15 15 4>
      constraintSetup {angularLimitsMax=<35 6 1>, angularLimitsMin=<-45 -6 -1>, angularTarget=<0 0 0>}
      localJointOffset=<0 0 -3>
     }]
   }
   n68 = Anim Dynamics - Bone: dyn_spikes_08(ComponentPose=@n67, Alpha=1, ExternalForce="0,0,0") {
    angularDampingOverride=1
    angularSpringConstant=1500
    bAngularSpring=true
    bGravityOverrideInSimSpace=true
    bOverrideAngularDamping=true
    gravityScale=0.35
    simulationSpace=RootRelative
    alphaBoolBlend {blendOption=Linear}
    boundBone {boneName=dyn_spikes_08}
    componentAppliedLinearAccClamp=<3000 3000 3000>
    componentLinearAccScale=<0.2 0.2 4>
    componentLinearVelScale=<0.1 0.1 0.2>
    physicsBodyDefinitions [{
      sphereCollisionRadius=0
      boundBone {boneName=dyn_spikes_08}
      boxExtents=<20 20 4>
      constraintSetup {angularLimitsMax=<35 6 1>, angularLimitsMin=<-45 -6 -1>, angularTarget=<0 0 0>}
      localJointOffset=<0 0 -6>
     }]
   }
   n69 = Anim Dynamics - Bone: dyn_spikes_07(ComponentPose=@n68, Alpha=1, ExternalForce="0,0,0") {
    angularDampingOverride=1
    angularSpringConstant=1500
    bAngularSpring=true
    bGravityOverrideInSimSpace=true
    bOverrideAngularDamping=true
    gravityScale=0.35
    simulationSpace=RootRelative
    alphaBoolBlend {blendOption=Linear}
    boundBone {boneName=dyn_spikes_07}
    componentAppliedLinearAccClamp=<3000 3000 3000>
    componentLinearAccScale=<0.2 0.2 4>
    componentLinearVelScale=<0.1 0.1 0.2>
    physicsBodyDefinitions [{
      sphereCollisionRadius=0
      boundBone {boneName=dyn_spikes_07}
      boxExtents=<10 10 4>
      constraintSetup {angularLimitsMax=<35 6 1>, angularLimitsMin=<-45 -6 -1>, angularTarget=<0 0 0>}
      localJointOffset=<0 0 -1>
     }]
   }
   n70 = Anim Dynamics - Bone: dyn_spikes_06(ComponentPose=@n69, Alpha=1, ExternalForce="0,0,0") {
    angularDampingOverride=1
    angularSpringConstant=1500
    bAngularSpring=true
    bGravityOverrideInSimSpace=true
    bOverrideAngularDamping=true
    gravityScale=0.35
    simulationSpace=RootRelative
    alphaBoolBlend {blendOption=Linear}
    boundBone {boneName=dyn_spikes_06}
    componentAppliedLinearAccClamp=<3000 3000 3000>
    componentLinearAccScale=<0.2 0.2 4>
    componentLinearVelScale=<0.1 0.1 0.2>
    physicsBodyDefinitions [{
      sphereCollisionRadius=0
      boundBone {boneName=dyn_spikes_06}
      boxExtents=<15 15 4>
      constraintSetup {angularLimitsMax=<35 6 1>, angularLimitsMin=<-45 -6 -1>, angularTarget=<0 0 0>}
      localJointOffset=<0 0 -3>
     }]
   }
   n71 = Anim Dynamics - Bone: dyn_spikes_05(ComponentPose=@n70, Alpha=1, ExternalForce="0,0,0") {
    angularDampingOverride=1
    angularSpringConstant=1500
    bAngularSpring=true
    bGravityOverrideInSimSpace=true
    bOverrideAngularDamping=true
    gravityScale=0.35
    simulationSpace=RootRelative
    alphaBoolBlend {blendOption=Linear}
    boundBone {boneName=dyn_spikes_05}
    componentAppliedLinearAccClamp=<3000 3000 3000>
    componentLinearAccScale=<0.2 0.2 4>
    componentLinearVelScale=<0.1 0.1 0.2>
    physicsBodyDefinitions [{
      sphereCollisionRadius=0
      boundBone {boneName=dyn_spikes_05}
      boxExtents=<20 20 4>
      constraintSetup {angularLimitsMax=<35 6 1>, angularLimitsMin=<-45 -6 -1>, angularTarget=<0 0 0>}
      localJointOffset=<0 0 -6>
     }]
   }
   n72 = Anim Dynamics - Bone: dyn_spikes_04(ComponentPose=@n71, Alpha=1, ExternalForce="0,0,0") {
    angularDampingOverride=1
    angularSpringConstant=1500
    bAngularSpring=true
    bGravityOverrideInSimSpace=true
    bOverrideAngularDamping=true
    gravityScale=0.35
    simulationSpace=RootRelative
    alphaBoolBlend {blendOption=Linear}
    boundBone {boneName=dyn_spikes_04}
    componentAppliedLinearAccClamp=<3000 3000 3000>
    componentLinearAccScale=<0.2 0.2 4>
    componentLinearVelScale=<0.1 0.1 0.2>
    physicsBodyDefinitions [{
      sphereCollisionRadius=0
      boundBone {boneName=dyn_spikes_04}
      boxExtents=<10 10 4>
      constraintSetup {angularLimitsMax=<35 6 1>, angularLimitsMin=<-45 -6 -1>, angularTarget=<0 0 0>}
      localJointOffset=<0 0 -1>
     }]
   }
   n73 = Anim Dynamics - Bone: dyn_spikes_03(ComponentPose=@n72, Alpha=1, ExternalForce="0,0,0") {
    angularDampingOverride=1
    angularSpringConstant=1500
    bAngularSpring=true
    bGravityOverrideInSimSpace=true
    bOverrideAngularDamping=true
    gravityScale=0.35
    simulationSpace=RootRelative
    alphaBoolBlend {blendOption=Linear}
    boundBone {boneName=dyn_spikes_03}
    componentAppliedLinearAccClamp=<3000 3000 3000>
    componentLinearAccScale=<0.2 0.2 4>
    componentLinearVelScale=<0.1 0.1 0.2>
    physicsBodyDefinitions [{
      sphereCollisionRadius=0
      boundBone {boneName=dyn_spikes_03}
      boxExtents=<15 15 4>
      constraintSetup {angularLimitsMax=<35 6 1>, angularLimitsMin=<-45 -6 -1>, angularTarget=<0 0 0>}
      localJointOffset=<0 0 -3>
     }]
   }
   n74 = Anim Dynamics - Bone: dyn_spikes_02(ComponentPose=@n73, Alpha=1, ExternalForce="0,0,0") {
    angularDampingOverride=1
    angularSpringConstant=1500
    bAngularSpring=true
    bGravityOverrideInSimSpace=true
    bOverrideAngularDamping=true
    gravityScale=0.35
    simulationSpace=RootRelative
    alphaBoolBlend {blendOption=Linear}
    boundBone {boneName=dyn_spikes_02}
    componentAppliedLinearAccClamp=<3000 3000 3000>
    componentLinearAccScale=<0.2 0.2 4>
    componentLinearVelScale=<0.1 0.1 0.2>
    physicsBodyDefinitions [{
      sphereCollisionRadius=0
      boundBone {boneName=dyn_spikes_02}
      boxExtents=<20 20 4>
      constraintSetup {angularLimitsMax=<35 6 1>, angularLimitsMin=<-45 -6 -1>, angularTarget=<0 0 0>}
      localJointOffset=<0 0 -6>
     }]
   }
   n75 = Anim Dynamics - Bone: dyn_spikes_01(ComponentPose=@n74, Alpha=1, ExternalForce="0,0,0") {
    angularDampingOverride=1
    angularSpringConstant=1500
    bAngularSpring=true
    bGravityOverrideInSimSpace=true
    bOverrideAngularDamping=true
    gravityScale=0.35
    simulationSpace=RootRelative
    alphaBoolBlend {blendOption=Linear}
    boundBone {boneName=dyn_spikes_01}
    componentAppliedLinearAccClamp=<3000 3000 3000>
    componentLinearAccScale=<0.2 0.2 4>
    componentLinearVelScale=<0.1 0.1 0.2>
    physicsBodyDefinitions [{
      sphereCollisionRadius=0
      boundBone {boneName=dyn_spikes_01}
      boxExtents=<10 10 4>
      constraintSetup {angularLimitsMax=<35 6 1>, angularLimitsMin=<-45 -6 -1>, angularTarget=<0 0 0>}
      localJointOffset=<0 0 -1>
     }]
   }
   n76 = Anim Dynamics - Bone: dyn_spikes_00(ComponentPose=@n75, Alpha=1, ExternalForce="0,0,0") {
    angularDampingOverride=1
    angularSpringConstant=1500
    bAngularSpring=true
    bGravityOverrideInSimSpace=true
    bOverrideAngularDamping=true
    gravityScale=0.35
    simulationSpace=RootRelative
    alphaBoolBlend {blendOption=Linear}
    boundBone {boneName=dyn_spikes_00}
    componentAppliedLinearAccClamp=<3000 3000 3000>
    componentLinearAccScale=<0.2 0.2 4>
    componentLinearVelScale=<0.1 0.1 0.2>
    physicsBodyDefinitions [{
      sphereCollisionRadius=0
      boundBone {boneName=dyn_spikes_00}
      boxExtents=<15 15 4>
      constraintSetup {angularLimitsMax=<35 6 1>, angularLimitsMin=<-45 -6 -1>, angularTarget=<0 0 0>}
      localJointOffset=<0 0 -3>
     }]
   }
   n77 = Anim Dynamics - Bone: dyn_spikes_20(ComponentPose=@n76, Alpha=1, ExternalForce="0,0,0") {
    angularDampingOverride=1
    angularSpringConstant=1500
    bAngularSpring=true
    bGravityOverrideInSimSpace=true
    bOverrideAngularDamping=true
    gravityScale=0.35
    simulationSpace=RootRelative
    alphaBoolBlend {blendOption=Linear}
    boundBone {boneName=dyn_spikes_20}
    componentAppliedLinearAccClamp=<3000 3000 3000>
    componentLinearAccScale=<0.2 0.2 4>
    componentLinearVelScale=<0.1 0.1 0.2>
    physicsBodyDefinitions [{
      sphereCollisionRadius=0
      boundBone {boneName=dyn_spikes_20}
      boxExtents=<20 20 4>
      constraintSetup {angularLimitsMax=<35 6 1>, angularLimitsMin=<-45 -6 -1>, angularTarget=<0 0 0>}
      localJointOffset=<0 0 -6>
     }]
   }
   Save cached pose 'PostPoseAsset'(Pose=Component To Local(ComponentPose=@n77)) {cacheName=PostPoseAsset}
   n78 = Use cached pose 'PostPoseAsset'()
   Save cached pose 'preScaleCheck'(Pose=Layered blend per bone(BasePose=Use cached pose 'PostPoseAsset'(), BlendPoses_0=Slot 'FaceSlot'(Source=@n78) {slotName=FaceSlot}) {layerSetup [{branchFilters [{boneName=faceAttach}]}]}) {cacheName=preScaleCheck}
   n79 = Break Fort Anim Input Facial!(FortAnimInput_Facial:"Fort Anim Input Facial Structure (by ref)"=Get FacialInput() -> (FacialInput:Fort Anim Input Facial Structure)) -> (SubjectName:Live Link Subject Name Structure, CurrentAnimType:EFortFacialAnimTypes Enum, bCurvesOnly:bool=~) {is=title K2Node_BreakStruct}
   Save cached pose 'LiveLinkFacial'(Pose=Blend Poses by bool(BlendPose_0=Use cached pose 'GameFacial'(), BlendPose_1=Use cached pose 'GameFacial'(), bActiveValue=@n79.bCurvesOnly) {blendType=Linear}) {cacheName=LiveLinkFacial}
   n80 = "Blend Poses (EFortFacialAnimTypes)"(
    BlendPose_0=Use cached pose 'GameFacial'()
    BlendPose_1=Layered blend per bone(BasePose=Use cached pose 'GameFacial'(), BlendPoses_0=Use cached pose 'LiveLinkFacial'()) {layerSetup [{branchFilters [{boneName=faceAttach}]}]}
    BlendPose_2=Use cached pose 'LiveLinkFacial'()
    BlendTime_0=0.1
    BlendTime_1=0.1
    BlendTime_2=0.1
    ActiveEnumValue=Get() -> (FacialInput_bCurvesOnly=~).FacialInput_CurrentAnimType
   ) {
    blendType=Linear
    blendPose [{linkId=-1, sourceLinkId=-1}, {linkId=-1, sourceLinkId=-1}, {linkId=-1, sourceLinkId=-1}]
    blendTime [0.1,0.1,0.1]
   }
   Save cached pose 'post driven'(Pose=@n80) {cacheName=post driven}
   Save cached pose 'IncomingPose'(Pose=Copy Pose From Mesh() {bCopyCurves=true, bUseAttachedParent=true}) {cacheName=IncomingPose}
   Sequence Player 'Fortnite_Base_Head_ROM'() {sequence=/Game/Characters/Player/Common/Fortnite_Base_Head/Anim_Tests/Fortnite_Base_Head_ROM}
   GetPlaylistUsesCustomCharacterParts!() -> (ReturnValue:bool)
   Reroute Node(InputPin:Live Link Subject Name Structure=@n79) -> (OutputPin:Live Link Subject Name Structure)
  ]}
 EventGraph {nodes [
   Event OnCharacterCustomizationLoaded()
   > n1 = Branch(Condition=CharacterItemDefinitionHasAnyMetaTag(Tags=Make Array!(
      "[0]"=(TagName="Cosmetics.Variant.Meta.AlienTrooper.Tentacle")
      "[1]":Gameplay Tag Structure=(TagName="Cosmetics.Variant.Meta.AlienTrooper.Horns")
      "[2]":Gameplay Tag Structure=(TagName="Cosmetics.Variant.Meta.AlienTrooper.Owl")
      "[3]":Gameplay Tag Structure
     ) -> (Array)))
   > Set Chin_Active(Chin_Active:bool=true) -> (Output_Get:bool)
   n2 = Branch(
    execute=@n1.else
    Condition=CharacterItemDefinitionHasAnyMetaTag(Tags=Make Array!("[0]"=(TagName="Cosmetics.Variant.Meta.AlienTrooper.Cyclops")) -> (Array))
   )
   > Set Cyclops_Chin_Active(Cyclops_Chin_Active:bool=true) -> (Output_Get:bool)
   Set Chin_Active(execute=@n2.else, Chin_Active=false)
   > Set Cyclops_Chin_Active(Cyclops_Chin_Active=false)
   K2Node_Event_2 = CustomEvent() {is=title K2Node_CustomEvent}
   n3 = TryGetFortPawn()
   > Is Valid(InputObject=@n3)
   > n4 = Cast To FortPlayerPawn(Object=@n3)
   > n5 = Get Dynamic Material Instance(SkeletalMeshComponent=GetSkeletalMeshForPartType(self=@n4, PartType=Head))
   > Set PartMID(PartMID:Material Instance Dynamic Object Reference=@n5) -> (Output_Get:Material Instance Dynamic Object Reference)
   Event BlueprintInitializeAnimation()
   n6 = GetOwnerSkeletalMeshComponent()
   > Is Valid(InputObject=@n6)
   > Set MeshToCopy(MeshToCopy:Skeletal Mesh Component Object Reference=@n6) -> (Output_Get:Skeletal Mesh Component Object Reference)
   > Set GetPlayerListUsesCustomCharacterParts(GetPlayerListUsesCustomCharacterParts:bool=GetPlaylistUsesCustomCharacterParts()) -> (Output_Get:bool)
   TryGetPawnOwner()
   TryGetPawnOwner()
  ]}
]
variables [
 Chin_Active {type=boolean}
 Cyclops_Chin_Active {type=boolean}
 GetPlayerListUsesCustomCharacterParts {type=boolean}
 MeshToCopy {type=SkeletalMeshComponent}
 PartMID {type=MaterialInstanceDynamic}
]
